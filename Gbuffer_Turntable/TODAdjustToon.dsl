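; TODAdjustToon is a helper Blueprint that re-orients the global time-of-day actor, apparently for a toon G-buffer turntable setup.
; It casts the global time-of-day manager to FortDaySequenceActor and sets that actor's yaw to a fixed angle, keeping its existing pitch and roll.
; The same logic runs in the Construction Script and on BeginPlay, so the change applies at construction and play start.
; Tick and overlap events are empty.
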(event EventBeginPlay
  (bind _asfort_day_sequence_actor (Utilities|Casting|CastToFortDaySequenceActor (TimeOfDay|GetGlobalTimeOfDayManager)))
  (Transformation|SetActorRotation _asfort_day_sequence_actor (Math|Rotator|MakeRotator (.roll (Transformation|GetActorRotation _asfort_day_sequence_actor)) (.pitch (Transformation|GetActorRotation _asfort_day_sequence_actor)) 1.7e+02) false))

(event Collision|EventActorBeginOverlap
  (params (OtherActor :type Actor%20Object%20Reference)))

(event EventTick
  (params (DeltaSeconds :type Float%20%28single-precision%29)))

(fn ConstructionScript
  (bind _asfort_day_sequence_actor (Utilities|Casting|CastToFortDaySequenceActor (TimeOfDay|GetGlobalTimeOfDayManager)))
  (Transformation|SetActorRotation _asfort_day_sequence_actor (Math|Rotator|MakeRotator (.roll (Transformation|GetActorRotation _asfort_day_sequence_actor)) (.pitch (Transformation|GetActorRotation _asfort_day_sequence_actor)) 1.7e+02) false))
