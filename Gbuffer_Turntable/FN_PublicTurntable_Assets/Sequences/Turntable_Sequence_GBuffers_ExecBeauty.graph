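# Level Sequence for a G-buffer turntable "ExecBeauty" capture, staging user objects with a camera, background, planes and spawned ExecRimlights.
# Timed event keys drive BP_Turntable_GBuffers and BP_LevelsChecker, alongside master trigger and repeater event tracks.
# Eight camera cuts and two level-visibility phases split the shots, while Turntable_UserObjects snaps between 0 and 180 degrees yaw.
# Background and plane visibility toggles and keyed camera exposure compensation vary the setup between shots.
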LevelSequence /Gbuffer_Turntable/FN_PublicTurntable_Assets/Sequences/Turntable_Sequence_GBuffers_ExecBeauty
bindings [
 BP_Turntable_GBuffers {
  id=#1
  possessed_class=BP_Turntable_GBuffers_C
  tracks [
   {
    type=MovieSceneEventTrack
    sections [
     {
      type=MovieSceneEventTriggerSection
      channels [None {
        type=MovieSceneScriptingEventChannel
        keys [
         [0, "<Struct 'MovieSceneEvent' (0x49e7bfb6c20) {}>"]
         [180, "<Struct 'MovieSceneEvent' (0x49e7bfb6a40) {}>"]
         [210, "<Struct 'MovieSceneEvent' (0x49e7bfb6720) {}>"]
         [240, "<Struct 'MovieSceneEvent' (0x49e7bfb6860) {}>"]
         [270, "<Struct 'MovieSceneEvent' (0x49e7bfb5500) {}>"]
         [300, "<Struct 'MovieSceneEvent' (0x49e7bfb65e0) {}>"]
         [330, "<Struct 'MovieSceneEvent' (0x49e7bfb6400) {}>"]
         [360, "<Struct 'MovieSceneEvent' (0x49e7bfb6220) {}>"]
         [390, "<Struct 'MovieSceneEvent' (0x49e7bfb5fa0) {}>"]
        ]
       }]
     }
     {
      type=MovieSceneEventTriggerSection
      channels [None {
        type=MovieSceneScriptingEventChannel
        keys [
         [-1, "<Struct 'MovieSceneEvent' (0x49e7bfb6680) {}>"]
         [0, "<Struct 'MovieSceneEvent' (0x49e7bfb6540) {}>"]
         [330, "<Struct 'MovieSceneEvent' (0x49e7bfb5c80) {}>"]
         [360, "<Struct 'MovieSceneEvent' (0x49e7bfb50a0) {}>"]
        ]
       }]
     }
    ]
   }
   {
    type=MovieSceneVisibilityTrack
    sections [{type=MovieSceneVisibilitySection, channels [None {type=MovieSceneScriptingBoolChannel, default=true}]}]
   }
  ]
 }
 SM_ColorCalibrator {
  id=#2
  possessed_class=StaticMeshActor
  tracks [
   {
    type=MovieScene3DTransformTrack
    sections [{
      type=MovieScene3DTransformSection
      channels [
       Location.X {type=MovieSceneScriptingDoubleChannel, default=133}
       Location.Y {type=MovieSceneScriptingDoubleChannel, default=-23}
       Location.Z {type=MovieSceneScriptingDoubleChannel, default=32}
       Rotation.X {type=MovieSceneScriptingDoubleChannel, default=0}
       Rotation.Y {type=MovieSceneScriptingDoubleChannel, default=0}
       Rotation.Z {type=MovieSceneScriptingDoubleChannel, default=-30}
       Scale.X {type=MovieSceneScriptingDoubleChannel, default=0.8}
       Scale.Y {type=MovieSceneScriptingDoubleChannel, default=0.8}
       Scale.Z {type=MovieSceneScriptingDoubleChannel, default=0.8}
      ]
     }]
   }
   {
    type=MovieSceneVisibilityTrack
    sections [{type=MovieSceneVisibilitySection, channels [None {type=MovieSceneScriptingBoolChannel, default=false}]}]
   }
  ]
 }
 ExecRimlights {
  id=#3
  possessed_class=ExecRimlights_C
  tracks [
   {type=MovieSceneSpawnTrack, sections [{type=MovieSceneSpawnSection, channels [None {type=MovieSceneScriptingBoolChannel, default=true}]}]}
   {
    type=MovieScene3DTransformTrack
    sections [{
      type=MovieScene3DTransformSection
      channels [
       Location.X {type=MovieSceneScriptingDoubleChannel, default=0}
       Location.Y {type=MovieSceneScriptingDoubleChannel, default=0}
       Location.Z {type=MovieSceneScriptingDoubleChannel, default=0}
       Rotation.X {type=MovieSceneScriptingDoubleChannel, default=0}
       Rotation.Y {type=MovieSceneScriptingDoubleChannel, default=0}
       Rotation.Z {type=MovieSceneScriptingDoubleChannel, default=0}
       Scale.X {type=MovieSceneScriptingDoubleChannel, default=1}
       Scale.Y {type=MovieSceneScriptingDoubleChannel, default=1}
       Scale.Z {type=MovieSceneScriptingDoubleChannel, default=1}
      ]
     }]
   }
  ]
 }
 CameraActor {id=#4, possessed_class=CameraActor, tracks []}
 SM_background {
  id=#5
  possessed_class=StaticMeshActor
  tracks [
   {
    type=MovieScene3DTransformTrack
    sections [{
      type=MovieScene3DTransformSection
      channels [
       Location.X {type=MovieSceneScriptingDoubleChannel, default=0}
       Location.Y {type=MovieSceneScriptingDoubleChannel, default=-460}
       Location.Z {type=MovieSceneScriptingDoubleChannel, default=202}
       Rotation.X {type=MovieSceneScriptingDoubleChannel, default=0}
       Rotation.Y {type=MovieSceneScriptingDoubleChannel, default=0}
       Rotation.Z {type=MovieSceneScriptingDoubleChannel, default=0}
       Scale.X {type=MovieSceneScriptingDoubleChannel, default=1}
       Scale.Y {type=MovieSceneScriptingDoubleChannel, default=1}
       Scale.Z {type=MovieSceneScriptingDoubleChannel, default=1}
      ]
     }]
   }
   {
    type=MovieSceneVisibilityTrack
    sections [{
      type=MovieSceneVisibilitySection
      channels [None {type=MovieSceneScriptingBoolChannel, default=true, keys [[-2,true],[0,false],[9,true],[11,false]]}]
     }]
   }
  ]
 }
 Plane {
  id=#6
  possessed_class=StaticMeshActor
  tracks [
   {
    type=MovieScene3DTransformTrack
    sections [{
      type=MovieScene3DTransformSection
      channels [
       Location.X {type=MovieSceneScriptingDoubleChannel, default=12}
       Location.Y {type=MovieSceneScriptingDoubleChannel, default=-128}
       Location.Z {type=MovieSceneScriptingDoubleChannel, default=84}
       Rotation.X {type=MovieSceneScriptingDoubleChannel, default=90.0001}
       Rotation.Y {type=MovieSceneScriptingDoubleChannel, default=0}
       Rotation.Z {type=MovieSceneScriptingDoubleChannel, default=0}
       Scale.X {type=MovieSceneScriptingDoubleChannel, default=5.5}
       Scale.Y {type=MovieSceneScriptingDoubleChannel, default=5.5}
       Scale.Z {type=MovieSceneScriptingDoubleChannel, default=1}
      ]
     }]
   }
   {
    type=MovieSceneVisibilityTrack
    sections [{type=MovieSceneVisibilitySection, channels [None {type=MovieSceneScriptingBoolChannel, default=false, keys [[4,true],[6,false]]}]}]
   }
  ]
 }
 Turntable_UserObjects {
  id=#7
  possessed_class=Actor
  tracks [{
    type=MovieScene3DTransformTrack
    sections [{
      type=MovieScene3DTransformSection
      channels [
       Location.X {type=MovieSceneScriptingDoubleChannel, default=0}
       Location.Y {type=MovieSceneScriptingDoubleChannel, default=0}
       Location.Z {type=MovieSceneScriptingDoubleChannel, default=0}
       Rotation.Z {
        type=MovieSceneScriptingDoubleChannel
        default=180
        keys [[0,0],[1,180],[2,0],[3,180],[4,0],[5,180],[6,0],[7,180],[8,180],[9,0],[10,180],[11,0],[12,180],[13,0],[14,180],[15,180]]
       }
      ]
     }]
   }]
 }
 CameraComponent {
  id=#8
  parent_id=#4
  possessed_class=CameraComponent
  tracks [
   {
    type=MovieSceneFloatTrack
    display_name=Field Of View
    sections [{type=MovieSceneFloatSection, channels [None {type=MovieSceneScriptingFloatChannel, default=35}]}]
   }
   {
    type=MovieSceneFloatTrack
    display_name="Exposure Compensation (Post Process Settings)"
    sections [{
      type=MovieSceneFloatSection
      channels [None {type=MovieSceneScriptingFloatChannel, default=-0.8, keys [[329,0],[330,-0.8],[359,-0.8],[360,0]]}]
     }]
   }
  ]
 }
 ReplaceMe {
  id=#9
  possessed_class=StaticMeshActor
  tracks [
   {
    type=MovieScene3DTransformTrack
    sections [{
      type=MovieScene3DTransformSection
      channels [
       Location.X {type=MovieSceneScriptingDoubleChannel, default=0}
       Location.Y {type=MovieSceneScriptingDoubleChannel, default=0}
       Location.Z {type=MovieSceneScriptingDoubleChannel, default=50}
       Rotation.X {type=MovieSceneScriptingDoubleChannel, default=0}
       Rotation.Y {type=MovieSceneScriptingDoubleChannel, default=0}
       Rotation.Z {type=MovieSceneScriptingDoubleChannel, default=0}
       Scale.X {type=MovieSceneScriptingDoubleChannel, default=0.3}
       Scale.Y {type=MovieSceneScriptingDoubleChannel, default=0.3}
       Scale.Z {type=MovieSceneScriptingDoubleChannel, default=0.3}
      ]
     }]
   }
   {
    type=MovieSceneVisibilityTrack
    sections [{type=MovieSceneVisibilitySection, channels [None {type=MovieSceneScriptingBoolChannel, default=false}]}]
   }
  ]
 }
 Plane2 {
  id=#10
  possessed_class=StaticMeshActor
  tracks [
   {
    type=MovieScene3DTransformTrack
    sections [{
      type=MovieScene3DTransformSection
      channels [
       Location.X {type=MovieSceneScriptingDoubleChannel, default=12}
       Location.Y {type=MovieSceneScriptingDoubleChannel, default=32}
       Location.Z {type=MovieSceneScriptingDoubleChannel, default=0.3}
       Rotation.X {type=MovieSceneScriptingDoubleChannel, default=0}
       Rotation.Y {type=MovieSceneScriptingDoubleChannel, default=0}
       Rotation.Z {type=MovieSceneScriptingDoubleChannel, default=0}
       Scale.X {type=MovieSceneScriptingDoubleChannel, default=5.5}
       Scale.Y {type=MovieSceneScriptingDoubleChannel, default=5.5}
       Scale.Z {type=MovieSceneScriptingDoubleChannel, default=1}
      ]
     }]
   }
   {
    type=MovieSceneVisibilityTrack
    sections [{type=MovieSceneVisibilitySection, channels [None {type=MovieSceneScriptingBoolChannel, default=false, keys [[4,true],[6,false]]}]}]
   }
  ]
 }
 BP_LevelsChecker {
  id=#11
  possessed_class=BP_LevelsChecker_C
  tracks [
   {
    type=MovieSceneEventTrack
    sections [{
      type=MovieSceneEventTriggerSection
      channels [None {
        type=MovieSceneScriptingEventChannel
        keys [[180, "<Struct 'MovieSceneEvent' (0x49e7bfb73a0) {}>"], [210, "<Struct 'MovieSceneEvent' (0x49e7bfb7300) {}>"]]
       }]
     }]
   }
   {
    type=MovieSceneVisibilityTrack
    sections [{type=MovieSceneVisibilitySection, channels [None {type=MovieSceneScriptingBoolChannel, default=false}]}]
   }
  ]
 }
]
display_rate {denominator=1, numerator=30}
tick_resolution {denominator=1, numerator=24000}
tracks [
 {
  type=MovieSceneLevelVisibilityTrack
  sections [
   {type=MovieSceneLevelVisibilitySection, end_frame=4, start_frame=-17}
   {type=MovieSceneLevelVisibilitySection, end_frame=4, start_frame=-17}
   {type=MovieSceneLevelVisibilitySection, end_frame=16, start_frame=4}
   {type=MovieSceneLevelVisibilitySection, end_frame=16, start_frame=4}
  ]
 }
 {
  type=MovieSceneEventTrack
  sections [{
    type=MovieSceneEventTriggerSection
    channels [None {type=MovieSceneScriptingEventChannel, const ["<Struct 'MovieSceneEvent' (0x49e7bfb3fc0) {}>", 1]}]
   }]
 }
 {type=MovieSceneEventTrack, sections [{type=MovieSceneEventRepeaterSection, end_frame=391, start_frame=0}]}
 {
  type=MovieSceneCameraCutTrack
  sections [
   {type=MovieSceneCameraCutSection, end_frame=2, start_frame=-16}
   {type=MovieSceneCameraCutSection, end_frame=4, start_frame=2}
   {type=MovieSceneCameraCutSection, end_frame=6, start_frame=4}
   {type=MovieSceneCameraCutSection, end_frame=8, start_frame=6}
   {type=MovieSceneCameraCutSection, end_frame=10, start_frame=8}
   {type=MovieSceneCameraCutSection, end_frame=12, start_frame=10}
   {type=MovieSceneCameraCutSection, end_frame=14, start_frame=12}
   {type=MovieSceneCameraCutSection, end_frame=17, start_frame=14}
  ]
 }
]
