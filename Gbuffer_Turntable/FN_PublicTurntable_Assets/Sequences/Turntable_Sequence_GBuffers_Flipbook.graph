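# Level Sequence for a G-buffer turntable flipbook capture, staging a reference mannequin and a Jonesy example side by side before a static, narrow-FOV camera.
# A ManifestTracks event track triggers BP_Turntable_GBuffers in pre-roll, on each early frame and once late in the range.
# Master tracks add level visibility, a camera cut and trigger and repeater events.
# Transforms and face Control Rig values are held constant, so no turntable spin is keyed here.
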LevelSequence /Gbuffer_Turntable/FN_PublicTurntable_Assets/Sequences/Turntable_Sequence_GBuffers_Flipbook
bindings [
 BP_Turntable_GBuffers {
  id=#1
  possessed_class=BP_Turntable_GBuffers_C
  tracks [{
    type=MovieSceneEventTrack
    display_name=ManifestTracks
    sections [
     {
      type=MovieSceneEventTriggerSection
      channels [None {
        type=MovieSceneScriptingEventChannel
        keys [
         [-0.9937, "<Struct 'MovieSceneEvent' (0x49e7bfb96a0) {}>"]
         [0, "<Struct 'MovieSceneEvent' (0x49e7bfb9920) {}>"]
         [1, "<Struct 'MovieSceneEvent' (0x49e7bfb9880) {}>"]
         [2, "<Struct 'MovieSceneEvent' (0x49e7bfb9a60) {}>"]
         [3, "<Struct 'MovieSceneEvent' (0x49e7bfb97e0) {}>"]
         [4, "<Struct 'MovieSceneEvent' (0x49e7bfb9b00) {}>"]
         [5, "<Struct 'MovieSceneEvent' (0x49e7bfb9ba0) {}>"]
         [6, "<Struct 'MovieSceneEvent' (0x49e7bfb9600) {}>"]
         [7, "<Struct 'MovieSceneEvent' (0x49e7bfb9560) {}>"]
         [8, "<Struct 'MovieSceneEvent' (0x49e7bfb92e0) {}>"]
         [9, "<Struct 'MovieSceneEvent' (0x49e7bfb99c0) {}>"]
         [10, "<Struct 'MovieSceneEvent' (0x49e7bfb9420) {}>"]
         [11, "<Struct 'MovieSceneEvent' (0x49e7bfb9240) {}>"]
         [12, "<Struct 'MovieSceneEvent' (0x49e7bfb94c0) {}>"]
         [13, "<Struct 'MovieSceneEvent' (0x49e7bfb8ac0) {}>"]
         [14, "<Struct 'MovieSceneEvent' (0x49e7bfb8f20) {}>"]
         [15, "<Struct 'MovieSceneEvent' (0x49e7bfb9060) {}>"]
         [16, "<Struct 'MovieSceneEvent' (0x49e7bfb9100) {}>"]
         [17, "<Struct 'MovieSceneEvent' (0x49e7bfb9380) {}>"]
         [18, "<Struct 'MovieSceneEvent' (0x49e7bfb9c40) {}>"]
         [19, "<Struct 'MovieSceneEvent' (0x49e7bfb9ce0) {}>"]
         [20, "<Struct 'MovieSceneEvent' (0x49e7bfb9d80) {}>"]
         [21, "<Struct 'MovieSceneEvent' (0x49e7bfb9e20) {}>"]
         [22, "<Struct 'MovieSceneEvent' (0x49e7bfb9ec0) {}>"]
         [23, "<Struct 'MovieSceneEvent' (0x49e7bfb9f60) {}>"]
         [24, "<Struct 'MovieSceneEvent' (0x49e7bfba000) {}>"]
         [25, "<Struct 'MovieSceneEvent' (0x49e7bfba0a0) {}>"]
         [26, "<Struct 'MovieSceneEvent' (0x49e7bfba140) {}>"]
         [27, "<Struct 'MovieSceneEvent' (0x49e7bfba1e0) {}>"]
         [28, "<Struct 'MovieSceneEvent' (0x49e7bfba280) {}>"]
         [424, "<Struct 'MovieSceneEvent' (0x49e7bfba320) {}>"]
        ]
       }]
     }
     {
      type=MovieSceneEventTriggerSection
      channels [None {
        type=MovieSceneScriptingEventChannel
        keys [[-2, "<Struct 'MovieSceneEvent' (0x49e7bfba6e0) {}>"], [-0.9937, "<Struct 'MovieSceneEvent' (0x49e7bfba8c0) {}>"]]
       }]
     }
    ]
   }]
 }
 SkeletalMeshComponent0 {
  id=#2
  parent_id=#3
  possessed_class=SkeletalMeshComponentBudgeted
  tracks [{
    type=MovieSceneControlRigParameterTrack
    display_name=Fortnite_Legacy_Face_CtrlRig
    sections [{
      type=MovieSceneControlRigParameterSection
      channels [
       Flipbook {type=MovieSceneScriptingBoolChannel, default=false, const [false,547]}
       Face_Ctrl_Master.Location.X {type=MovieSceneScriptingFloatChannel, default=0, const [0,575]}
       Face_Ctrl_Master.Location.Y {type=MovieSceneScriptingFloatChannel, default=0, const [0,575]}
       Face_Ctrl_Master.Location.Z {type=MovieSceneScriptingFloatChannel, default=0, const [0,575]}
       Face_Ctrl_Master.Rotation.X {type=MovieSceneScriptingFloatChannel, default=0, thinned_from=575, keys [[0,0],[574,0]]}
       Face_Ctrl_Master.Rotation.Y {type=MovieSceneScriptingFloatChannel, default=0, const [0,575]}
       Face_Ctrl_Master.Rotation.Z {type=MovieSceneScriptingFloatChannel, default=0, const [0,575]}
       Face_Ctrl_Master.Scale.X {type=MovieSceneScriptingFloatChannel, default=1, const [1,575]}
       Face_Ctrl_Master.Scale.Y {type=MovieSceneScriptingFloatChannel, default=1, const [1,575]}
       Face_Ctrl_Master.Scale.Z {type=MovieSceneScriptingFloatChannel, default=1, const [1,575]}
       gui_follow_head {type=MovieSceneScriptingFloatChannel, default=1, const [1,575]}
      ]
     }]
   }]
 }
 CameraComponent {
  id=#4
  parent_id=#5
  possessed_class=CameraComponent
  tracks [{
    type=MovieSceneFloatTrack
    display_name=Field Of View
    sections [{type=MovieSceneFloatSection, channels [None {type=MovieSceneScriptingFloatChannel, default=10.25}]}]
   }]
 }
 SkeletalMeshComponent0 {
  id=#6
  parent_id=#7
  possessed_class=SkeletalMeshComponentBudgeted
  tracks [{
    type=MovieSceneControlRigParameterTrack
    display_name=Fortnite_Legacy_Face_CtrlRig
    sections [{
      type=MovieSceneControlRigParameterSection
      channels [
       Flipbook {type=MovieSceneScriptingBoolChannel, default=false, const [false,547]}
       Face_Ctrl_Master.Location.X {type=MovieSceneScriptingFloatChannel, default=0, const [0,575]}
       Face_Ctrl_Master.Location.Y {type=MovieSceneScriptingFloatChannel, default=0, const [0,575]}
       Face_Ctrl_Master.Location.Z {type=MovieSceneScriptingFloatChannel, default=0, const [0,575]}
       Face_Ctrl_Master.Rotation.X {type=MovieSceneScriptingFloatChannel, default=0, thinned_from=575, keys [[0,0],[574,0]]}
       Face_Ctrl_Master.Rotation.Y {type=MovieSceneScriptingFloatChannel, default=0, const [0,575]}
       Face_Ctrl_Master.Rotation.Z {type=MovieSceneScriptingFloatChannel, default=0, const [0,575]}
       Face_Ctrl_Master.Scale.X {type=MovieSceneScriptingFloatChannel, default=1, const [1,575]}
       Face_Ctrl_Master.Scale.Y {type=MovieSceneScriptingFloatChannel, default=1, const [1,575]}
       Face_Ctrl_Master.Scale.Z {type=MovieSceneScriptingFloatChannel, default=1, const [1,575]}
       gui_follow_head {type=MovieSceneScriptingFloatChannel, default=1, const [1,575]}
      ]
     }]
   }]
 }
 SkeletalMeshComponent0 {
  id=#8
  parent_id=#9
  possessed_class=SkeletalMeshComponentBudgeted
  tracks [{
    type=MovieSceneControlRigParameterTrack
    display_name=Fortnite_Legacy_Face_CtrlRig
    sections [{
      type=MovieSceneControlRigParameterSection
      channels [
       Flipbook {type=MovieSceneScriptingBoolChannel, default=false, const [false,547]}
       Face_Ctrl_Master.Location.X {type=MovieSceneScriptingFloatChannel, default=0, const [0,575]}
       Face_Ctrl_Master.Location.Y {type=MovieSceneScriptingFloatChannel, default=0, const [0,575]}
       Face_Ctrl_Master.Location.Z {type=MovieSceneScriptingFloatChannel, default=0, const [0,575]}
       Face_Ctrl_Master.Rotation.X {type=MovieSceneScriptingFloatChannel, default=0, thinned_from=575, keys [[0,0],[574,0]]}
       Face_Ctrl_Master.Rotation.Y {type=MovieSceneScriptingFloatChannel, default=0, const [0,575]}
       Face_Ctrl_Master.Rotation.Z {type=MovieSceneScriptingFloatChannel, default=0, const [0,575]}
       Face_Ctrl_Master.Scale.X {type=MovieSceneScriptingFloatChannel, default=1, const [1,575]}
       Face_Ctrl_Master.Scale.Y {type=MovieSceneScriptingFloatChannel, default=1, const [1,575]}
       Face_Ctrl_Master.Scale.Z {type=MovieSceneScriptingFloatChannel, default=1, const [1,575]}
       gui_follow_head {type=MovieSceneScriptingFloatChannel, default=1, const [1,575]}
      ]
     }]
   }]
 }
 CameraActor {
  id=#5
  possessed_class=CameraActor
  tracks [{
    type=MovieScene3DTransformTrack
    sections [{
      type=MovieScene3DTransformSection
      channels [
       Location.X {type=MovieSceneScriptingDoubleChannel, default=0.0001}
       Location.Y {type=MovieSceneScriptingDoubleChannel, default=996}
       Location.Z {type=MovieSceneScriptingDoubleChannel, default=153}
       Rotation.X {type=MovieSceneScriptingDoubleChannel, default=0}
       Rotation.Y {type=MovieSceneScriptingDoubleChannel, default=0}
       Rotation.Z {type=MovieSceneScriptingDoubleChannel, default=-90.2}
       Scale.X {type=MovieSceneScriptingDoubleChannel, default=1}
       Scale.Y {type=MovieSceneScriptingDoubleChannel, default=1}
       Scale.Z {type=MovieSceneScriptingDoubleChannel, default=1}
      ]
     }]
   }]
 }
 SK_Mannequin {
  id=#7
  possessed_class=SoldierMannequin_Male_C
  tracks [
   {
    type=MovieScene3DTransformTrack
    sections [{
      type=MovieScene3DTransformSection
      channels [
       Location.X {type=MovieSceneScriptingDoubleChannel, default=-60}
       Location.Y {type=MovieSceneScriptingDoubleChannel, default=0}
       Location.Z {type=MovieSceneScriptingDoubleChannel, default=0}
       Rotation.X {type=MovieSceneScriptingDoubleChannel, default=0}
       Rotation.Y {type=MovieSceneScriptingDoubleChannel, default=0}
       Rotation.Z {type=MovieSceneScriptingDoubleChannel, default=0}
       Scale.X {type=MovieSceneScriptingDoubleChannel, default=1}
       Scale.Y {type=MovieSceneScriptingDoubleChannel, default=1}
       Scale.Z {type=MovieSceneScriptingDoubleChannel, default=1}
      ]
     }]
   }
   {type=MovieSceneSkeletalAnimationTrack, sections []}
  ]
 }
 JonesyExample {
  id=#9
  possessed_class=FortPlayerMannequin
  tracks [
   {
    type=MovieScene3DTransformTrack
    sections [{
      type=MovieScene3DTransformSection
      channels [
       Location.X {type=MovieSceneScriptingDoubleChannel, default=60}
       Location.Y {type=MovieSceneScriptingDoubleChannel, default=0}
       Location.Z {type=MovieSceneScriptingDoubleChannel, default=0}
       Rotation.X {type=MovieSceneScriptingDoubleChannel, default=0}
       Rotation.Y {type=MovieSceneScriptingDoubleChannel, default=0}
       Rotation.Z {type=MovieSceneScriptingDoubleChannel, default=0}
       Scale.X {type=MovieSceneScriptingDoubleChannel, default=1}
       Scale.Y {type=MovieSceneScriptingDoubleChannel, default=1}
       Scale.Z {type=MovieSceneScriptingDoubleChannel, default=1}
      ]
     }]
   }
   {type=MovieSceneSkeletalAnimationTrack, sections []}
   {type=MovieSceneSpawnTrack, sections [{type=MovieSceneSpawnSection, channels [None {type=MovieSceneScriptingBoolChannel, default=true}]}]}
  ]
 }
 Turntable_UserObjects {
  id=#10
  possessed_class=Actor
  tracks [{
    type=MovieScene3DTransformTrack
    sections [{
      type=MovieScene3DTransformSection
      channels [
       Location.X {type=MovieSceneScriptingDoubleChannel, default=60}
       Location.Y {type=MovieSceneScriptingDoubleChannel, default=0}
       Location.Z {type=MovieSceneScriptingDoubleChannel, default=0}
       Rotation.Z {type=MovieSceneScriptingDoubleChannel, default=3240, const [0,1]}
      ]
     }]
   }]
 }
 ReferenceMann {
  id=#3
  possessed_class=SoldierMannequin_Male_C
  tracks [
   {
    type=MovieScene3DTransformTrack
    sections [{
      type=MovieScene3DTransformSection
      channels [
       Location.X {type=MovieSceneScriptingDoubleChannel, default=-60}
       Location.Y {type=MovieSceneScriptingDoubleChannel, default=0}
       Location.Z {type=MovieSceneScriptingDoubleChannel, default=0}
       Rotation.X {type=MovieSceneScriptingDoubleChannel, default=0}
       Rotation.Y {type=MovieSceneScriptingDoubleChannel, default=0}
       Rotation.Z {type=MovieSceneScriptingDoubleChannel, default=0}
       Scale.X {type=MovieSceneScriptingDoubleChannel, default=1}
       Scale.Y {type=MovieSceneScriptingDoubleChannel, default=1}
       Scale.Z {type=MovieSceneScriptingDoubleChannel, default=1}
      ]
     }]
   }
   {type=MovieSceneSpawnTrack, sections [{type=MovieSceneSpawnSection, channels [None {type=MovieSceneScriptingBoolChannel, default=true}]}]}
  ]
 }
]
display_rate {denominator=1, numerator=30}
tick_resolution {denominator=1, numerator=24000}
tracks [
 {
  type=MovieSceneLevelVisibilityTrack
  sections [
   {type=MovieSceneLevelVisibilitySection, end_frame=574, start_frame=-17}
   {type=MovieSceneLevelVisibilitySection, end_frame=574, start_frame=-17}
  ]
 }
 {
  type=MovieSceneEventTrack
  sections [{
    type=MovieSceneEventTriggerSection
    channels [None {type=MovieSceneScriptingEventChannel, const ["<Struct 'MovieSceneEvent' (0x49e7bfb9740) {}>", 1]}]
   }]
 }
 {type=MovieSceneEventTrack, sections [{type=MovieSceneEventRepeaterSection, end_frame=574, start_frame=0}]}
 {type=MovieSceneCameraCutTrack, sections [{type=MovieSceneCameraCutSection, end_frame=574, start_frame=0}]}
]
