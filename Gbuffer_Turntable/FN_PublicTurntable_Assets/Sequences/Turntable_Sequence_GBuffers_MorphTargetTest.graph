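# Level Sequence for a G-buffer turntable, named as a morph target test, with a camera cut track, a CameraActor and level visibility sections.
# Turntable_UserObjects has Z rotation keyed through 360 degrees over the first 30 frames at 30 fps.
# A master event track fires triggers every 30 frames from 0 to 300 alongside a repeater section, and BP_Turntable_GBuffers has its own event trigger.
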LevelSequence /Gbuffer_Turntable/FN_PublicTurntable_Assets/Sequences/Turntable_Sequence_GBuffers_MorphTargetTest
bindings [
 CameraActor {id=#1, possessed_class=CameraActor, tracks []}
 BP_Turntable_GBuffers {
  id=#2
  possessed_class=BP_Turntable_GBuffers_C
  tracks [{
    type=MovieSceneEventTrack
    sections [{
      type=MovieSceneEventTriggerSection
      channels [None {type=MovieSceneScriptingEventChannel, const ["<Struct 'MovieSceneEvent' (0x49e7bfb9880) {}>", 1]}]
     }]
   }]
 }
 Turntable_UserObjects {
  id=#3
  possessed_class=Actor
  tracks [{
    type=MovieScene3DTransformTrack
    sections [{
      type=MovieScene3DTransformSection
      channels [
       Location.X {type=MovieSceneScriptingDoubleChannel, default=0}
       Location.Y {type=MovieSceneScriptingDoubleChannel, default=0}
       Location.Z {type=MovieSceneScriptingDoubleChannel, default=0}
       Rotation.X {type=MovieSceneScriptingDoubleChannel, default=0}
       Rotation.Y {type=MovieSceneScriptingDoubleChannel, default=0}
       Rotation.Z {type=MovieSceneScriptingDoubleChannel, default=360, keys [[0,0],[30,360]]}
       Scale.X {type=MovieSceneScriptingDoubleChannel, default=1}
       Scale.Y {type=MovieSceneScriptingDoubleChannel, default=1}
       Scale.Z {type=MovieSceneScriptingDoubleChannel, default=1}
      ]
     }]
   }]
 }
]
display_rate {denominator=1, numerator=30}
tick_resolution {denominator=1, numerator=24000}
tracks [
 {
  type=MovieSceneLevelVisibilityTrack
  sections [
   {type=MovieSceneLevelVisibilitySection, end_frame=428, start_frame=-17}
   {type=MovieSceneLevelVisibilitySection, end_frame=428, start_frame=-17}
  ]
 }
 {
  type=MovieSceneEventTrack
  sections [
   {
    type=MovieSceneEventTriggerSection
    channels [None {type=MovieSceneScriptingEventChannel, const ["<Struct 'MovieSceneEvent' (0x49e7bfba0a0) {}>", 1]}]
   }
   {
    type=MovieSceneEventTriggerSection
    channels [None {
      type=MovieSceneScriptingEventChannel
      keys [
       [0, "<Struct 'MovieSceneEvent' (0x49e7bfba780) {}>"]
       [30, "<Struct 'MovieSceneEvent' (0x49e7bfba8c0) {}>"]
       [60, "<Struct 'MovieSceneEvent' (0x49e7bfba960) {}>"]
       [90, "<Struct 'MovieSceneEvent' (0x49e7bfb9740) {}>"]
       [120, "<Struct 'MovieSceneEvent' (0x49e7bfbd160) {}>"]
       [150, "<Struct 'MovieSceneEvent' (0x49e7bfbd0c0) {}>"]
       [180, "<Struct 'MovieSceneEvent' (0x49e7bfbcf80) {}>"]
       [210, "<Struct 'MovieSceneEvent' (0x49e7bfbd020) {}>"]
       [240, "<Struct 'MovieSceneEvent' (0x49e7bfba6e0) {}>"]
       [270, "<Struct 'MovieSceneEvent' (0x49e7bfba320) {}>"]
       [300, "<Struct 'MovieSceneEvent' (0x49e7bfba1e0) {}>"]
      ]
     }]
   }
   {type=MovieSceneEventRepeaterSection, end_frame=391, start_frame=0}
  ]
 }
 {type=MovieSceneCameraCutTrack, sections [{type=MovieSceneCameraCutSection, end_frame=391, start_frame=-10}]}
]
