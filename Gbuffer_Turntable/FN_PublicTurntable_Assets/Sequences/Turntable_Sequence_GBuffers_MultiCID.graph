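# Level Sequence for a G-buffer turntable capture, staging a soldier mannequin beside a user-object actor before a fixed, narrow-FOV camera.
# Event triggers on the BP_Turntable_GBuffers binding (ManifestTracks) and root event tracks, including SampleKey, Generated and a repeater, hand control to scripted logic.
# Level visibility sections span the roughly thousand-frame range, while the camera cut track splits its opening frames into consecutive two-frame sections.
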LevelSequence /Gbuffer_Turntable/FN_PublicTurntable_Assets/Sequences/Turntable_Sequence_GBuffers_MultiCID
bindings [
 BP_Turntable_GBuffers {
  id=#1
  possessed_class=BP_Turntable_GBuffers_C
  tracks [{
    type=MovieSceneEventTrack
    display_name=ManifestTracks
    sections [{
      type=MovieSceneEventTriggerSection
      channels [None {
        type=MovieSceneScriptingEventChannel
        keys [[0, "<Struct 'MovieSceneEvent' (0x49e7bfb9f60) {}>"], [1.0063, "<Struct 'MovieSceneEvent' (0x49e7bfb9d80) {}>"]]
       }]
     }]
   }]
 }
 CameraComponent {
  id=#2
  parent_id=#3
  possessed_class=CameraComponent
  tracks [{
    type=MovieSceneFloatTrack
    display_name=Field Of View
    sections [{type=MovieSceneFloatSection, channels [None {type=MovieSceneScriptingFloatChannel, default=20}]}]
   }]
 }
 CameraActor {
  id=#3
  possessed_class=CameraActor
  tracks [{
    type=MovieScene3DTransformTrack
    sections [{
      type=MovieScene3DTransformSection
      channels [
       Location.X {type=MovieSceneScriptingDoubleChannel, default=0.0001}
       Location.Y {type=MovieSceneScriptingDoubleChannel, default=996}
       Location.Z {type=MovieSceneScriptingDoubleChannel, default=153}
       Rotation.X {type=MovieSceneScriptingDoubleChannel, default=0}
       Rotation.Y {type=MovieSceneScriptingDoubleChannel, default=-3.5}
       Rotation.Z {type=MovieSceneScriptingDoubleChannel, default=-90}
       Scale.X {type=MovieSceneScriptingDoubleChannel, default=1}
       Scale.Y {type=MovieSceneScriptingDoubleChannel, default=1}
       Scale.Z {type=MovieSceneScriptingDoubleChannel, default=1}
      ]
     }]
   }]
 }
 SK_Mannequin {
  id=#4
  possessed_class=SoldierMannequin_Male_C
  tracks [
   {
    type=MovieScene3DTransformTrack
    sections [{
      type=MovieScene3DTransformSection
      channels [
       Location.X {type=MovieSceneScriptingDoubleChannel, default=-60}
       Location.Y {type=MovieSceneScriptingDoubleChannel, default=0}
       Location.Z {type=MovieSceneScriptingDoubleChannel, default=0}
       Rotation.X {type=MovieSceneScriptingDoubleChannel, default=0}
       Rotation.Y {type=MovieSceneScriptingDoubleChannel, default=0}
       Rotation.Z {type=MovieSceneScriptingDoubleChannel, default=0}
       Scale.X {type=MovieSceneScriptingDoubleChannel, default=1}
       Scale.Y {type=MovieSceneScriptingDoubleChannel, default=1}
       Scale.Z {type=MovieSceneScriptingDoubleChannel, default=1}
      ]
     }]
   }
   {type=MovieSceneSkeletalAnimationTrack, sections []}
  ]
 }
 Turntable_UserObjects {
  id=#5
  possessed_class=Actor
  tracks [{
    type=MovieScene3DTransformTrack
    sections [{
      type=MovieScene3DTransformSection
      channels [
       Location.X {type=MovieSceneScriptingDoubleChannel, default=60}
       Location.Y {type=MovieSceneScriptingDoubleChannel, default=0}
       Location.Z {type=MovieSceneScriptingDoubleChannel, default=0}
       Rotation.Z {type=MovieSceneScriptingDoubleChannel, default=3240, const [0,1]}
      ]
     }]
   }]
 }
]
display_rate {denominator=1, numerator=30}
tick_resolution {denominator=1, numerator=24000}
tracks [
 {
  type=MovieSceneLevelVisibilityTrack
  sections [
   {type=MovieSceneLevelVisibilitySection, end_frame=1036, start_frame=-17}
   {type=MovieSceneLevelVisibilitySection, end_frame=1036, start_frame=-17}
  ]
 }
 {
  type=MovieSceneEventTrack
  sections [{
    type=MovieSceneEventTriggerSection
    channels [None {type=MovieSceneScriptingEventChannel, const ["<Struct 'MovieSceneEvent' (0x49e7bfb9560) {}>", 1]}]
   }]
 }
 {type=MovieSceneEventTrack, sections [{type=MovieSceneEventRepeaterSection, end_frame=1036, start_frame=0}]}
 {
  type=MovieSceneEventTrack
  display_name=SampleKey
  sections [{
    type=MovieSceneEventTriggerSection
    channels [None {type=MovieSceneScriptingEventChannel, const ["<Struct 'MovieSceneEvent' (0x49e7bfbd520) {}>", 1]}]
   }]
 }
 {
  type=MovieSceneEventTrack
  display_name=Generated
  sections [{
    type=MovieSceneEventTriggerSection
    channels [None {
      type=MovieSceneScriptingEventChannel
      keys [
       [0, "<Struct 'MovieSceneEvent' (0x49e7bfbda20) {}>"]
       [1, "<Struct 'MovieSceneEvent' (0x49e7bfbd7a0) {}>"]
       [2, "<Struct 'MovieSceneEvent' (0x49e7bfbd5c0) {}>"]
       [3, "<Struct 'MovieSceneEvent' (0x49e7bfbd3e0) {}>"]
       [4, "<Struct 'MovieSceneEvent' (0x49e7bfbd200) {}>"]
       [5, "<Struct 'MovieSceneEvent' (0x49e7bfbd2a0) {}>"]
       [6, "<Struct 'MovieSceneEvent' (0x49e7bfbd980) {}>"]
      ]
     }]
   }]
 }
 {
  type=MovieSceneCameraCutTrack
  sections [
   {type=MovieSceneCameraCutSection, end_frame=2, start_frame=0}
   {type=MovieSceneCameraCutSection, end_frame=4, start_frame=2}
   {type=MovieSceneCameraCutSection, end_frame=6, start_frame=4}
   {type=MovieSceneCameraCutSection, end_frame=8, start_frame=6}
   {type=MovieSceneCameraCutSection, end_frame=10, start_frame=8}
   {type=MovieSceneCameraCutSection, end_frame=12, start_frame=10}
   {type=MovieSceneCameraCutSection, end_frame=14, start_frame=12}
   {type=MovieSceneCameraCutSection, end_frame=16, start_frame=14}
   {type=MovieSceneCameraCutSection, end_frame=18, start_frame=16}
   {type=MovieSceneCameraCutSection, end_frame=20, start_frame=18}
   {type=MovieSceneCameraCutSection, end_frame=22, start_frame=20}
   {type=MovieSceneCameraCutSection, end_frame=24, start_frame=22}
   {type=MovieSceneCameraCutSection, end_frame=26, start_frame=24}
   {type=MovieSceneCameraCutSection, end_frame=28, start_frame=26}
   {type=MovieSceneCameraCutSection, end_frame=30, start_frame=28}
   {type=MovieSceneCameraCutSection, end_frame=32, start_frame=30}
   {type=MovieSceneCameraCutSection, end_frame=34, start_frame=32}
   {type=MovieSceneCameraCutSection, end_frame=36, start_frame=34}
  ]
 }
]
