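# Level Sequence for a GBuffer turntable that spins the Turntable_UserObjects actor under a fixed-FOV camera cut.
# The object makes one slow full turn, then roughly one turn per 30-frame segment from frame 180.
# Event keys on BP_Turntable_GBuffers and BP_LevelsChecker fire at segment boundaries.
# Level visibility sections switch at frames 330 and 390, with lowered exposure compensation between them.
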LevelSequence /Gbuffer_Turntable/FN_PublicTurntable_Assets/Sequences/Turntable_Sequence_GBuffers_Standard
bindings [
 BP_Turntable_GBuffers {
  id=#1
  possessed_class=BP_Turntable_GBuffers_C
  tracks [{
    type=MovieSceneEventTrack
    display_name=ManifestTracks
    sections [
     {
      type=MovieSceneEventTriggerSection
      channels [None {
        type=MovieSceneScriptingEventChannel
        keys [
         [0.0063, "<Struct 'MovieSceneEvent' (0x49e7bfb99c0) {}>"]
         [180, "<Struct 'MovieSceneEvent' (0x49e7bfb9b00) {}>"]
         [210, "<Struct 'MovieSceneEvent' (0x49e7bfb9a60) {}>"]
         [240, "<Struct 'MovieSceneEvent' (0x49e7bfb92e0) {}>"]
         [270, "<Struct 'MovieSceneEvent' (0x49e7bfb9f60) {}>"]
         [300, "<Struct 'MovieSceneEvent' (0x49e7bfb9ba0) {}>"]
         [330, "<Struct 'MovieSceneEvent' (0x49e7bfbd8e0) {}>"]
         [361, "<Struct 'MovieSceneEvent' (0x49e7bfbd660) {}>"]
         [390, "<Struct 'MovieSceneEvent' (0x49e7bfbd5c0) {}>"]
         [421, "<Struct 'MovieSceneEvent' (0x49e7bfbd7a0) {}>"]
        ]
       }]
     }
     {
      type=MovieSceneEventTriggerSection
      channels [None {
        type=MovieSceneScriptingEventChannel
        keys [
         [-1, "<Struct 'MovieSceneEvent' (0x49e7bfb8ac0) {}>"]
         [0.0063, "<Struct 'MovieSceneEvent' (0x49e7bfb9100) {}>"]
         [330, "<Struct 'MovieSceneEvent' (0x49e7bfb9c40) {}>"]
         [361, "<Struct 'MovieSceneEvent' (0x49e7bfb9e20) {}>"]
         [390, "<Struct 'MovieSceneEvent' (0x49e7bfba140) {}>"]
        ]
       }]
     }
    ]
   }]
 }
 CameraActor {id=#2, possessed_class=CameraActor, tracks []}
 Turntable_UserObjects {
  id=#3
  possessed_class=Actor
  tracks [{
    type=MovieScene3DTransformTrack
    sections [{
      type=MovieScene3DTransformSection
      channels [
       Location.X {type=MovieSceneScriptingDoubleChannel, default=0}
       Location.Y {type=MovieSceneScriptingDoubleChannel, default=0}
       Location.Z {type=MovieSceneScriptingDoubleChannel, default=0}
       Rotation.Z {type=MovieSceneScriptingDoubleChannel, default=3240, thinned_from=10, keys [[0,0],[180,360],[330,2160],[361,2520],[390,2880],[420,3240]]}
      ]
     }]
   }]
 }
 CameraComponent {
  id=#4
  parent_id=#2
  possessed_class=CameraComponent
  tracks [
   {
    type=MovieSceneFloatTrack
    display_name=Field Of View
    sections [{type=MovieSceneFloatSection, channels [None {type=MovieSceneScriptingFloatChannel, default=35}]}]
   }
   {
    type=MovieSceneFloatTrack
    display_name="Exposure Compensation (Post Process Settings)"
    sections [{
      type=MovieSceneFloatSection
      channels [None {type=MovieSceneScriptingFloatChannel, default=-0.8, keys [[328.0325,0],[330,-0.8],[389,-0.8],[390,0]]}]
     }]
   }
  ]
 }
 BP_LevelsChecker {
  id=#5
  possessed_class=BP_LevelsChecker_C
  tracks [{
    type=MovieSceneEventTrack
    sections [{
      type=MovieSceneEventTriggerSection
      channels [None {
        type=MovieSceneScriptingEventChannel
        keys [[180, "<Struct 'MovieSceneEvent' (0x49e7bfbcf80) {}>"], [210, "<Struct 'MovieSceneEvent' (0x49e7bfb9740) {}>"]]
       }]
     }]
   }]
 }
]
display_rate {denominator=1, numerator=30}
tick_resolution {denominator=1, numerator=24000}
tracks [
 {
  type=MovieSceneLevelVisibilityTrack
  sections [
   {type=MovieSceneLevelVisibilitySection, end_frame=390, start_frame=330}
   {type=MovieSceneLevelVisibilitySection, end_frame=330, start_frame=-17}
   {type=MovieSceneLevelVisibilitySection, end_frame=390, start_frame=330}
   {type=MovieSceneLevelVisibilitySection, end_frame=330, start_frame=-17}
   {type=MovieSceneLevelVisibilitySection, end_frame=426, start_frame=390}
   {type=MovieSceneLevelVisibilitySection, end_frame=426, start_frame=390}
  ]
 }
 {
  type=MovieSceneEventTrack
  sections [{
    type=MovieSceneEventTriggerSection
    channels [None {type=MovieSceneScriptingEventChannel, const ["<Struct 'MovieSceneEvent' (0x49e7bfb9240) {}>", 1]}]
   }]
 }
 {type=MovieSceneEventTrack, sections [{type=MovieSceneEventRepeaterSection, end_frame=420, start_frame=0}]}
 {type=MovieSceneCameraCutTrack, sections [{type=MovieSceneCameraCutSection, end_frame=421, start_frame=0}]}
]
