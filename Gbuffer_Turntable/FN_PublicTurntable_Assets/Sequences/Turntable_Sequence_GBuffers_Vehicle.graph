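# Level Sequence for a vehicle G-buffer turntable capture at 30 fps, with a camera cut and animated user objects, camera root and color calibrator.
# The objects turn and tilt through frame 230, then spin once per 30-frame segment until 440.
# Events on BP_Turntable_GBuffers mark those segments, with BP_LevelsChecker triggered at the first two.
# Level visibility switches at 380 and 410, with exposure compensation lowered between them.
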LevelSequence /Gbuffer_Turntable/FN_PublicTurntable_Assets/Sequences/Turntable_Sequence_GBuffers_Vehicle
bindings [
 CameraComponent {
  id=#1
  parent_id=#2
  possessed_class=CameraComponent
  tracks [
   {
    type=MovieSceneFloatTrack
    display_name=Field Of View
    sections [{type=MovieSceneFloatSection, channels [None {type=MovieSceneScriptingFloatChannel, default=35}]}]
   }
   {
    type=MovieSceneFloatTrack
    display_name="Exposure Compensation (Post Process Settings)"
    sections [{
      type=MovieSceneFloatSection
      channels [None {type=MovieSceneScriptingFloatChannel, default=-0.8, keys [[379,0],[380,-0.8],[409,-0.8],[410,0]]}]
     }]
   }
  ]
 }
 CameraActor {id=#2, possessed_class=CameraActor, tracks []}
 CamRoot {
  id=#3
  possessed_class=FortStaticMeshActor
  tracks [{
    type=MovieScene3DTransformTrack
    sections [{
      type=MovieScene3DTransformSection
      channels [
       Location.X {type=MovieSceneScriptingDoubleChannel, default=0, const [0,3]}
       Location.Y {type=MovieSceneScriptingDoubleChannel, default=0, keys [[0,0],[180,76],[230,0]]}
       Location.Z {type=MovieSceneScriptingDoubleChannel, default=0, keys [[0,0],[180,24],[230,0]]}
      ]
     }]
   }]
 }
 Turntable_UserObjects {
  id=#4
  possessed_class=Actor
  tracks [{
    type=MovieScene3DTransformTrack
    sections [
     {
      type=MovieScene3DTransformSection
      channels [
       Location.X {type=MovieSceneScriptingDoubleChannel, default=0}
       Location.Y {type=MovieSceneScriptingDoubleChannel, default=24, keys [[0,24],[207,101],[230,24]]}
       Location.Z {type=MovieSceneScriptingDoubleChannel, default=86, keys [[0,0],[180,86],[230,0]]}
       Rotation.Z {type=MovieSceneScriptingDoubleChannel, default=360, thinned_from=10, keys [[1,0],[180,360],[230,360],[440,2880]]}
       Scale.X {type=MovieSceneScriptingDoubleChannel, default=0.5}
       Scale.Y {type=MovieSceneScriptingDoubleChannel, default=0.5}
       Scale.Z {type=MovieSceneScriptingDoubleChannel, default=0.5}
      ]
     }
     {
      type=MovieScene3DTransformSection
      channels [Rotation.X {type=MovieSceneScriptingDoubleChannel, default=-90, keys [[0,0],[180,90],[207,-90],[230,0]]}]
     }
    ]
   }]
 }
 BP_LevelsChecker {
  id=#5
  possessed_class=BP_LevelsChecker_C
  tracks [{
    type=MovieSceneEventTrack
    sections [{
      type=MovieSceneEventTriggerSection
      channels [None {
        type=MovieSceneScriptingEventChannel
        keys [[230, "<Struct 'MovieSceneEvent' (0x49e7bfb9740) {}>"], [260, "<Struct 'MovieSceneEvent' (0x49e7bfbd0c0) {}>"]]
       }]
     }]
   }]
 }
 SM_ColorCalibrator {
  id=#6
  possessed_class=StaticMeshActor
  tracks [{
    type=MovieScene3DTransformTrack
    sections [{
      type=MovieScene3DTransformSection
      channels [
       Location.X {type=MovieSceneScriptingDoubleChannel, default=105.2872, keys [[0,105.2872],[180,129.2872],[230,105.2872]]}
       Location.Y {type=MovieSceneScriptingDoubleChannel, default=163.7846, const [163.7846,3]}
       Location.Z {type=MovieSceneScriptingDoubleChannel, default=10, const [10,3]}
       Rotation.X {type=MovieSceneScriptingDoubleChannel, default=0}
       Rotation.Y {type=MovieSceneScriptingDoubleChannel, default=0}
       Rotation.Z {type=MovieSceneScriptingDoubleChannel, default=-30}
       Scale.X {type=MovieSceneScriptingDoubleChannel, default=0.25, const [0.25,1]}
       Scale.Y {type=MovieSceneScriptingDoubleChannel, default=0.25, const [0.25,1]}
       Scale.Z {type=MovieSceneScriptingDoubleChannel, default=0.25, const [0.25,1]}
      ]
     }]
   }]
 }
 BP_Turntable_GBuffers {
  id=#7
  possessed_class=BP_Turntable_GBuffers_C
  tracks [{
    type=MovieSceneEventTrack
    sections [
     {
      type=MovieSceneEventTriggerSection
      channels [None {
        type=MovieSceneScriptingEventChannel
        keys [
         [0, "<Struct 'MovieSceneEvent' (0x49e7bfb9380) {}>"]
         [230, "<Struct 'MovieSceneEvent' (0x49e7bfb9100) {}>"]
         [260, "<Struct 'MovieSceneEvent' (0x49e7bfb8ac0) {}>"]
         [290, "<Struct 'MovieSceneEvent' (0x49e7bfbd7a0) {}>"]
         [320, "<Struct 'MovieSceneEvent' (0x49e7bfba1e0) {}>"]
         [350, "<Struct 'MovieSceneEvent' (0x49e7bfbd700) {}>"]
         [380, "<Struct 'MovieSceneEvent' (0x49e7bfbd200) {}>"]
         [410, "<Struct 'MovieSceneEvent' (0x49e7bfb9b00) {}>"]
         [440, "<Struct 'MovieSceneEvent' (0x49e7bfb9060) {}>"]
        ]
       }]
     }
     {
      type=MovieSceneEventTriggerSection
      channels [None {
        type=MovieSceneScriptingEventChannel
        keys [
         [-1, "<Struct 'MovieSceneEvent' (0x49e7bfbd660) {}>"]
         [0, "<Struct 'MovieSceneEvent' (0x49e7bfb9ba0) {}>"]
         [410, "<Struct 'MovieSceneEvent' (0x49e7bfbd840) {}>"]
        ]
       }]
     }
     {type=MovieSceneEventRepeaterSection, end_frame=410, start_frame=380}
    ]
   }]
 }
]
display_rate {denominator=1, numerator=30}
tick_resolution {denominator=1, numerator=24000}
tracks [
 {
  type=MovieSceneLevelVisibilityTrack
  sections [
   {type=MovieSceneLevelVisibilitySection, end_frame=410, start_frame=380}
   {type=MovieSceneLevelVisibilitySection, end_frame=380, start_frame=0}
   {type=MovieSceneLevelVisibilitySection, end_frame=410, start_frame=380}
   {type=MovieSceneLevelVisibilitySection, end_frame=380, start_frame=0}
   {type=MovieSceneLevelVisibilitySection, end_frame=446, start_frame=410}
   {type=MovieSceneLevelVisibilitySection, end_frame=446, start_frame=410}
  ]
 }
 {
  type=MovieSceneEventTrack
  sections [{
    type=MovieSceneEventTriggerSection
    channels [None {type=MovieSceneScriptingEventChannel, const ["<Struct 'MovieSceneEvent' (0x49e7bfbd8e0) {}>", 1]}]
   }]
 }
 {type=MovieSceneEventTrack, sections [{type=MovieSceneEventRepeaterSection, end_frame=391, start_frame=0}]}
 {type=MovieSceneCameraCutTrack, sections [{type=MovieSceneCameraCutSection, end_frame=440, start_frame=1}]}
]
