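# Level Sequence for a toon-variant G-buffer turntable presentation, running about 15 seconds at 30 fps with a camera cut.
# Turntable_UserObjects spins around Z through several full turns at varying speeds, easing back slightly at the end.
# Timed event keys drive BP_LevelsChecker and BP_Turntable_GBuffers at boundaries that largely match the rotation keys, alongside sequence-level trigger and repeater events.
# Level visibility sections span the run, and the camera holds a fixed exposure compensation.
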LevelSequence /Gbuffer_Turntable/FN_PublicTurntable_Assets/Sequences/Turntable_Sequence_Gbuffers_Toon
bindings [
 BP_LevelsChecker {
  id=#1
  possessed_class=BP_LevelsChecker_C
  tracks [{
    type=MovieSceneEventTrack
    sections [{
      type=MovieSceneEventTriggerSection
      channels [None {
        type=MovieSceneScriptingEventChannel
        keys [
         [0, "<Struct 'MovieSceneEvent' (0x49e7bfbcf80) {}>"]
         [180, "<Struct 'MovieSceneEvent' (0x49e7bfbd3e0) {}>"]
         [210, "<Struct 'MovieSceneEvent' (0x49e7bfb9060) {}>"]
        ]
       }]
     }]
   }]
 }
 CameraActor {id=#2, possessed_class=CameraActor, tracks []}
 BP_Turntable_GBuffers {
  id=#3
  possessed_class=BP_Turntable_GBuffers_C
  tracks [{
    type=MovieSceneEventTrack
    sections [
     {
      type=MovieSceneEventTriggerSection
      channels [None {
        type=MovieSceneScriptingEventChannel
        keys [
         [-4, "<Struct 'MovieSceneEvent' (0x49e7bfba140) {}>"]
         [0, "<Struct 'MovieSceneEvent' (0x49e7bfbd480) {}>"]
         [180, "<Struct 'MovieSceneEvent' (0x49e7bfba320) {}>"]
         [210, "<Struct 'MovieSceneEvent' (0x49e7bfbd840) {}>"]
         [240, "<Struct 'MovieSceneEvent' (0x49e7bfb92e0) {}>"]
         [270, "<Struct 'MovieSceneEvent' (0x49e7bfbd660) {}>"]
         [300, "<Struct 'MovieSceneEvent' (0x49e7bfb9b00) {}>"]
         [330, "<Struct 'MovieSceneEvent' (0x49e7bfbd700) {}>"]
         [420, "<Struct 'MovieSceneEvent' (0x49e7bfbd200) {}>"]
        ]
       }]
     }
     {
      type=MovieSceneEventTriggerSection
      channels [None {type=MovieSceneScriptingEventChannel, const ["<Struct 'MovieSceneEvent' (0x49e7bfb9100) {}>", 1]}]
     }
    ]
   }]
 }
 CameraComponent {
  id=#4
  parent_id=#2
  possessed_class=CameraComponent
  tracks [{
    type=MovieSceneFloatTrack
    display_name="Exposure Compensation (Post Process Settings)"
    sections [{type=MovieSceneFloatSection, channels [None {type=MovieSceneScriptingFloatChannel, default=-0.8, const [-0.8,1]}]}]
   }]
 }
 Turntable_UserObjects {
  id=#5
  possessed_class=Actor
  tracks [{
    type=MovieScene3DTransformTrack
    sections [{
      type=MovieScene3DTransformSection
      channels [
       Location.X {type=MovieSceneScriptingDoubleChannel, default=0}
       Location.Y {type=MovieSceneScriptingDoubleChannel, default=0}
       Location.Z {type=MovieSceneScriptingDoubleChannel, default=0}
       Rotation.Z {type=MovieSceneScriptingDoubleChannel, default=360, thinned_from=9, keys [[0,0],[180,360],[210,1080],[330,2520],[420,2880],[450,2790]]}
      ]
     }]
   }]
 }
]
display_rate {denominator=1, numerator=30}
tick_resolution {denominator=1, numerator=24000}
tracks [
 {
  type=MovieSceneEventTrack
  sections [{
    type=MovieSceneEventTriggerSection
    channels [None {type=MovieSceneScriptingEventChannel, const ["<Struct 'MovieSceneEvent' (0x49e7bfbd8e0) {}>", 1]}]
   }]
 }
 {type=MovieSceneEventTrack, sections [{type=MovieSceneEventRepeaterSection, end_frame=450, start_frame=0}]}
 {
  type=MovieSceneLevelVisibilityTrack
  sections [
   {type=MovieSceneLevelVisibilitySection, end_frame=454, start_frame=-4}
   {type=MovieSceneLevelVisibilitySection, end_frame=459, start_frame=-8}
  ]
 }
 {type=MovieSceneCameraCutTrack, sections [{type=MovieSceneCameraCutSection, end_frame=450, start_frame=-30}]}
]
