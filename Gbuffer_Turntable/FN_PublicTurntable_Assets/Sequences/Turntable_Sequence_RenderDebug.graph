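# Level Sequence for a G-buffer turntable debug render, stepping through four one-frame setups at 30 fps under a camera cut.
# Keyed visibility and weight switch on one Debug_Frame post-process volume per frame, while paired level visibility sections change alongside.
# A white PointLight shows only on frame 0, and Turntable_UserObjects has a constant visibility value.
# ScalabilitySphere holds a fixed transform and carries an event trigger, alongside a sequence-level event trigger.
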LevelSequence /Gbuffer_Turntable/FN_PublicTurntable_Assets/Sequences/Turntable_Sequence_RenderDebug
bindings [
 Debug_Frame_0 {
  id=#1
  possessed_class=PostProcessVolume
  tracks [
   {
    type=MovieSceneVisibilityTrack
    sections [{type=MovieSceneVisibilitySection, channels [None {type=MovieSceneScriptingBoolChannel, default=false, keys [[0,true],[1,false]]}]}]
   }
   {
    type=MovieSceneFloatTrack
    display_name="Weight (Array[0])"
    sections [{type=MovieSceneFloatSection, channels [None {type=MovieSceneScriptingFloatChannel, default=0, keys [[0,1],[1,0]]}]}]
   }
  ]
 }
 Debug_Frame_3 {
  id=#2
  possessed_class=PostProcessVolume
  tracks [
   {
    type=MovieSceneVisibilityTrack
    sections [{type=MovieSceneVisibilitySection, channels [None {type=MovieSceneScriptingBoolChannel, default=true, keys [[0,false],[3,true]]}]}]
   }
   {
    type=MovieSceneFloatTrack
    display_name="Weight (Array[0])"
    sections [{type=MovieSceneFloatSection, channels [None {type=MovieSceneScriptingFloatChannel, default=0, keys [[0,0],[2,0],[3,1]]}]}]
   }
  ]
 }
 Debug_Frame_2 {
  id=#3
  possessed_class=PostProcessVolume
  tracks [
   {
    type=MovieSceneVisibilityTrack
    sections [{type=MovieSceneVisibilitySection, channels [None {type=MovieSceneScriptingBoolChannel, default=false, keys [[0,false],[2,true],[3,false]]}]}]
   }
   {
    type=MovieSceneFloatTrack
    display_name="Weight (Array[0])"
    sections [{type=MovieSceneFloatSection, channels [None {type=MovieSceneScriptingFloatChannel, default=0, keys [[0,0],[1,0],[2,1],[3,0]]}]}]
   }
  ]
 }
 CameraActor {id=#4, possessed_class=CameraActor, tracks []}
 Debug_Frame_1 {
  id=#5
  possessed_class=PostProcessVolume
  tracks [
   {
    type=MovieSceneVisibilityTrack
    sections [{type=MovieSceneVisibilitySection, channels [None {type=MovieSceneScriptingBoolChannel, default=false, keys [[0,false],[1,true],[2,false]]}]}]
   }
   {
    type=MovieSceneFloatTrack
    display_name="Weight (Array[0])"
    sections [{type=MovieSceneFloatSection, channels [None {type=MovieSceneScriptingFloatChannel, default=0, keys [[0,0],[1,1],[2,0]]}]}]
   }
  ]
 }
 PointLight {
  id=#6
  possessed_class=PointLight
  tracks [{
    type=MovieSceneVisibilityTrack
    sections [{type=MovieSceneVisibilitySection, channels [None {type=MovieSceneScriptingBoolChannel, default=false, keys [[0,true],[1,false]]}]}]
   }]
 }
 ScalabilitySphere {
  id=#7
  possessed_class=StaticMeshActor
  tracks [
   {
    type=MovieScene3DTransformTrack
    sections [{
      type=MovieScene3DTransformSection
      channels [
       Location.X {type=MovieSceneScriptingDoubleChannel, default=110}
       Location.Y {type=MovieSceneScriptingDoubleChannel, default=0}
       Location.Z {type=MovieSceneScriptingDoubleChannel, default=140}
       Rotation.X {type=MovieSceneScriptingDoubleChannel, default=0}
       Rotation.Y {type=MovieSceneScriptingDoubleChannel, default=0}
       Rotation.Z {type=MovieSceneScriptingDoubleChannel, default=0}
       Scale.X {type=MovieSceneScriptingDoubleChannel, default=0.25}
       Scale.Y {type=MovieSceneScriptingDoubleChannel, default=0.25}
       Scale.Z {type=MovieSceneScriptingDoubleChannel, default=0.25}
      ]
     }]
   }
   {
    type=MovieSceneEventTrack
    sections [{
      type=MovieSceneEventTriggerSection
      channels [None {type=MovieSceneScriptingEventChannel, const ["<Struct 'MovieSceneEvent' (0x49e7bfbd3e0) {}>", 1]}]
     }]
   }
  ]
 }
 Turntable_UserObjects {
  id=#8
  possessed_class=Actor
  tracks [{
    type=MovieSceneVisibilityTrack
    sections [{type=MovieSceneVisibilitySection, channels [None {type=MovieSceneScriptingBoolChannel, default=false}]}]
   }]
 }
 LightComponent0 {
  id=#9
  parent_id=#6
  possessed_class=PointLightComponent
  tracks [
   {
    type=MovieSceneFloatTrack
    display_name=Intensity
    sections [{type=MovieSceneFloatSection, channels [None {type=MovieSceneScriptingFloatChannel, default=0.6}]}]
   }
   {
    type=MovieSceneColorTrack
    display_name=Light Color
    sections [{
      type=MovieSceneColorSection
      channels [
       Color.R {type=MovieSceneScriptingFloatChannel, default=1}
       Color.G {type=MovieSceneScriptingFloatChannel, default=1}
       Color.B {type=MovieSceneScriptingFloatChannel, default=1}
       Color.A {type=MovieSceneScriptingFloatChannel, default=1}
      ]
     }]
   }
  ]
 }
]
display_rate {denominator=1, numerator=30}
tick_resolution {denominator=1, numerator=24000}
tracks [
 {
  type=MovieSceneLevelVisibilityTrack
  sections [
   {type=MovieSceneLevelVisibilitySection, end_frame=3, start_frame=2}
   {type=MovieSceneLevelVisibilitySection, end_frame=3, start_frame=2}
   {type=MovieSceneLevelVisibilitySection, end_frame=2, start_frame=1}
   {type=MovieSceneLevelVisibilitySection, end_frame=2, start_frame=1}
   {type=MovieSceneLevelVisibilitySection, end_frame=4, start_frame=3}
   {type=MovieSceneLevelVisibilitySection, end_frame=4, start_frame=3}
   {type=MovieSceneLevelVisibilitySection, end_frame=1, start_frame=0}
   {type=MovieSceneLevelVisibilitySection, end_frame=1, start_frame=0}
  ]
 }
 {
  type=MovieSceneEventTrack
  sections [{
    type=MovieSceneEventTriggerSection
    channels [None {type=MovieSceneScriptingEventChannel, const ["<Struct 'MovieSceneEvent' (0x49e7bfbdf20) {}>", 1]}]
   }]
 }
 {type=MovieSceneCameraCutTrack, sections [{type=MovieSceneCameraCutSection, end_frame=4, start_frame=-1}]}
]
