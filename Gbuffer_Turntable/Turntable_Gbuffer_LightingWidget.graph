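# Editor utility widget that controls a G-buffer turntable for reviewing and rendering characters, vehicles and other meshes.
# Its controls switch streamed lighting levels and time-of-day, choose sequences, spawn override meshes on the spin center, and adjust rotation, camera, scalability and GPU dumps.
# Render optionally builds a multi-CID sequence and shot manifest before calling RenderSequence.
# Embedded Python edits Sequencer event keys and playback ranges for collection and morph-target captures.
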EditorUtilityWidgetBlueprint /Gbuffer_Turntable/Turntable_Gbuffer_LightingWidget
graphs [
 AddTurntableActors {nodes [
   n1 = AddTurntableActors!() -> (then, ObjectToUse:Array of Object References) {is=title K2Node_FunctionEntry}
   > n2 = For Each Loop(Array:Array of Object References=@n1) -> (Array Element:Object Reference)
   > PrintString(InString=GetDisplayName(Object=@n2))
   > n3 = SpawnActorFromObject!(execute, ObjectToUse:Object Reference=@n2, Location:Vector="0, 0, 0", Rotation:Rotator, bTransient:bool=false) -> (then, ReturnValue:Actor Object Reference)
   > n4 = Cast To StaticMeshActor!(execute, Object=@n3) -> (then, CastFailed, AsStatic Mesh Actor:Static Mesh Actor Object Reference) {is=title K2Node_DynamicCast}
   > n5 = SetMobility!(execute, self:Static Mesh Actor Object Reference=@n4, InMobility:EComponentMobility Enum=Movable) -> (then)
   Attach Actor To Actor!(
    execute=[@n5.then, @n4.CastFailed]
    self:Actor Object Reference=@n3
    ParentActor:Actor Object Reference=Get TurntableSpinCenter() -> (TurntableSpinCenter:Actor Object Reference)
    SocketName:Name=None
    LocationRule:EAttachmentRule Enum=KeepRelative
    RotationRule:EAttachmentRule Enum=KeepRelative
    ScaleRule:EAttachmentRule Enum=KeepRelative
    bWeldSimulatedBodies:bool=TRUE
   ) -> (then, ReturnValue:bool)
  ]}
 AddTurntableVehicleActors {nodes [
   n1 = AddTurntableVehicleActors!() -> (then, ObjectToUse:Array of Object References) {is=title K2Node_FunctionEntry}
   > n2 = Sequence()
   n3 = "Get (a copy)"(Array:Array of Object References=@n1) -> (Output:Object Reference)
   > PrintString(InString=GetDisplayName(Object=@n3))
   > n4 = SpawnActorFromObject(ObjectToUse=@n3, Location="0, 0, 0", bTransient=false)
   > n5 = Cast To StaticMeshActor(Object=@n4)
   > n6 = SetMobility(self=@n5, InMobility=Movable)
   Attach Actor To Actor(
    execute=[@n6.then, @n5.CastFailed]
    self=@n4
    ParentActor=Get TurntableSpinCenter()
    SocketName=None
    LocationRule=KeepRelative
    RotationRule=KeepRelative
    ScaleRule=KeepRelative
    bWeldSimulatedBodies=TRUE
   )
   n7 = "Get (a copy)"(Array=@n1, Dimension 1=1)
   Branch(execute=@n2.then_1, Condition=IsValid(Object=@n7))
   > PrintString(InString=GetDisplayName(Object=@n7))
   > n8 = Sequence!(execute) -> (then_0, then_1, then_2:Exec, then_3:Exec)
   > n9 = SpawnActorFromObject(ObjectToUse=@n7, Location="0, 0, 0", bTransient=false)
   > n10 = Cast To StaticMeshActor(Object=@n9)
   > n11 = SetMobility(self=@n10, InMobility=Movable)
   Attach Actor To Actor(
    execute=[@n11.then, @n10.CastFailed]
    self=@n9
    ParentActor=@n4
    SocketName=Socket_FL_Wheel
    LocationRule=KeepRelative
    RotationRule=KeepRelative
    ScaleRule=KeepRelative
    bWeldSimulatedBodies=TRUE
   )
   n12 = SpawnActorFromObject(execute=@n8.then_1, ObjectToUse=@n7, Location="0, 0, 0", bTransient=false)
   > n13 = Cast To StaticMeshActor(Object=@n12)
   > n14 = SetMobility(self=@n13, InMobility=Movable)
   Attach Actor To Actor(
    execute=[@n14.then, @n13.CastFailed]
    self=@n12
    ParentActor=@n4
    SocketName=Socket_FR_Wheel
    LocationRule=KeepRelative
    RotationRule=KeepRelative
    ScaleRule=KeepRelative
    bWeldSimulatedBodies=TRUE
   )
   n15 = SpawnActorFromObject(execute=@n8.then_2, ObjectToUse=@n7, Location="0, 0, 0", bTransient=false)
   > n16 = Cast To StaticMeshActor(Object=@n15)
   > n17 = SetMobility(self=@n16, InMobility=Movable)
   Attach Actor To Actor(
    execute=[@n17.then, @n16.CastFailed]
    self=@n15
    ParentActor=@n4
    SocketName=Socket_BL_Wheel
    LocationRule=KeepRelative
    RotationRule=KeepRelative
    ScaleRule=KeepRelative
    bWeldSimulatedBodies=TRUE
   )
   n18 = SpawnActorFromObject(execute=@n8.then_3, ObjectToUse=@n7, Location="0, 0, 0", bTransient=false)
   > n19 = Cast To StaticMeshActor(Object=@n18)
   > n20 = SetMobility(self=@n19, InMobility=Movable)
   Attach Actor To Actor(
    execute=[@n20.then, @n19.CastFailed]
    self=@n18
    ParentActor=@n4
    SocketName=Socket_BR_Wheel
    LocationRule=KeepRelative
    RotationRule=KeepRelative
    ScaleRule=KeepRelative
    bWeldSimulatedBodies=TRUE
   )
  ]}
 CompileKeyData {nodes [
   n1 = CompileKeyData!() -> (then, Keys:Array of Movie Scene Scripting Key Object References) {is=title K2Node_FunctionEntry}
   > n2 = Sequence!(execute) -> (then_0, then_1, then_2)
   Reroute Node(InputPin=@n2.then_1)
   Return Node!(
    execute=@n2.then_2
    KeyFrameSections1:Array of Gbuffer TT Key Frame Data Structures=Get KeyFrameSections() -> (KeyFrameSections:Array of Gbuffer TT Key Frame Data Structures)
   )
   n3 = For Loop(execute=@n2.then_0, LastIndex=int - int(A=Length(TargetArray:Array of Movie Scene Scripting Key Object References=@n1), B=2))
   n4 = Cast To MovieSceneScriptingEventKey!(Object="Get (a copy)"(Array:Array of Movie Scene Scripting Key Object References=@n1, Dimension 1=@n3) -> (Output:Movie Scene Scripting Key Object Reference)) -> (AsMovie Scene Scripting Event Key:Movie Scene Scripting Event Key Object Reference, bSuccess:bool=~) {is=title K2Node_DynamicCast}
   > Execute Python Script!(
    execute
    PythonScript="import unreal\r\n\r\nendpoint = key.get_editor_property(\"weak_endpoint\")\r\n\r\nprint(endpoint.get_function_name())"
    key:Movie Scene Event Structure="Get Value (Event)"!(self:Movie Scene Scripting Event Key Object Reference=@n4) -> (ReturnValue:Movie Scene Event Structure) {is=title K2Node_CallFunction}
   ) -> (then, ReturnValue) {inputs [key]}
   n5 = int + int(A=@n3)
   n6 = "Get Time (Event)"!(
    self:Movie Scene Scripting Event Key Object Reference=Cast To MovieSceneScriptingEventKey(Object="Get (a copy)"(Array=@n1, Dimension 1=@n5)) -> (bSuccess=~)
    TimeUnit:EMovieSceneTimeUnit Enum=DisplayRate
   ) -> (ReturnValue_FrameNumber_Value:int, ReturnValue_SubFrame:float=~) {is=title K2Node_CallFunction}
   > Add!(
    execute
    TargetArray:Array of Gbuffer TT Key Frame Data Structures=Get KeyFrameSections()
    NewItem_Passname_2_#1:String
    NewItem_FrameStart_6_#2:String="To String (Integer)"(InInt="Get Time (Event)"(self=@n4, TimeUnit=DisplayRate) -> (ReturnValue_SubFrame=~))
    NewItem_FrameEnd_8_#3:String="To String (Integer)"(InInt=int - int(A=@n6))
    NewItem_ID_10_#4:String="To String (Integer)"(InInt=@n5)
    NewItem_Name_12_#5:String
    NewItem_Type_15_#6:String=Shot
   ) -> (then, ReturnValue) {is=title K2Node_CallArrayFunction}
  ]}
 CountTotalDetectedMorphTargets {nodes [
   CountTotalDetectedMorphTargets() {is=title K2Node_FunctionEntry}
   > n1 = GetAllActorsOfClass!(execute, WorldContextObject:Object Reference, ActorClass:Actor Class Reference=/Script/Engine.SkeletalMeshActor) -> (then, OutActors:Array of Skeletal Mesh Actor Object References)
   > n2 = For Each Loop(Array:Array of Skeletal Mesh Actor Object References=@n1) -> (Array Element:Skeletal Mesh Actor Object Reference)
   Return Node!(
    execute=@n2.Completed
    TotalMorphTargets:int=Length(TargetArray:Array of Strings=Get MorphTargetNames() -> (MorphTargetNames:Array of Strings))
   )
   n3 = Get All Morph Target Names!(self:Skeletal Mesh Object Reference=Get SkeletalMesh!(self:Skinned Mesh Component Object Reference=Get SkeletalMeshComponent!(self:Skeletal Mesh Actor Object Reference=@n2) -> (SkeletalMeshComponent:Skeletal Mesh Component Object Reference)) -> (SkeletalMesh:Skeletal Mesh Object Reference)) -> (ReturnValue:Array of Strings) {is=title K2Node_CallFunction}
   > Branch(Condition=Is Not Empty!(TargetArray=@n3) -> (ReturnValue:bool))
   > n4 = For Each Loop(Array:Array of Strings=@n3) -> (Array Element:String)
   > Add Unique(TargetArray=Get MorphTargetNames(), NewItem=@n4)
  ]}
 DestroyTurntableActors {nodes [
   DestroyTurntableActors() {is=title K2Node_FunctionEntry}
   > n1 = Reverse for Each Loop!(
    Exec:Exec
    Array:Array of Actor Object References=GetAttachedActors!(self:Actor Object Reference=Get TurntableSpinCenter(), bResetArray:bool=true, bRecursivelyIncludeAttachedActors:bool=true) -> (OutActors:Array of Actor Object References)
   ) -> (LoopBody:Exec, ArrayIndex:int, ArrayElement:Actor Object Reference, Completed:Exec) {is=title K2Node_MacroInstance}
   Return Node!(execute=@n1.Completed)
   > Destroy Actor(self=@n1.ArrayElement)
  ]}
 EventGraph {nodes [
   Hide Level()
   > Load Level!(execute, self, PackageName:Name=GBufferLighting) -> (then)
   n1 = GenerateMultiCIDSequence!(execute, self, IsMultiCID?:bool=Get IsMultiCID() -> (IsMultiCID:bool)) -> (then)
   n2 = Branch(execute=[@n1.then, @n58.then], Condition=Get GenerateData?() -> (GenerateData?:bool))
   > n3 = GenerateSequenceManifest!(
    execute
    self
    KeyFrames:Array of Gbuffer TT Key Frame Data Structures=Get SelectedShotData() -> (SelectedShotData:Array of Gbuffer TT Key Frame Data Structures)
   ) -> (then)
   RenderSequence(execute=[@n3.then, @n2.else, @n11.then])
   Hide Level()
   > Load Level(PackageName=BRARTLighting_Toon)
   Hide Level()
   > Load Level(PackageName=BR_Art_Active_EditorOnly)
   n4 = "On Value Changed (Rotation_Slider)"!() -> (OutputDelegate, then, Value:float) {is=title K2Node_ComponentBoundEvent}
   n5 = Get Actor Reference() -> (Actor Reference:Actor Object Reference)
   > Is Valid(InputObject=@n5)
   n6 = Get Actor Rotation!(self:Actor Object Reference=@n5) -> (ReturnValue_Roll:float, ReturnValue_Pitch:float, ReturnValue_Yaw:float) {is=title K2Node_CallFunction}
   > Set Actor Rotation!(
    execute
    self:Actor Object Reference=@n5
    NewRotation_Roll:float=@n6
    NewRotation_Pitch:float=@n6.ReturnValue_Pitch
    NewRotation_Yaw:float=@n4.Value
    bTeleportPhysics:bool=false
   ) -> (then, ReturnValue:bool) {is=title K2Node_CallFunction}
   "On Property Changed (RotationSelector)"!() -> (OutputDelegate, then, PropertyName:Name=None) {is=title K2Node_ComponentBoundEvent}
   > n7 = Is Valid(InputObject=Get Actor Reference())
   n8 = Get Rotation_Slider() -> (Rotation_Slider:Slider Object Reference)
   SetVisibility(execute=@n7.Is Not Valid, self=@n8, InVisibility=Collapsed)
   > Set RotationSliderIsEnabled(RotationSliderIsEnabled:bool=false) -> (Output_Get:bool)
   > SetToolTipText!(
    execute
    self:Widget Object Reference=@n8
    InToolTipText:"Text (by ref)"=MakeLiteralText(Value="LOCTEXT(\"Set ACTOR REFERENCE to enable Rotation Slider\")")
   ) -> (then)
   SetVisibility(execute=@n7.Is Valid, self=@n8, InVisibility=Visible)
   > Set RotationSliderIsEnabled(RotationSliderIsEnabled=true)
   > SetToolTipText(self=@n8, InToolTipText=MakeLiteralText(Value="LOCTEXT(\"Slide To Rotate Actor\")"))
   > UpdateRotationSliderPosition()
   n9 = "On Selection Changed (ForceQualityMode)"!() -> (OutputDelegate, then, SelectedItem:String, SelectionType:ESelectInfo Enum) {is=title K2Node_ComponentBoundEvent}
   n10 = Get ForceQualityMode() -> (ForceQualityMode:"Combo Box (String) Object Reference")
   SetScalability!(
    execute=[@n71.then, @n9.then]
    self
    Scalability_Low:bool="Equal, Case Insensitive (String)"(A=GetSelectedOption!(self:"Combo Box (String) Object Reference"=@n10) -> (ReturnValue:String), B=Low)
    ForceScalability:bool=true
   ) -> (then)
   "On Clicked (Debug)"() {is=title K2Node_ComponentBoundEvent}
   > n11 = Set MRQ_Sequence(MRQ_Sequence:Soft Object Path Structure="Get (a copy)"(
     Array:Array of Soft Object Path Structures=Get MRQ_Sequence_MasterList() -> (MRQ_Sequence_MasterList:Array of Soft Object Path Structures)
     Dimension 1=5
    ) -> (Output:Soft Object Path Structure)) -> (Output_Get:Soft Object Path Structure)
   "On Clicked (Documentation)"() {is=title K2Node_ComponentBoundEvent}
   > LaunchURL!(execute, URL:String=https://confluence.it.epicgames.com/display/ART/Turntable+Setup+in+UE) -> (then)
   n12 = "On Check State Changed (GPUDump_Checkbox)"!() -> (OutputDelegate, then, bIsChecked:bool) {is=title K2Node_ComponentBoundEvent}
   > n13 = Set DUMPGPUCheckboxState(DUMPGPUCheckboxState:bool=@n12.bIsChecked) -> (Output_Get:bool)
   > n14 = SetIsEnabled(self=Get GPUDumpFrame() -> (GPUDumpFrame:Editable Text Object Reference), bInIsEnabled=@n13)
   n15 = GetAllActorsOfClass(execute=[@n14.then, @n92.then], ActorClass=/Gbuffer_Turntable/FN_PublicTurntable_Assets/BP/BP_Turntable_GBuffers.BP_Turntable_GBuffers_C) -> (OutActors:Array of BP Turntable GBuffers Object References)
   > Branch(Condition=Is Not Empty(TargetArray:Array of BP Turntable GBuffers Object References=@n15))
   n16 = "Get (a copy)"(Array:Array of BP Turntable GBuffers Object References=@n15) -> (Output:BP Turntable GBuffers Object Reference)
   > n17 = Set UseDumpGPU?!(execute, UseDumpGPU?:bool=Get DUMPGPUCheckboxState() -> (DUMPGPUCheckboxState:bool), self:BP Turntable GBuffers Object Reference=@n16) -> (then, Output_Get:bool)
   n18 = "OnTextChanged (Editable Text) (GPUDumpFrame)"!() -> (OutputDelegate, then, Text:"Text (by ref)") {is=title K2Node_ComponentBoundEvent}
   n19 = String To Integer(InString="To String (Text)"(InText=@n18.Text))
   Set DumpFrame!(execute=@n17.then, DumpFrame:int=@n19, self:BP Turntable GBuffers Object Reference=@n16) -> (then, Output_Get:int=0)
   "On Property Changed (MeshOverride_DV)"!() -> (OutputDelegate, then, PropertyName:Name=None) {is=title K2Node_ComponentBoundEvent}
   > Set MeshOverride_SafeToReplace(MeshOverride_SafeToReplace:bool=false) -> (Output_Get:bool)
   n20 = Get OverrideMesh() -> (OverrideMesh:Array of Object References)
   n21 = Is Not Empty(TargetArray:Array of Object References=@n20)
   > n22 = Branch(Condition=@n21)
   > n23 = For Each Loop(Array:Array of Object References=@n20) -> (Array Element:Object Reference)
   Branch(execute=@n23.Completed, Condition=AND Boolean(A=@n21, B=Get MeshOverride_SafeToReplace() -> (MeshOverride_SafeToReplace:bool)))
   > DestroyTurntableActors() {is=title K2Node_CallFunction}
   > n24 = Switch on ECheckBoxState!(
    execute
    Selection:ECheckBoxState Enum=GetCheckedState!(self:Check Box Object Reference=Get VehicleModeSelect() -> (VehicleModeSelect:Check Box Object Reference)) -> (ReturnValue:ECheckBoxState Enum=Unchecked)
   ) -> (Unchecked:Exec, Checked:Exec, Undetermined:Exec) {is=title K2Node_SwitchEnum}
   n25 = Get OverrideMesh()
   AddTurntableVehicleActors!(execute=@n24.Checked, self, ObjectToUse:Array of Object References=@n25) -> (then) {is=title K2Node_CallFunction}
   AddTurntableActors!(execute=@n24.Unchecked, self, ObjectToUse:Array of Object References=@n25) -> (then) {is=title K2Node_CallFunction}
   Is Valid(exec=@n23.LoopBody, InputObject=@n23)
   > n26 = Cast To StaticMesh(Object=@n23)
   Set MeshOverride_SafeToReplace(execute=[@n26.then, @n27.then], MeshOverride_SafeToReplace=true)
   n27 = Cast To SkeletalMesh!(execute=@n26.CastFailed, Object=@n23) -> (then, CastFailed, AsSkeletal Mesh:Skeletal Mesh Object Reference) {is=title K2Node_DynamicCast}
   Remove Index!(execute=@n27.CastFailed, TargetArray:Array of Object References=@n20, IndexToRemove:int=@n23.Array Index) -> (then)
   DestroyTurntableActors(execute=@n22.else)
   n28 = "On Check State Changed (ToggleToCurrentTODM)"!() -> (OutputDelegate, then, bIsChecked:bool) {is=title K2Node_ComponentBoundEvent}
   > n29 = Branch(Condition=@n28.bIsChecked)
   > SetGlobalTimeOfDayManager!(
    execute
    WorldContextObject:Object Reference
    InTimeOfDayManager:Info Soft Class Reference=/Game/TimeOfDay/TODM_Sequence/Blueprint/BP_DaySequence_Current.BP_DaySequence_Current_C
   ) -> (then)
   n30 = GetGlobalTimeOfDayManager!(WorldContextObject:Object Reference) -> (ReturnValue:Fort Time Of Day Manager Interface Interface)
   > SetTimeOfDayManagerVisibility!(execute, self:Fort Time Of Day Manager Interface Interface=@n30, bVisible:bool=true) -> (then)
   > n31 = GetAllActorsOfClass(ActorClass=/Script/FortniteGame.FortDaySequenceActor) -> (OutActors:Array of Fort Day Sequence Actor Object References)
   > n32 = Branch(Condition="integer > integer"(A=Length(TargetArray:Array of Fort Day Sequence Actor Object References=@n31)))
   > Set Actor Rotation(
    self="Get (a copy)"(Array:Array of Fort Day Sequence Actor Object References=@n31) -> (Output:Fort Day Sequence Actor Object Reference)
    NewRotation_Roll=0
    NewRotation_Pitch=0
    NewRotation_Yaw=10
    bTeleportPhysics=false
   )
   PrintString(execute=@n32.else, InString="~~~~~~~~~~~~~~~~~~~~~~~~~~~~~~~~~~~~~~~~~~~CANNOT FIND TODM FOR ROTATION~~~~~~~~~~~~~~~~~~~~~~~~~")
   SetGlobalTimeOfDayManager(
    execute=@n29.else
    InTimeOfDayManager=/Game/TimeOfDay/TODM_Sequence/Blueprint/BP_DaySequence_CharacterReview.BP_DaySequence_CharacterReview_C
   )
   > SetTimeOfDayManagerVisibility(self=@n30, bVisible=true)
   > n33 = GetAllActorsOfClass(ActorClass=&FortDaySequenceActor)
   > n34 = Branch(Condition="integer > integer"(A=Length(TargetArray=@n33)))
   > Set Actor Rotation(self="Get (a copy)"(Array=@n33), NewRotation_Roll=0, NewRotation_Pitch=0, NewRotation_Yaw=0, bTeleportPhysics=false)
   PrintString(execute=@n34.else, InString="~~~~~~~~~~~~~~~~~~~~~~~~~~~~~~~~~~~~~~~~~~~CANNOT FIND TODM FOR ROTATION~~~~~~~~~~~~~~~~~~~~~~~~~")
   n35 = "On Selection Changed (LightingModeSelector)"!() -> (OutputDelegate, then, SelectedItem:String, SelectionType:ESelectInfo Enum) {is=title K2Node_ComponentBoundEvent}
   > PrintString(InString=@n35.SelectedItem)
   > n36 = Switch on String!(execute:Exec, Selection:String=@n35.SelectedItem) -> (Neutral Lighting:Exec, BR_Art Lighting:Exec, BR_Art Lighting Toon:Exec, Front End Lighting:Exec, TODM Lighting:Exec, 3 Point Lighting:Exec) {
    bHasDefaultPin=false
    pinNames [Neutral Lighting, BR_Art Lighting, BR_Art Lighting Toon, Front End Lighting, TODM Lighting, 3 Point Lighting]
   }
   Hide Level(execute=@n36.3 Point Lighting)
   > Load Level(PackageName=ThreePointLighting)
   > n37 = Sequence!(execute) -> (then_0, then_1)
   n38 = Get BR_Art_LightingBox() -> (BR_Art_LightingBox:Horizontal Box Object Reference)
   SetVisibility(execute=[@n37.then_0, @n40.then_0], self=@n38, InVisibility=Collapsed)
   UpdateTODM!(
    execute=@n37.then_1
    self
    InTimeOfDayManager:Info Soft Class Reference
    bVisible:bool=false
    UpdateRotation?:bool=false
    NewRotation:double=0
   ) -> (then)
   > Adjust Exposure!(execute, self, InputPin:double=0) -> (then)
   Hide Level(execute=@n36.BR_Art Lighting)
   > Load Level(PackageName=BRARTLighting_Standard)
   > n39 = Sequence()
   > SetVisibility(self=@n38, InVisibility=Visible)
   UpdateTODM(
    execute=@n39.then_1
    InTimeOfDayManager=&BP_DaySequence_CharacterReview_C
    bVisible=true
    UpdateRotation?=true
    NewRotation=0
   )
   > Adjust Exposure(InputPin=-0.8)
   Hide Level(execute=@n36.Front End Lighting)
   > Load Level(PackageName=FELightingCopy)
   > n40 = Sequence()
   UpdateTODM(execute=@n40.then_1, bVisible=false, UpdateRotation?=false, NewRotation=0)
   > Adjust Exposure(InputPin=0)
   n41 = "On Selection Changed (SequenceSelector)"!() -> (OutputDelegate, then, SelectedItem:String, SelectionType:ESelectInfo Enum) {is=title K2Node_ComponentBoundEvent}
   > n42 = Switch on String!(execute, Selection=@n41.SelectedItem) -> (
    Default:Exec
    Standard Characters:Exec
    Standard Characters - IP Headshot:Exec
    Toon Characters:Exec
    Toon Characters - BR_Art_Current:Exec
    Toon Characters - IP Headshot:Exec
    Modelling Sequence:Exec
    Turntable Status Verification:Exec
    Morph Target:Exec
    Vehicles:Exec
    ExecBeautyShots:Exec
    FlipbookCapture:Exec
    MultiCIDCapture:Exec
   ) {pinNames [
     Standard Characters
     Standard Characters - IP Headshot
     Toon Characters
     Toon Characters - BR_Art_Current
     Toon Characters - IP Headshot
     Modelling Sequence
     Turntable Status Verification
     Morph Target
     Vehicles
     ExecBeautyShots
     FlipbookCapture
     MultiCIDCapture
    ]}
   > PrintString(InString=Unknown Error in Sequence Switch)
   Change Sequence!(execute=@n42.ExecBeautyShots, self, Sequence:int=6) -> (then)
   > SetShotData!(execute, self, PassType:int=0) -> (then)
   n43 = Get MRQ_Sequence_MasterList()
   > n44 = Set MRQ_Sequence(MRQ_Sequence="Get (a copy)"(Array:Array of Soft Object Path Structures=@n43, Dimension 1=7) -> (Output:Soft Object Path Structure))
   n45 = Set GenerateCustomSequence(
    execute=[@n50.then_0, @n48.then, @n57.then, @n44.then, @n47.then, @n53.then, @n51.then, @n52.then, @n55.then, @n54.then]
    GenerateCustomSequence:bool=false
   ) -> (Output_Get:bool)
   n46 = Get Render() -> (Render:Button Object Reference)
   SetIsEnabled(execute=[@n45.then, @n56.then, @n49.then], self=@n46, bInIsEnabled=true)
   > Set IsMultiCID(IsMultiCID:bool=false) -> (Output_Get:bool)
   Change Sequence(execute=@n42.FlipbookCapture, Sequence=7)
   > SetShotData(PassType=3)
   > n47 = Set MRQ_Sequence(MRQ_Sequence="Get (a copy)"(Array=@n43, Dimension 1=8))
   Change Sequence(execute=@n42.Modelling Sequence, Sequence=2)
   > SetShotData(PassType=0)
   > n48 = Set MRQ_Sequence(MRQ_Sequence="Get (a copy)"(Array=@n43, Dimension 1=2))
   Change Sequence(execute=@n42.Morph Target, Sequence=4)
   > SetShotData(PassType=0)
   > Set MRQ_Sequence(MRQ_Sequence="Get (a copy)"(Array=@n43, Dimension 1=4))
   > n49 = Set GenerateCustomSequence(GenerateCustomSequence=true)
   Change Sequence(execute=@n42.MultiCIDCapture, Sequence=8)
   > SetShotData(PassType=0)
   > Set MRQ_Sequence(MRQ_Sequence="Get (a copy)"(Array=@n43, Dimension 1=9))
   > n50 = Sequence()
   Set IsMultiCID(execute=@n50.then_1, IsMultiCID=true)
   Change Sequence(execute=@n42.Standard Characters, Sequence=0)
   > SetShotData(PassType=1)
   > n51 = Set MRQ_Sequence(MRQ_Sequence="Get (a copy)"(Array=@n43))
   Change Sequence(execute=@n42.Standard Characters - IP Headshot, Sequence=10)
   > SetShotData(PassType=1)
   > n52 = Set MRQ_Sequence(MRQ_Sequence="Get (a copy)"(Array=@n43, Dimension 1=11))
   Change Sequence(execute=@n42.Toon Characters, Sequence=1)
   > SetShotData(PassType=2)
   > n53 = Set MRQ_Sequence(MRQ_Sequence="Get (a copy)"(Array=@n43, Dimension 1=1))
   Change Sequence(execute=@n42.Toon Characters - BR_Art_Current, Sequence=9)
   > SetShotData(PassType=2)
   > n54 = Set MRQ_Sequence(MRQ_Sequence="Get (a copy)"(Array=@n43, Dimension 1=10))
   Change Sequence(execute=@n42.Toon Characters - IP Headshot, Sequence=11)
   > SetShotData(PassType=2)
   > n55 = Set MRQ_Sequence(MRQ_Sequence="Get (a copy)"(Array=@n43, Dimension 1=12))
   Change Sequence(execute=@n42.Turntable Status Verification, Sequence=3)
   > SetShotData(PassType=0)
   > n56 = Set MRQ_Sequence(MRQ_Sequence="Get (a copy)"(Array=@n43, Dimension 1=3))
   Change Sequence(execute=@n42.Vehicles, Sequence=5)
   > SetShotData(PassType=0)
   > n57 = Set MRQ_Sequence(MRQ_Sequence="Get (a copy)"(Array=@n43, Dimension 1=6))
   "On Mouse Capture Begin (Rotation_Slider)"() {is=title K2Node_ComponentBoundEvent}
   > Set RotationSliderUserIsAdjusting(RotationSliderUserIsAdjusting:bool=true) -> (Output_Get:bool)
   "On Mouse Capture End (Rotation_Slider)"() {is=title K2Node_ComponentBoundEvent}
   > Set RotationSliderUserIsAdjusting(RotationSliderUserIsAdjusting=false)
   n58 = "On Clicked (Render)"() {is=title K2Node_ComponentBoundEvent}
   n59 = "On Check State Changed (BackpackMode)"!() -> (OutputDelegate, then, bIsChecked:bool) {is=title K2Node_ComponentBoundEvent}
   > n60 = Sequence!(execute) -> (then_0, then_1, then_2:Exec)
   > n61 = GetAllActorsOfClass(ActorClass=/Script/Engine.CameraActor) -> (OutActors:Array of Camera Actor Object References)
   n62 = GetAllActorsOfClass(execute=@n60.then_1, ActorClass=&StaticMeshActor) -> (OutActors:Array of Static Mesh Actor Object References)
   > n63 = For Each Loop(Array:Array of Static Mesh Actor Object References=@n62) -> (Array Element:Static Mesh Actor Object Reference)
   n64 = Get StaticMeshComponent(self=@n63)
   > Branch(Condition="Equal (Object)"(A=Get StaticMesh(self=@n64), B=Get ColorCube() -> (ColorCube:Static Mesh Object Reference)))
   > n65 = Set LevelColorCube(LevelColorCube:Static Mesh Actor Object Reference=@n63) -> (Output_Get:Static Mesh Actor Object Reference)
   n66 = Branch(execute=@n60.then_2, Condition=@n59.bIsChecked)
   n67 = "Get (a copy)"(Array:Array of Camera Actor Object References=@n61) -> (Output:Camera Actor Object Reference)
   > Set Actor Relative Transform!(
    execute
    self:Actor Object Reference=@n67
    NewRelativeTransform:"Transform (by ref)"=Get CameraLocation_Backpack() -> (CameraLocation_Backpack:Transform)
    bSweep:bool=false
    bTeleport:bool=false
   ) -> (then, SweepHitResult:Hit Result Structure) {is=title K2Node_CallFunction}
   > Set Actor Relative Transform(self=@n65, NewRelativeTransform=Get ColorCalibrator_Backpack() -> (ColorCalibrator_Backpack:Transform), bSweep=false, bTeleport=false)
   Set Actor Relative Transform(
    execute=@n66.else
    self=@n67
    NewRelativeTransform=Get CameraLocation_Standard() -> (CameraLocation_Standard:Transform)
    bSweep=false
    bTeleport=false
   )
   > Set Actor Relative Transform(self=@n65, NewRelativeTransform=Get ColorCalibrator_Standard() -> (ColorCalibrator_Standard:Transform), bSweep=false, bTeleport=false)
   n68 = "On Check State Changed (ForceQuality_Checkbox)"!() -> (OutputDelegate, then, bIsChecked:bool) {is=title K2Node_ComponentBoundEvent}
   > n69 = Branch(Condition=@n68.bIsChecked)
   n70 = Get ForceQualityMode()
   > n71 = SetVisibility(self=@n70, InVisibility=Visible)
   SetVisibility(execute=@n69.else, self=@n70, InVisibility=Hidden)
   > SetScalability(Scalability_Low=false, ForceScalability=false)
   n72 = "On Check State Changed (ShowAdvanced_Checkbox)"!() -> (OutputDelegate, then, bIsChecked:bool) {is=title K2Node_ComponentBoundEvent}
   > n73 = Branch(Condition=@n72.bIsChecked)
   n74 = Get AdvancedBox() -> (AdvancedBox:Horizontal Box Object Reference)
   > SetVisibility(self=@n74, InVisibility=Visible)
   SetVisibility(execute=@n73.else, self=@n74, InVisibility=Collapsed)
   n75 = Adjust Exposure!() -> (OutputDelegate, then, InputPin:double=~) {is=title K2Node_CustomEvent}
   > n76 = GetAllActorsOfClass(ActorClass=&CameraActor) -> (OutActors:Array of Camera Actor Object References)
   > n77 = Cast To CameraActor!(execute, Object="Get (a copy)"(Array=@n76)) -> (then, CastFailed, AsCamera Actor:Camera Actor Object Reference) {is=title K2Node_DynamicCast}
   > Set members in Post Process Settings!(
    execute:Exec
    StructRef:"Post Process Settings Structure (by ref)"=Get PostProcessSettings!(self:Camera Component Object Reference=Get CameraComponent!(self:Camera Actor Object Reference=@n77) -> (CameraComponent:Camera Component Object Reference)) -> (PostProcessSettings:Post Process Settings Structure)
    AutoExposureBias:float=@n75.InputPin
   ) -> (then:Exec, StructOut:"Post Process Settings Structure (by ref)") {is=title K2Node_SetFieldsInStruct}
   n78 = Load Level!() -> (OutputDelegate, then, PackageName:Name) {is=title K2Node_CustomEvent}
   > n79 = Set Currently Loaded(Currently Loaded:Name=@n78.PackageName) -> (Output_Get:Name=None)
   > SetLevelVisibility!(
    execute
    Level:Level Object Reference=GetLoadedLevel!(self:Level Streaming Object Reference=GetStreamingLevel!(WorldContextObject:Object Reference, PackageName:Name=@n79) -> (ReturnValue:Level Streaming Object Reference)) -> (ReturnValue:Level Object Reference)
    bShouldBeVisible:bool=true
    bForceLayersVisible:bool=false
    ModifyMode:ELevelVisibilityDirtyMode Enum=ModifyOnChange
   ) -> (then)
   Hide Level() {is=title K2Node_CustomEvent}
   n80 = Get Currently Loaded() -> (Currently Loaded:Name=None)
   > SetLevelVisibility(Level=GetLoadedLevel(self=GetStreamingLevel(PackageName=@n80)), bShouldBeVisible=false, bForceLayersVisible=false, ModifyMode=ModifyOnChange)
   n81 = Change Sequence!() -> (OutputDelegate, then, Sequence:int) {is=title K2Node_CustomEvent}
   > n82 = GetAllActorsOfClass(WorldContextObject=Self-Reference(), ActorClass=/Script/LevelSequence.LevelSequenceActor) -> (OutActors:Array of Level Sequence Actor Object References)
   > n83 = Branch(Condition="integer > integer"(A=Length(TargetArray:Array of Level Sequence Actor Object References=@n82), B=0) {is=title K2Node_CallFunction})
   > n84 = Cast To LevelSequenceActor!(execute, Object="Get (a copy)"(Array:Array of Level Sequence Actor Object References=@n82) -> (Output:Level Sequence Actor Object Reference)) -> (then, CastFailed, AsLevel Sequence Actor:Level Sequence Actor Object Reference) {is=title K2Node_DynamicCast}
   n85 = "Get (a copy)"(
    Array:Array of Level Sequence Object References=Get LevelSequences() -> (LevelSequences:Array of Level Sequence Object References)
    Dimension 1=@n81.Sequence
   ) -> (Output:Level Sequence Object Reference)
   > SetSequence!(execute, self:Level Sequence Actor Object Reference=@n84, InSequence:Level Sequence Object Reference=@n85) -> (then)
   > PrintString(InString=Append(A="Setting Level Sequence to: ", B=GetDisplayName(Object=@n85)))
   PrintString(execute=@n83.else, InString="Error, No Level Sequences Detected")
   Event PreConstruct()
   > ExecuteConsoleCommand(Command=Slate.bAllowThrottling 0)
   > SetObject(self=Get MultiCIDCapture_DV() -> (MultiCIDCapture_DV:Details View Object Reference), NewObject=Self-Reference())
   Event Construct()
   > SetObject(self=Get RotationSelector() -> (RotationSelector:Details View Object Reference), NewObject=Self-Reference())
   > SetVisibility(self=Get Rotation_Slider(), InVisibility=Visible)
   > SetObject(
    self=[
     Get OutputFolder_DV() -> (OutputFolder_DV:Details View Object Reference)
     Get MRQSettingLocator() -> (MRQSettingLocator:Details View Object Reference)
     Get MeshOverride_DV() -> (MeshOverride_DV:Details View Object Reference)
    ]
    NewObject=Self-Reference()
   )
   n86 = "To String (Float)"(InDouble=Get!() -> (VersionInfo_CurrentVersion_3_#7:float=~))
   > "SetText (Text)"(
    self=Get TextBlock_126() -> (TextBlock_126:Text Object Reference)
    InText="To Text (String)"(InString=Append(A=Turntable Controls V, B=@n86))
   )
   > n87 = GetAllActorsOfClass(ActorClass=&BP_Turntable_GBuffers_C) -> (OutActors:Array of BP Turntable GBuffers Object References)
   > Branch(Condition=Is Not Empty(TargetArray:Array of BP Turntable GBuffers Object References=@n87))
   n88 = Get Scalability!(self:BP Turntable GBuffers Object Reference="Get (a copy)"(Array:Array of BP Turntable GBuffers Object References=@n87) -> (Output:BP Turntable GBuffers Object Reference)) -> (Scalability:String)
   > SetSelectedIndex!(
    execute
    self:"Combo Box (String) Object Reference"=Get ForceQualityMode()
    Index:int="To Integer (Boolean)"!(InBool:bool="Equal, Case Insensitive (String)"(A=@n88, B=Low)) -> (ReturnValue:int=0)
   ) -> (then)
   > n89 = GetAllActorsWithTag!(execute, WorldContextObject:Object Reference, Tag:Name=TurnTable_UserObject) -> (then, OutActors:Array of Actor Object References)
   > Branch(Condition=Is Not Empty(TargetArray:Array of Actor Object References=@n89))
   n90 = "Get (a copy)"(Array:Array of Actor Object References=@n89) -> (Output:Actor Object Reference)
   > Set TurntableSpinCenter(TurntableSpinCenter:Actor Object Reference=@n90) -> (Output_Get:Actor Object Reference)
   > Set Actor Reference(Actor Reference:Actor Object Reference=@n90) -> (Output_Get:Actor Object Reference)
   Event Tick()
   > Branch(Condition=Get RotationSliderIsEnabled() -> (RotationSliderIsEnabled:bool))
   > n91 = Branch(Condition=Get RotationSliderUserIsAdjusting() -> (RotationSliderUserIsAdjusting:bool))
   UpdateRotationSliderPosition(execute=@n91.else)
   GetSequenceShotKeyData!(execute, self, MRQ_Sequence:Soft Object Path Structure=Get MRQ_Sequence() -> (MRQ_Sequence:Soft Object Path Structure)) -> (then)
   PrintString(InString=Level Missing Tagged PostProcess)
   SetIsEnabled(self=@n46, bInIsEnabled=false)
   GetAssetByObjectPath!(execute, self:Asset Registry Interface=GetAssetRegistry(), ObjectPath:Name=None, bIncludeOnlyOnDiskAssets:bool=false) -> (then, ReturnValue:Asset Data Structure)
   Set OverrideMesh(OverrideMesh:Array of Object References) -> (Output_Get:Array of Object References)
   n92 = PrintString(execute=@n18.then, InString="To String (Integer)"(InInt=@n19))
  ]}
 GenerateCustomSequence_Morph {nodes [
   GenerateCustomSequence_Morph() {is=title K2Node_FunctionEntry}
   > n1 = CountTotalDetectedMorphTargets!(execute, self) -> (then, TotalMorphTargets:int=0) {is=title K2Node_CallFunction}
   > n2 = Branch(Condition="integer > integer"(A=@n1) {is=title K2Node_PromotableOperator})
   > n3 = PrintString(InString=Append(A="To String (Integer)"(InInt=@n1), B=" Morph Targets Detected"))
   n4 = "Get (a copy)"(Array:Array of Level Sequence Object References=Get LevelSequences(), Dimension 1=4) -> (Output:Level Sequence Object Reference)
   n5 = Execute Python Script!(
    execute=[@n6.then, @n3.then]
    PythonScript="import unreal\r\n\r\nseq.set_playback_start(0)\r\nseq.set_playback_end(30+(30*num_morphtargets))"
    seq:Level Sequence Object Reference=@n4
    num_morphtargets:int=@n1
   ) -> (then, ReturnValue) {inputs [seq, num_morphtargets]}
   Return Node!(execute=[@n5.then, @n7.then], HasValidTargets:bool="To Boolean (Integer)"!(InInt:int=@n1) -> (ReturnValue:bool))
   n6 = PrintString(execute=@n2.else, InString="Error, No Valid Morph Targets Detected!")
   n7 = Construct Object from Class(Class=/Script/LevelSequence, self=@n4) -> (ReturnValue:Level Sequence Object Reference)
  ]}
 GenerateMultiCIDSequence {nodes [
   n1 = GetAssetsInCollection!(
    execute=@n25.then_1
    Collection Name:String="To String (Name)"(InName=Get CollectionName() -> (CollectionName:Name=None))
    LoadObjects?:bool=false
    __WorldContext:Object Reference
   ) -> (then, AsObjects:Array of Object References, AsData:Array of Asset Data Structures)
   n2 = For Each Loop!(Exec=@n29.then, Array:Array of Asset Data Structures=@n1.AsData) -> (
    LoopBody
    Array Element_PackageName:Name
    Array Element_PackagePath:Name
    Array Element_AssetName:Name
    Array Element_AssetClassPath:Top Level Asset Path Structure
    Array Index
    Completed
   )
   n3 = GetAssetsByPackageName!(
    execute
    self:Asset Registry Interface=GetAssetRegistry()
    PackageName:Name=@n2
    bIncludeOnlyOnDiskAssets:bool=false
    bSkipARFilteredAssets:bool=true
   ) -> (then, OutAssetData:Array of Asset Data Structures, ReturnValue:bool)
   n4 = "Get (a copy)"(Array:Array of Asset Data Structures=@n3) -> (Output:Asset Data Structure)
   > PrintString(InString=GetDisplayName(Object=GetAsset(InAssetData=@n4)))
   n5 = GetBindings!(
    execute
    Sequence:Movie Scene Sequence Object Reference=/Gbuffer_Turntable/FN_PublicTurntable_Assets/Sequences/Turntable_Sequence_GBuffers_MultiCID
   ) -> (then, ReturnValue:Array of Movie Scene Binding Proxy Structures)
   > n6 = For Each Loop!(Exec, Array:Array of Movie Scene Binding Proxy Structures=@n5) -> (LoopBody, Array Element:Movie Scene Binding Proxy Structure, Array Index, Completed)
   > GetTracks!(execute, InBinding:"Movie Scene Binding Proxy Structure (by ref)"=@n6) -> (then, ReturnValue:Array of Movie Scene Track Object References)
   n7 = GenerateMultiCIDSequence!() -> (then, IsMultiCID?:bool=~) {is=title K2Node_FunctionEntry}
   > n8 = Branch(Condition=@n7)
   n9 = Get MultiCIDSeq() -> (MultiCIDSeq:Movie Scene Sequence Object Reference)
   > n10 = GetTracks!(execute, Sequence:Movie Scene Sequence Object Reference=@n9) -> (then, ReturnValue)
   > n11 = Sequence!(execute) -> (then_0, then_1, then_2, then_3:Exec)
   > n12 = For Each Loop(Array:Array of Movie Scene Track Object References=@n10) -> (Array Element:Movie Scene Track Object Reference)
   > n13 = GetDisplayName!(execute:Exec, Track:Movie Scene Track Object Reference=@n12) -> (then:Exec, ReturnValue:Text)
   > Branch(Condition="Equal, Case Insensitive (Text)"!(A:"Text (by ref)"=@n13, B:"Text (by ref)"=MakeLiteralText(Value="LOCTEXT(\"SampleKey\")")) -> (ReturnValue:bool))
   > n14 = GetSections!(execute, Track:Movie Scene Track Object Reference=@n12) -> (then, ReturnValue:Array of Movie Scene Section Object References)
   > n15 = For Each Loop(Array:Array of Movie Scene Section Object References=@n14) -> (Array Element:Movie Scene Section Object Reference)
   > GetTrackRowDisplayName!(
    execute
    Track:Movie Scene Track Object Reference=@n12
    RowIndex:int=GetRowIndex!(self:Movie Scene Section Object Reference=@n15) -> (ReturnValue:int=0)
   ) -> (then, ReturnValue:Text)
   > n16 = GetAllChannels!(execute, Section:Movie Scene Section Object Reference=@n15) -> (then, ReturnValue:Array of Movie Scene Scripting Channel Object References)
   > n17 = Cast To MovieSceneScriptingEventChannel!(
    execute
    Object="Get (a copy)"(Array:Array of Movie Scene Scripting Channel Object References=@n16) -> (Output:Movie Scene Scripting Channel Object Reference)
   ) -> (then, CastFailed, AsMovie Scene Scripting Event Channel:Movie Scene Scripting Event Channel Object Reference) {is=title K2Node_DynamicCast}
   > n18 = Execute Python Script!(
    execute
    PythonScript="##Get the Reference key\r\nimport unreal\r\nexamplekey = channel.get_keys()[0]\r\nexamplekey_value = examplekey.get_value()\r\nprint(examplekey_value.get_editor_property(\"payload_variables\"))\r\nprint(examplekey_value.get_editor_property(\"payload_variables\").get(\"CharacterPackageName\").get_editor_property(\"value\"))"
    channel:Movie Scene Scripting Event Channel Object Reference=@n17
   ) -> (then, ReturnValue, examplekey_value:Movie Scene Event Structure) {inputs [channel], outputs [examplekey_value]}
   > Set ReferenceKey(ReferenceKey:Movie Scene Event Structure=@n18.examplekey_value) -> (Output_Get:Movie Scene Event Structure)
   n19 = For Each Loop(Exec=@n11.then_1, Array:Array of Movie Scene Track Object References=@n10) -> (Array Element:Movie Scene Track Object Reference)
   > n20 = GetDisplayName(Track=@n19)
   > Branch(Condition="Equal, Case Insensitive (Text)"(A=@n20, B=MakeLiteralText(Value="LOCTEXT(\"Generated\")")))
   > n21 = GetSections(Track=@n19)
   > n22 = For Each Loop(Array:Array of Movie Scene Section Object References=@n21) -> (Array Element:Movie Scene Section Object Reference)
   > GetTrackRowDisplayName(Track=@n19, RowIndex=GetRowIndex(self=@n22) -> (ReturnValue=0))
   > n23 = GetAllChannels(Section=@n22)
   > n24 = Cast To MovieSceneScriptingEventChannel(Object="Get (a copy)"(Array=@n23))
   > n25 = Sequence!(execute) -> (then_0, then_1)
   > Execute Python Script!(
    execute
    PythonScript="##clear all keys\r\nimport unreal\r\nall_keys = channel.get_keys()\r\nfor key in all_keys:\r\n\tchannel.remove_key(key)"
    channel=@n24
   ) -> (then, ReturnValue) {inputs [channel]}
   Execute Python Script!(
    execute=@n11.then_2
    PythonScript="import unreal\r\nmulti_cid_sequence.set_playback_end(num_cid)"
    MultiCIDSequence:Movie Scene Sequence Object Reference=@n9
    num_cid:int=Get NumCids() -> (NumCids:int=0)
   ) -> (then, ReturnValue) {inputs [MultiCIDSequence, num_cid]}
   Return Node!(execute=@n8.else)
   n26 = Execute Python Script!(
    execute=@n2.LoopBody
    PythonScript="##Modify Reference Key with new data.\r\nimport unreal\r\nnewvalues = referencekeyvalue.get_editor_property(\"payload_variables\")\r\nfoo = newvalues['CharacterPackageName']\r\nfoo.set_editor_property(\"value\", newcharacter)\r\nnewvalues['CharacterPackageName'] = foo\r\n##print(newvalues.get(\"NewCharacter\").get_editor_property(\"value\"))\r\nreferencekeyvalue.set_editor_property(\"payload_variables\", newvalues)\r\ngeneratedkey = referencekeyvalue\r\nprint(generatedkey.get_editor_property(\"payload_variables\"))\r\nprint(generatedkey.get_editor_property(\"payload_variables\").get(\"CharacterPackageName\").get_editor_property(\"value\"))"
    newcharacter:Name=@n2
    Referencekeyvalue:Movie Scene Event Structure=Get ReferenceKey() -> (ReferenceKey:Movie Scene Event Structure)
   ) -> (then, ReturnValue, generatedkey:Movie Scene Event Structure) {inputs [newcharacter, Referencekeyvalue], outputs [generatedkey]}
   n27 = Set GeneratedKey(GeneratedKey:Movie Scene Event Structure=@n26.generatedkey) -> (Output_Get:Movie Scene Event Structure)
   > Execute Python Script!(
    execute
    PythonScript="import unreal\r\nprint(\"Second Python Block\")\r\nprint(generatedkey.get_editor_property(\"payload_variables\"))\r\nprint(generatedkey.get_editor_property(\"payload_variables\").get(\"NewCharacter\").get_editor_property(\"objectvalue\"))"
    generatedkey:Movie Scene Event Structure=@n27
   ) -> (then, ReturnValue) {inputs [generatedkey]}
   "Get Keys (Event)"!(self:Movie Scene Scripting Event Channel Object Reference) -> (ReturnValue:Array of Movie Scene Scripting Key Object References) {is=title K2Node_CallFunction}
   "Add Key (Event)"!(
    execute=@n26.then
    self:Movie Scene Scripting Event Channel Object Reference=@n24
    InTime:"Frame Number Structure (by ref)"=Make Frame Number!(Value:int=@n2.Array Index) -> (FrameNumber:Frame Number Structure) {is=title K2Node_MakeStruct}
    NewValue:Movie Scene Event Structure=@n26.generatedkey
    SubFrame:float=0
    TimeUnit:EMovieSceneTimeUnit Enum=DisplayRate
   ) -> (then, ReturnValue:Movie Scene Scripting Event Key Object Reference)
   Append!(
    A=/Script/FortniteGame.AthenaCharacterItemDefinition'
    B="To String (Name)"(InName=@n2)
    C:String=.
    D:String="To String (Name)"(InName=@n2.Array Element_AssetName)
    E:String='
   ) -> (ReturnValue)
   Execute Python Script!(
    execute
    PythonScript="import unreal\r\n\r\nkey = unreal.MovieSceneEvent()\r\n##key.set_editor_property(\"payload_variables\",map(\"string\", \"HELLOWORLDHELLOWORLD\"))\r\nkey.set_editor_property(\"weak_endpoint\", THISPRINTSAWORD)"
   ) -> (then, ReturnValue, Key:Wildcard) {outputs [Key]}
   Get ReferenceKey()
   n28 = Length(TargetArray:Array of Asset Data Structures=@n1.AsData)
   n29 = Set NumCids(execute=@n1.then, NumCids:int=@n28) -> (Output_Get:int=0)
  ]}
 GenerateSequenceManifest {nodes [
   n1 = GenerateSequenceManifest!() -> (then, KeyFrames:Array of Gbuffer TT Key Frame Data Structures) {is=title K2Node_FunctionEntry}
   > n2 = SpliceKeyframeData!(execute, self, KeyFrames:Array of Gbuffer TT Key Frame Data Structures=@n1) -> (
    then
    PassnameAll:Array of Strings
    FrameStartAll:Array of Strings
    FrameEndAll:Array of Strings
    ID_All:Array of Strings
    NameAll:Array of Strings
    Type_All:Array of Strings
   )
   > Execute Python Script!(
    execute
    PythonScript="import Developers.ChrisMorrison.ManifestGenerate as MG\r\nimport importlib\r\nimportlib.reload(MG)\r\n\r\nMG.write_manifest(dirpath,in_filename, passname_all, frame_start_all, frame_end_all, id_all, name_all, type_all)"
    dirpath:String=Get!() -> (OutputDirectory_Path:String)
    passname_all:Array of Strings=@n2
    frame_start_all:Array of Strings=@n2.FrameStartAll
    frame_end_all:Array of Strings=@n2.FrameEndAll
    id_all:Array of Strings=@n2.ID_All
    name_all:Array of Strings=@n2.NameAll
    type_all:Array of Strings=@n2.Type_All
    in_filename:String=Split(SourceString=Get!() -> (MRQ_Sequence_PathString:String), InStr=., SearchDir=FromEnd).RightS
   ) -> (then, ReturnValue) {inputs [dirpath, passname_all, frame_start_all, frame_end_all, id_all, name_all, type_all, in_filename]}
  ]}
 GetSequenceShotKeyData {nodes [
   n1 = Length(TargetArray:Array of Wildcards)
   PrintString(InString=Append!(A="Num Sections: ", B="To String (Integer)"(InInt=@n1)) -> (ReturnValue))
   > n2 = For Each Loop()
   n3 = GetSequenceShotKeyData!() -> (then, MRQ_Sequence:Soft Object Path Structure) {is=title K2Node_FunctionEntry}
   > n4 = LoadAsset_Blocking(Asset=To Soft Object Reference(SoftObjectPath=@n3))
   > n5 = Cast To LevelSequence!(execute, Object=@n4) -> (then, CastFailed, AsLevel Sequence:Level Sequence Object Reference) {is=title K2Node_DynamicCast}
   > n6 = GetBindings(Sequence=@n5)
   > n7 = For Each Loop(Array:Array of Movie Scene Binding Proxy Structures=@n6) -> (Array Element:Movie Scene Binding Proxy Structure)
   > n8 = GetTracks!(execute, InBinding:"Movie Scene Binding Proxy Structure (by ref)"=@n7) -> (then, ReturnValue)
   > n9 = For Each Loop(Array:Array of Movie Scene Track Object References=@n8) -> (Array Element:Movie Scene Track Object Reference)
   > n10 = GetDisplayName(Track=@n9)
   > Branch(Condition="Equal, Case Insensitive (Text)"(A=@n10, B=MakeLiteralText(Value="LOCTEXT(\"ManifestTracks\")")))
   > n11 = GetSections(Track=@n9)
   > n12 = For Each Loop(Array:Array of Movie Scene Section Object References=@n11) -> (Array Element:Movie Scene Section Object Reference)
   > n13 = GetTrackRowDisplayName(Track=@n9, RowIndex=GetRowIndex(self=@n12) -> (ReturnValue=0))
   > Branch(Condition="Equal, Case Insensitive (Text)"(A=@n13, B=MakeLiteralText(Value="LOCTEXT(\"ManifestTrackEvents\")")))
   > n14 = GetAllChannels(Section=@n12)
   > n15 = Cast To MovieSceneScriptingEventChannel(Object="Get (a copy)"(Array=@n14))
   > n16 = CompileKeyData!(execute, self, Keys:Array of Movie Scene Scripting Key Object References="Get Keys (Event)"(self=@n15)) -> (then, KeyFrameSections1:Array of Gbuffer TT Key Frame Data Structures) {is=title K2Node_CallFunction}
   > n17 = SpliceKeyframeData(KeyFrames=@n16)
   > n18 = For Each Loop(Array:Array of Strings=@n17) -> (Array Element:String)
   > PrintString(InString=Append!(
     A=@n18
     B="\r\n~"
     C:String="Get (a copy)"(Array:Array of Strings=@n17.FrameStartAll, Dimension 1=@n18.Array Index) -> (Output:String)
     D:String="\r\n~"
     E:String="Get (a copy)"(Array=@n17.FrameEndAll, Dimension 1=@n18.Array Index)
     F:String="\r\n~"
     G:String="Get (a copy)"(Array=@n17.ID_All, Dimension 1=@n18.Array Index)
     H:String="\r\n~"
     I:String="Get (a copy)"(Array=@n17.NameAll, Dimension 1=@n18.Array Index)
     J:String="\r\n~"
     K:String="Get (a copy)"(Array=@n17.Type_All, Dimension 1=@n18.Array Index)
    ) -> (ReturnValue))
   GetDisplayName!(Object:Object Reference=@n2) -> (ReturnValue:String)
   n19 = Length()
   PrintString(InString=Append!(A="Section: ", B="To String (Integer)"(InInt=@n2.Array Index), C="  --- Num Channels: ", D="To String (Integer)"(InInt=@n19)) -> (ReturnValue))
   Execute Python Script!(execute, PythonScript="import unreal\r\n\r\nname = channel.get_editor_property(\"channel_name\")", Channel:Wildcard) -> (then, ReturnValue, Name:Wildcard) {inputs [Channel], outputs [Name]}
  ]}
 RenderSequence {nodes [
   n1 = Cast To MoviePipelineConsoleVariableSetting!(execute, Object:Wildcard) -> (then, CastFailed, AsMovie Pipeline Console Variable Setting:Movie Pipeline Console Variable Setting Object Reference) {is=title K2Node_DynamicCast}
   n2 = Get ConsoleVariables!(self:Movie Pipeline Console Variable Setting Object Reference=@n1) -> (ConsoleVariables:"Map of Strings to Float (single-precision)s")
   n3 = Keys!(execute, TargetMap:"Map of Strings to Float (single-precision)s"=@n2) -> (then, Keys:Array of Strings)
   > For Each Loop(Array=@n3)
   n4 = PrintString(InString="~~~~~~Just Prior To Set Sequence~~~~~~~")
   n5 = Get MoviePipelineQueueSubsystem() -> (ReturnValue:Movie Pipeline Queue Subsystem Object Reference) {is=title K2Node_GetEditorSubsystem}
   n6 = GetQueue!(self:Movie Pipeline Queue Subsystem Object Reference=@n5) -> (ReturnValue:Movie Pipeline Queue Object Reference)
   n7 = AllocateNewJob!(
    execute=@n17.then_2
    self:Movie Pipeline Queue Object Reference=@n6
    InJobType:Movie Pipeline Executor Job Class Reference=/Script/MovieRenderPipelineCore.MoviePipelineExecutorJob
   ) -> (then, ReturnValue:Movie Pipeline Executor Job Object Reference)
   Set Sequence!(
    execute=@n4.then
    Sequence:Soft Object Path Structure=MakeSoftObjectPath!(PathString:String=Get Object Path String()) -> (ReturnValue:Soft Object Path Structure)
    self:Movie Pipeline Executor Job Object Reference=@n7
   ) -> (then, Output_Get:Soft Object Path Structure)
   > PrintString(InString="~~~~~~Just After Set Sequence~~~~~~~")
   n8 = TT_Func_GetAssetIfValid!(
    execute
    SoftObjectReference:"Soft Object Reference (by ref)"=Get FN_SettingPresets() -> (FN_SettingPresets:Soft Object Reference)
    __WorldContext:Object Reference
   ) -> (then, Is Valid?:bool, ReturnValue:Object Reference)
   > n9 = Branch(Condition=@n8)
   > n10 = Cast To MoviePipelinePrimaryConfig!(execute, Object=@n8.ReturnValue) -> (then, CastFailed, AsMovie Pipeline Primary Config:Movie Pipeline Primary Config Object Reference) {is=title K2Node_DynamicCast}
   > n11 = SetConfiguration!(execute, self:Movie Pipeline Executor Job Object Reference=@n7, InPreset:Movie Pipeline Primary Config Object Reference=@n10) -> (then)
   n12 = GetConfiguration!(self:Movie Pipeline Executor Job Object Reference=@n7) -> (ReturnValue:Movie Pipeline Primary Config Object Reference)
   n13 = FindOrAddSettingByClass!(
    execute=[@n11.then, @n14.Completed]
    self:Movie Pipeline Config Base Object Reference=@n12
    InClass:Movie Pipeline Setting Class Reference=/Script/MovieRenderPipelineCore.MoviePipelineOutputSetting
    bIncludeDisabledSettings:bool=false
    bExactMatch:bool=false
   ) -> (then, ReturnValue:Movie Pipeline Output Setting Object Reference)
   > Set OutputDirectory!(
    execute
    OutputDirectory:Directory Path Structure=Get OutputDirectory() -> (OutputDirectory:Directory Path Structure)
    self:Movie Pipeline Output Setting Object Reference=@n13
   ) -> (then, Output_Get:Directory Path Structure)
   SetConfiguration(
    execute=[@n10.CastFailed, @n9.else, @n21.then]
    self=@n7
    InPreset=Get MRQPresets() -> (MRQPresets:Movie Pipeline Primary Config Object Reference)
   )
   > n14 = For Each Loop(Array:Array of Movie Pipeline Setting Object References=GetAllSettings!(self:Movie Pipeline Primary Config Object Reference=@n12, bIncludeDisabledSettings:bool=false, bIncludeTransientSettings:bool=false) -> (ReturnValue:Array of Movie Pipeline Setting Object References)) -> (Array Element:Movie Pipeline Setting Object Reference)
   > n15 = Cast To MoviePipelineAntiAliasingSetting!(execute, Object=@n14) -> (then, CastFailed, AsMovie Pipeline Anti Aliasing Setting:Movie Pipeline Anti Aliasing Setting Object Reference) {is=title K2Node_DynamicCast}
   > n16 = Switch on ECheckBoxState(Selection=GetCheckedState(self=Get EnableTAA_Checkbox() -> (EnableTAA_Checkbox:Check Box Object Reference)) -> (ReturnValue=Unchecked))
   Set bOverrideAntiAliasing!(execute=[@n16.Checked, @n16.Undetermined], bOverrideAntiAliasing:bool=false, self:Movie Pipeline Anti Aliasing Setting Object Reference=@n15) -> (then, Output_Get:bool)
   Set bOverrideAntiAliasing(execute=@n16.Unchecked, bOverrideAntiAliasing=true, self=@n15)
   > Set AntiAliasingMethod!(execute, AntiAliasingMethod:EAntiAliasingMethod Enum=AAM_None, self:Movie Pipeline Anti Aliasing Setting Object Reference=@n15) -> (then, Output_Get:EAntiAliasingMethod Enum=AAM_None)
   > PrintString(InString="~~~~~~~~~~~~~~~~~ Overriding AntiAliasing to None")
   RenderSequence() {is=title K2Node_FunctionEntry}
   > n17 = Sequence!(execute) -> (then_0, then_1, then_2:Exec, then_3:Exec)
   n18 = For Each Loop(
    Exec=@n17.then_1
    Array:Array of Movie Pipeline Executor Job Object References=GetJobs!(self:Movie Pipeline Queue Object Reference=@n6) -> (ReturnValue:Array of Movie Pipeline Executor Job Object References)
   ) -> (Array Element:Movie Pipeline Executor Job Object Reference)
   > DeleteJob!(execute, self:Movie Pipeline Queue Object Reference=@n6, InJob:Movie Pipeline Executor Job Object Reference=@n18) -> (then)
   RenderQueueWithExecutor!(
    execute=@n17.then_3
    self:Movie Pipeline Queue Subsystem Object Reference=@n5
    InExecutorType:Movie Pipeline Executor Base Class Reference=/Script/MovieRenderPipelineEditor.MoviePipelinePIEExecutor
   ) -> (then, ReturnValue:Movie Pipeline PIEExecutor Object Reference)
   n19 = Switch on String!(execute=@n17.then_0, Selection=GetSelectedOption(self=Get SequenceSelector() -> (SequenceSelector:"Combo Box (String) Object Reference"))) -> (Default, Standard Characters, Toon Characters, Modelling Sequence, Turntable Status Verification, Morph Target, Vehicles, ExecBeautyShots) {pinNames [Standard Characters, Toon Characters, Modelling Sequence, Turntable Status Verification, Morph Target, Vehicles, ExecBeautyShots]}
   Set MRQPresets(
    execute=[
     @n19.Default
     @n19.Modelling Sequence
     @n19.Morph Target
     @n19.Standard Characters
     @n19.Toon Characters
     @n19.Turntable Status Verification
     @n19.Vehicles
    ]
    MRQPresets:Movie Pipeline Primary Config Object Reference=/Gbuffer_Turntable/FN_PublicTurntable_Assets/MRQ_Assets/Gbuffer_TT_Settings_Default
   ) -> (Output_Get:Movie Pipeline Primary Config Object Reference)
   Set MRQPresets(execute=@n19.ExecBeautyShots, MRQPresets=/Gbuffer_Turntable/FN_PublicTurntable_Assets/MRQ_Assets/MPPC_CharacterBeautyShot)
   GetCurrentShotFrameNumber!(InMoviePipeline:Movie Pipeline Object Reference) -> (ReturnValue_Value:int)
   Add!(execute, TargetMap:"Map of Strings to Float (single-precision)s"=@n2, Key:"String (by ref)", Value:"Float (single-precision) (by ref)") -> (then) {is=title K2Node_CallFunction}
   Values!(execute, TargetMap:Map of Wildcards to Wildcards) -> (then, Values:Array of Wildcards)
   Add(execute=@n1.then, TargetMap=@n2, Key=r.Lumen.Reflections.Allow)
   PrintString(InString=Hello)
   PrintString(InString="~~~~~~Just Prior To Generate Sequence~~~~~~~")
   n20 = GenerateCustomSequence_Morph!(execute=@n22.then, self) -> (then, HasValidTargets:bool) {is=title K2Node_CallFunction}
   Set Sequence(execute=[@n20.then, @n22.else], Sequence=Get MRQ_Sequence(), self=@n7)
   > n21 = Set Map!(execute, Map:Soft Object Path Structure=Get Map() -> (Map:Soft Object Path Structure), self:Movie Pipeline Executor Job Object Reference=@n7) -> (then, Output_Get:Soft Object Path Structure)
   n22 = Branch(execute=@n23.then, Condition=Get GenerateCustomSequence() -> (GenerateCustomSequence:bool))
   n23 = Set JobName!(execute=@n7.then, JobName:String=Gbuffer_Turntable_Render, self:Movie Pipeline Executor Job Object Reference=@n7) -> (then, Output_Get:String)
  ]}
 RunDebugSequence {nodes [
   RunDebugSequence() {is=title K2Node_FunctionEntry}
   > n1 = Sequence!(execute) -> (then_0, then_1, then_2)
   n2 = Get MoviePipelineQueueSubsystem()
   n3 = GetQueue(self=@n2)
   > n4 = For Each Loop(Array=GetJobs(self=@n3))
   > DeleteJob(self=@n3, InJob=@n4)
   n5 = AllocateNewJob(execute=@n1.then_1, self=@n3, InJobType=&MoviePipelineExecutorJob)
   > Set JobName(JobName=Gbuffer_Turntable_Render, self=@n5)
   > Set Sequence(
    Sequence="Get (a copy)"(Array:Array of Soft Object Path Structures=Get MRQ_Sequence_MasterList(), Dimension 1=3) -> (Output:Soft Object Path Structure)
    self=@n5
   )
   > Set Map(Map=Get Map(), self=@n5)
   > n6 = TT_Func_GetAssetIfValid(SoftObjectReference=Get FN_SettingPresets())
   > n7 = Branch(Condition=@n6)
   > n8 = Cast To MoviePipelinePrimaryConfig(Object=@n6.ReturnValue)
   > n9 = SetConfiguration(self=@n5, InPreset=@n8)
   n10 = FindOrAddSettingByClass(
    execute=[@n11.then, @n9.then]
    self=GetConfiguration(self=@n5)
    InClass=&MoviePipelineOutputSetting
    bIncludeDisabledSettings=false
    bExactMatch=false
   )
   > Set OutputDirectory(OutputDirectory=Get OutputDirectory(), self=@n10)
   n11 = SetConfiguration(
    execute=[@n8.CastFailed, @n7.else]
    self=@n5
    InPreset=Get SettingPresets() -> (SettingPresets:Movie Pipeline Primary Config Object Reference)
   )
   RenderQueueWithExecutor(execute=@n1.then_2, self=@n2, InExecutorType=&MoviePipelinePIEExecutor)
   PrintString(InString=GetDisplayName(Object=@n4))
   n12 = GetAllSettings(self=@n8, bIncludeDisabledSettings=false, bIncludeTransientSettings=false)
   Set ConsoleVariables!(
    execute
    ConsoleVariables:"Map of Strings to Float (single-precision)s"
    self:Movie Pipeline Console Variable Setting Object Reference=Cast To MoviePipelineConsoleVariableSetting(Object:Object Reference="Get (a copy)"(Array:Array of Movie Pipeline Setting Object References=@n12) -> (Output:Movie Pipeline Setting Object Reference))
   ) -> (then, Output_Get:"Map of Strings to Float (single-precision)s")
  ]}
 SetScalability {nodes [
   n1 = SetScalability!() -> (then, Scalability_Low:bool=~, ForceScalability:bool=~) {is=title K2Node_FunctionEntry}
   > n2 = GetAllActorsOfClass(ActorClass=&BP_Turntable_GBuffers_C)
   > Branch(Condition=Is Not Empty(TargetArray:Array of BP Turntable GBuffers Object References=@n2))
   > n3 = Branch(Condition=@n1.ForceScalability)
   > n4 = Branch(Condition=@n1)
   n5 = "Get (a copy)"(Array:Array of BP Turntable GBuffers Object References=@n2) -> (Output:BP Turntable GBuffers Object Reference)
   > n6 = Set Scalability!(execute, Scalability:String=LOW, self:BP Turntable GBuffers Object Reference=@n5) -> (then, Output_Get:String)
   PrintString(execute=[@n6.then, @n7.then, @n8.then], InString=SCALABILITY SET)
   n7 = Set Scalability(execute=@n4.else, Scalability=EPIC, self=@n5)
   n8 = Set Scalability(execute=@n3.else, Scalability=Default, self=@n5)
  ]}
 SetShotData {nodes [
   n1 = SetShotData!() -> (then, PassType:int) {is=title K2Node_FunctionEntry}
   > n2 = Switch on Int!(execute:Exec, Selection:int=@n1) -> (Default:Exec, 0:Exec, 1:Exec, 2:Exec, 3:Exec)
   Set GenerateData?(execute=[@n2.0, @n2.Default], GenerateData?:bool=false) -> (Output_Get:bool)
   n3 = Set SelectedShotData(
    execute=@n2.1
    SelectedShotData:Array of Gbuffer TT Key Frame Data Structures=Get ShotData_Standard() -> (ShotData_Standard:Array of Gbuffer TT Key Frame Data Structures)
   ) -> (Output_Get:Array of Gbuffer TT Key Frame Data Structures)
   Set GenerateData?(execute=[@n3.then, @n5.then, @n4.then], GenerateData?=true)
   n4 = Set SelectedShotData(execute=@n2.2, SelectedShotData=Get ShotData_Toon() -> (ShotData_Toon:Array of Gbuffer TT Key Frame Data Structures))
   n5 = Set SelectedShotData(execute=@n2.3, SelectedShotData=Get ShotData_Flipbook() -> (ShotData_Flipbook:Array of Gbuffer TT Key Frame Data Structures))
  ]}
 SpliceKeyframeData {nodes [
   n1 = SpliceKeyframeData!() -> (then, KeyFrames:Array of Gbuffer TT Key Frame Data Structures) {is=title K2Node_FunctionEntry}
   > n2 = For Each Loop!(Exec, Array:Array of Gbuffer TT Key Frame Data Structures=@n1) -> (
    LoopBody
    Array Element_Passname_2_#1:String
    Array Element_FrameStart_6_#2:String
    Array Element_FrameEnd_8_#3:String
    Array Element_ID_10_#4:String
    Array Element_Name_12_#5:String
    Array Element_Type_15_#6:String
    Array Index
    Completed
   )
   Return Node!(
    execute=@n2.Completed
    PassnameAll:Array of Strings=Get PassNames() -> (PassNames:Array of Strings)
    FrameStartAll:Array of Strings=Get FrameStarts() -> (FrameStarts:Array of Strings)
    FrameEndAll:Array of Strings=Get FrameEnds() -> (FrameEnds:Array of Strings)
    ID_All:Array of Strings=Get IDs() -> (IDs:Array of Strings)
    NameAll:Array of Strings=Get Names() -> (Names:Array of Strings)
    Type_All:Array of Strings=Get Types() -> (Types:Array of Strings)
   )
   > n3 = Sequence!(execute) -> (then_0, then_1, then_2, then_3:Exec, then_4:Exec, then_5:Exec)
   > Add!(execute, TargetArray:Array of Strings=Get PassNames(), NewItem:"String (by ref)"=@n2) -> (then, ReturnValue) {is=title K2Node_CallArrayFunction}
   Add(execute=@n3.then_1, TargetArray=Get FrameStarts(), NewItem=@n2.Array Element_FrameStart_6_#2)
   Add(execute=@n3.then_2, TargetArray=Get FrameEnds(), NewItem=@n2.Array Element_FrameEnd_8_#3)
   Add(execute=@n3.then_3, TargetArray=Get IDs(), NewItem=@n2.Array Element_ID_10_#4)
   Add(execute=@n3.then_4, TargetArray=Get Names(), NewItem=@n2.Array Element_Name_12_#5)
   Add(execute=@n3.then_5, TargetArray=Get Types(), NewItem=@n2.Array Element_Type_15_#6)
  ]}
 UpdateRotationSliderPosition {nodes [
   UpdateRotationSliderPosition() {is=title K2Node_FunctionEntry}
   n1 = Get Actor Reference()
   > SetValue!(
    execute
    self:Slider Object Reference=Get Rotation_Slider()
    InValue:float="% (Float)"!(A:double=Get Actor Rotation(self=@n1).ReturnValue_Yaw, B:double=180) -> (ReturnValue:double)
   ) -> (then)
  ]}
 UpdateTODM {nodes [
   n1 = UpdateTODM!() -> (then, InTimeOfDayManager:Info Soft Class Reference, bVisible:bool=~, UpdateRotation?:bool=~, NewRotation:double=~) {is=title K2Node_FunctionEntry}
   > n2 = GetAllActorsOfClass(ActorClass=/Gbuffer_Turntable/FN_PublicTurntable_Assets/LevelsChecker/BP_LevelsChecker.BP_LevelsChecker_C) -> (OutActors:Array of BP Levels Checker Object References)
   > n3 = Branch(Condition=@n1.bVisible)
   n4 = "Get (a copy)"(Array:Array of BP Levels Checker Object References=@n2) -> (Output:BP Levels Checker Object Reference)
   > EnableTOD(self:BP Levels Checker Object Reference=@n4)
   > Branch(Condition=@n1.UpdateRotation?)
   > n5 = GetAllActorsOfClass(ActorClass=&FortDaySequenceActor) -> (OutActors:Array of Fort Day Sequence Actor Object References)
   > Set Actor Rotation(
    self="Get (a copy)"(Array:Array of Fort Day Sequence Actor Object References=@n5) -> (Output:Fort Day Sequence Actor Object Reference)
    NewRotation_Roll=0
    NewRotation_Pitch=0
    NewRotation_Yaw=@n1.NewRotation
    bTeleportPhysics=false
   )
   DisableTOD(execute=@n3.else, self:BP Levels Checker Object Reference=@n4)
   GetGlobalTimeOfDayManager()
  ]}
]
variables [
 Actor Reference {type=Actor, default=null}
 AnimationToApply {type=boolean}
 CameraLocation_Backpack {type=Transform, default {location=<0.0001 150 50>, rotation=<rot -9.5 -90.2 0>, scale=<1 1 1>}}
 CameraLocation_Standard {type=Transform, default {location=<0.0001 539.7187 175>, rotation=<rot -9.5 -90.2 0>, scale=<1 1 1>}}
 CollectionName {type=string, default=1_CIDTestList}
 ColorCalibrator_Backpack {type=Transform, default {location=<40 -23 9>, rotation=<rot 0 -30 0>, scale=<0.2 0.2 0.2>}}
 ColorCalibrator_Standard {type=Transform, default {location=<133 -23 32>, rotation=<rot 0 -30 0>, scale=<0.8 0.8 0.8>}}
 ColorCube {type=StaticMesh, default=/Engine/EditorMeshes/ColorCalibrator/SM_ColorCalibrator}
 Currently Loaded {type=string, default=ThreePointLighting}
 DUMPGPUCheckboxState {type=boolean}
 FrameNumberInAnimation {type=integer}
 GenerateCustomSequence {type=boolean}
 GenerateData? {type=boolean, default=true}
 IsMultiCID {type=boolean}
 LevelColorCube {type=StaticMeshActor, default=null}
 LevelSequences {
  type="array<LevelSequence>"
  default [
   /Gbuffer_Turntable/FN_PublicTurntable_Assets/Sequences/Turntable_Sequence_GBuffers_Standard
   /Gbuffer_Turntable/FN_PublicTurntable_Assets/Sequences/Turntable_Sequence_Gbuffers_Toon
   /Gbuffer_Turntable/FN_PublicTurntable_Assets/Sequences/Turntable_Sequence_Gbuffers_Modeling
   /Gbuffer_Turntable/FN_PublicTurntable_Assets/Sequences/Turntable_Sequence_SelfCheck_Record
   /Gbuffer_Turntable/FN_PublicTurntable_Assets/Sequences/Turntable_Sequence_GBuffers_MorphTargetTest
   /Gbuffer_Turntable/FN_PublicTurntable_Assets/Sequences/Turntable_Sequence_GBuffers_Vehicle
   /Gbuffer_Turntable/FN_PublicTurntable_Assets/Sequences/Turntable_Sequence_GBuffers_ExecBeauty
   /Gbuffer_Turntable/FN_PublicTurntable_Assets/Sequences/Turntable_Sequence_GBuffers_Flipbook
   &Turntable_Sequence_GBuffers_MultiCID
   /Gbuffer_Turntable/FN_PublicTurntable_Assets/Sequences/Turntable_Sequence_Gbuffers_Toon_CurrentTODM
   /Gbuffer_Turntable/FN_PublicTurntable_Assets/Sequences/Turntable_Sequence_GBuffers_Standard_Headshot
   /Gbuffer_Turntable/FN_PublicTurntable_Assets/Sequences/Turntable_Sequence_Gbuffers_Toon_Headshot
  ]
 }
 MeshOverride_SafeToReplace {type=boolean}
 MRQ_Sequence {type=SoftObjectPath, default=&Turntable_Sequence_GBuffers_Standard}
 MRQ_Sequence_MasterList {
  type="array<SoftObjectPath>"
  default [
   &Turntable_Sequence_GBuffers_Standard
   &Turntable_Sequence_Gbuffers_Toon
   &Turntable_Sequence_Gbuffers_Modeling
   /Gbuffer_Turntable/FN_PublicTurntable_Assets/Sequences/Turntable_Sequence_RenderDebug
   &Turntable_Sequence_GBuffers_MorphTargetTest
   &Turntable_Sequence_SelfCheck_Record
   &Turntable_Sequence_GBuffers_Vehicle
   &Turntable_Sequence_GBuffers_ExecBeauty
   &Turntable_Sequence_GBuffers_Flipbook
   &Turntable_Sequence_GBuffers_MultiCID
   &Turntable_Sequence_Gbuffers_Toon_CurrentTODM
   &Turntable_Sequence_GBuffers_Standard_Headshot
   &Turntable_Sequence_Gbuffers_Toon_Headshot
  ]
 }
 MRQPresets {type=MoviePipelinePrimaryConfig, default=&Gbuffer_TT_Settings_Default}
 OutputDirectory {type=DirectoryPath, default {path=C:/DEFAULTDONTUSE}}
 OverrideMesh {type="array<Object>", default []}
 RotationSliderIsEnabled {type=boolean}
 RotationSliderUserIsAdjusting {type=boolean}
 SelectedShotData {
  type="array<GbufferTT_KeyFrameData>"
  default [
   Studio Lighting {type=Shot, frameEnd=179, frameStart=0, iD=1, passname=Studio_Lighting}
   Base Color {type=Shot, frameEnd=209, frameStart=180, iD=2, passname=BaseColor}
   Roughness {type=Shot, frameEnd=239, frameStart=210, iD=3, passname=Roughness}
   Specular {type=Shot, frameEnd=269, frameStart=240, iD=4, passname=Specular}
   Metallic {type=Shot, frameEnd=299, frameStart=270, iD=5, passname=Metallic}
   Emissive {type=Shot, frameEnd=329, frameStart=300, iD=6, passname=Emissive}
   BR Art Lighting {type=Shot, frameEnd=359, frameStart=330, iD=7, passname=BR_Art_Lighting}
   Front End Lighting {type=Shot, frameEnd=389, frameStart=360, iD=8, passname=FrontEndLighting}
  ]
 }
 ShotData_Flipbook {
  type="array<GbufferTT_KeyFrameData>"
  default [
   FlipbookFaces {type=Shot, frameEnd=27, frameStart=0, iD=1, passname=FlipbookFaces}
   Mime {type=Shot, frameEnd=423, frameStart=28, iD=2, passname=Mime}
   RageQuit {type=Shot, frameEnd=573, frameStart=424, iD=3, passname=RageQuit}
  ]
 }
 ShotData_Standard {
  type="array<GbufferTT_KeyFrameData>"
  default [
   Studio Lighting {type=Shot, frameEnd=179, frameStart=0, iD=1, passname=Studio_Lighting}
   Base Color {type=Shot, frameEnd=209, frameStart=180, iD=2, passname=BaseColor}
   Roughness {type=Shot, frameEnd=239, frameStart=210, iD=3, passname=Roughness}
   Specular {type=Shot, frameEnd=269, frameStart=240, iD=4, passname=Specular}
   Metallic {type=Shot, frameEnd=299, frameStart=270, iD=5, passname=Metallic}
   Emissive {type=Shot, frameEnd=329, frameStart=300, iD=6, passname=Emissive}
   Character Review Lighting {type=Shot, frameEnd=359, frameStart=330, iD=7, passname=Character Review Lighting}
   DaySequence_Current {type=Shot, frameEnd=389, frameStart=360, iD=8, passname=DaySequence_Current}
   Front End Lighting {type=Shot, frameEnd=419, frameStart=390, iD=9, passname=FrontEndLighting}
  ]
 }
 ShotData_Toon {
  type="array<GbufferTT_KeyFrameData>"
  default [
   Character Review Lighting {type=Shot, frameEnd=89, frameStart=0, iD=1, passname=Character Review Lighting}
   Current TODM Lighting {type=Shot, frameEnd=179, frameStart=90, iD=2, passname=Current TODM Lighting}
   Base Color {type=Shot, frameEnd=209, frameStart=180, iD=3, passname=BaseColor}
   Roughness {type=Shot, frameEnd=239, frameStart=210, iD=4, passname=Roughness}
   SSC Scheen {type=Shot, frameEnd=269, frameStart=240, iD=5, passname=SSC_Scheen}
   SSC AO Threshold {type=Shot, frameEnd=299, frameStart=270, iD=6, passname=SSC_AO_Threshold}
   Material AO {type=Shot, frameEnd=329, frameStart=300, iD=7, passname=Material_AO}
   Seperate Translucency {type=Shot, frameEnd=419, frameStart=330, iD=8, passname=Seperate_Translucency}
   Emissive {type=Shot, frameEnd=449, frameStart=420, iD=9, passname=Emissive}
  ]
 }
 StarterDefaultActorObjects {type="array<Object>", default []}
 StarterDefaultActors {type="array<Actor>"}
 TurntableSpinCenter {type=Actor, default=null}
 VersionInfo {type=GbufferTT_VersionInfo, default {current Version=2.5}}
]
widgets {
 info {rootWidgetClass=&Overlay, widgetCount=76}
 widgets [
  Overlay_36 {widgetClassPath=&Overlay, widget=Overlay_36}
  VerticalBox_5 {
   widgetClassPath=&VerticalBox
   parent=Overlay_36
   widget=VerticalBox_5
   slot {class=OverlaySlot, properties {horizontalAlignment=HAlign_Fill, verticalAlignment=VAlign_Fill}}
  }
  VerticalBox_72 {widgetClassPath=&VerticalBox, parent=VerticalBox_5, widget=VerticalBox_72, slot {class=VerticalBoxSlot}}
  HorizontalBox_3 {widgetClassPath=&HorizontalBox, parent=VerticalBox_72, widget=HorizontalBox_3, slot {class=VerticalBoxSlot}}
  TextBlock_126 {
   widgetClassPath=&TextBlock
   bIsVariable=true
   parent=HorizontalBox_3
   widget=TextBlock_126
   properties {
    text=Turntable Controls V2.0
    font {
     bForceMonospaced=false
     bMaterialIsStencil=false
     fontMaterial=null
     fontObject=&Roboto
     letterSpacing=0
     monospacedWidth=1
     size=12
     skewAmount=0
     typefaceFontName=Bold
     outlineSettings {
      bApplyOutlineToDropShadows=false
      bMiteredCorners=false
      bSeparateFillAlpha=false
      outlineMaterial=null
      outlineSize=0
      outlineColor=<rgba 0 0 0 1>
     }
    }
   }
   slot {class=HorizontalBoxSlot}
  }
  HorizontalBox_2 {
   widgetClassPath=&HorizontalBox
   parent=VerticalBox_72
   widget=HorizontalBox_2
   slot {class=VerticalBoxSlot, properties {verticalAlignment=VAlign_Top, size {sizeRule=Fill, value=1}}}
  }
  VerticalBox_6 {
   widgetClassPath=&VerticalBox
   parent=HorizontalBox_2
   widget=VerticalBox_6
   slot {class=HorizontalBoxSlot, properties {horizontalAlignment=HAlign_Center}}
  }
  HorizontalBox_0 {widgetClassPath=&HorizontalBox, parent=VerticalBox_6, widget=HorizontalBox_0, slot {class=VerticalBoxSlot}}
  LightingModeSelector {
   widgetClassPath=/Script/UMG.ComboBoxString
   bIsVariable=true
   parent=HorizontalBox_0
   widget=LightingModeSelector
   properties {
    selectedOption=3 Point Lighting
    defaultOptions [3 Point Lighting, BR_Art Lighting, Front End Lighting]
    foregroundColor {colorUseRule=UseColor_Specified, specifiedColor=<rgba 1 1 1 1>}
    itemStyle {
     bUseParentRowBrush=false
     activeBrush {
      bIsDynamicallyLoaded=false
      drawAs=Image
      imageType=NoImage
      mirroring=NoMirror
      resourceName=None
      resourceObject=null
      tiling=NoTile
      imageSize=<0 0>
      margin=<ltrb 0 0 0 0>
      outlineSettings {
       bUseBrushTransparency=false
       roundingType=HalfHeightRadius
       width=0
       color {colorUseRule=UseColor_Specified, specifiedColor=<rgba 0 0 0 0>}
       cornerRadii=<0 0 0 0>
      }
      tintColor {colorUseRule=UseColor_Specified, specifiedColor=<rgba 0 0.162 0.7454 1>}
      uVRegion {bIsValid=false, max=<0 0>, min=<0 0>}
     }
     activeHighlightedBrush {
      bIsDynamicallyLoaded=false
      drawAs=Image
      imageType=NoImage
      mirroring=NoMirror
      resourceName=None
      resourceObject=null
      tiling=NoTile
      imageSize=<0 0>
      margin=<ltrb 0 0 0 0>
      outlineSettings {
       bUseBrushTransparency=false
       roundingType=HalfHeightRadius
       width=0
       color {colorUseRule=UseColor_Specified, specifiedColor=<rgba 0 0 0 0>}
       cornerRadii=<0 0 0 0>
      }
      tintColor {colorUseRule=UseColor_Specified, specifiedColor=<rgba 0.0252 0.0319 0.0423 1>}
      uVRegion {bIsValid=false, max=<0 0>, min=<0 0>}
     }
     activeHoveredBrush {
      bIsDynamicallyLoaded=false
      drawAs=Image
      imageType=NoImage
      mirroring=NoMirror
      resourceName=None
      resourceObject=null
      tiling=NoTile
      imageSize=<0 0>
      margin=<ltrb 0 0 0 0>
      outlineSettings {
       bUseBrushTransparency=false
       roundingType=HalfHeightRadius
       width=0
       color {colorUseRule=UseColor_Specified, specifiedColor=<rgba 0 0 0 0>}
       cornerRadii=<0 0 0 0>
      }
      tintColor {colorUseRule=UseColor_Specified, specifiedColor=<rgba 0 0.162 0.7454 1>}
      uVRegion {bIsValid=false, max=<0 0>, min=<0 0>}
     }
     dropIndicator_Above {
      bIsDynamicallyLoaded=false
      drawAs=Box
      imageType=FullColor
      mirroring=NoMirror
      resourceName=../../../Engine/Content/Slate/Common/DropZoneIndicator_Above.png
      resourceObject=null
      tiling=NoTile
      imageSize=<0 0>
      margin=<ltrb 0.625 0.625 0 0>
      outlineSettings {
       bUseBrushTransparency=false
       roundingType=HalfHeightRadius
       width=0
       color {colorUseRule=UseColor_Specified, specifiedColor=<rgba 0 0 0 0>}
       cornerRadii=<0 0 0 0>
      }
      tintColor {colorUseRule=UseColor_Specified, specifiedColor=<rgba 0.0091 0.3515 0.9647 1>}
      uVRegion {bIsValid=false, max=<0 0>, min=<0 0>}
     }
     dropIndicator_Below {
      bIsDynamicallyLoaded=false
      drawAs=Box
      imageType=FullColor
      mirroring=NoMirror
      resourceName=../../../Engine/Content/Slate/Common/DropZoneIndicator_Below.png
      resourceObject=null
      tiling=NoTile
      imageSize=<0 0>
      margin=<ltrb 0.625 0 0 0.625>
      outlineSettings {
       bUseBrushTransparency=false
       roundingType=HalfHeightRadius
       width=0
       color {colorUseRule=UseColor_Specified, specifiedColor=<rgba 0 0 0 0>}
       cornerRadii=<0 0 0 0>
      }
      tintColor {colorUseRule=UseColor_Specified, specifiedColor=<rgba 0.0091 0.3515 0.9647 1>}
      uVRegion {bIsValid=false, max=<0 0>, min=<0 0>}
     }
     dropIndicator_Onto {
      bIsDynamicallyLoaded=false
      drawAs=Box
      imageType=FullColor
      mirroring=NoMirror
      resourceName=../../../Engine/Content/Slate/Common/DropZoneIndicator_Onto.png
      resourceObject=null
      tiling=NoTile
      imageSize=<0 0>
      margin=<ltrb 0.25 0.25 0.25 0.25>
      outlineSettings {
       bUseBrushTransparency=false
       roundingType=HalfHeightRadius
       width=0
       color {colorUseRule=UseColor_Specified, specifiedColor=<rgba 0 0 0 0>}
       cornerRadii=<0 0 0 0>
      }
      tintColor {colorUseRule=UseColor_Specified, specifiedColor=<rgba 0.0091 0.3515 0.9647 1>}
      uVRegion {bIsValid=false, max=<0 0>, min=<0 0>}
     }
     evenRowBackgroundBrush {
      bIsDynamicallyLoaded=false
      drawAs=Image
      imageType=NoImage
      mirroring=NoMirror
      resourceName=None
      resourceObject=null
      tiling=NoTile
      imageSize=<0 0>
      margin=<ltrb 0 0 0 0>
      outlineSettings {
       bUseBrushTransparency=false
       roundingType=HalfHeightRadius
       width=0
       color {colorUseRule=UseColor_Specified, specifiedColor=<rgba 0 0 0 0>}
       cornerRadii=<0 0 0 0>
      }
      tintColor {colorUseRule=UseColor_Specified, specifiedColor=<rgba 0.0103 0.0103 0.0103 1>}
      uVRegion {bIsValid=false, max=<0 0>, min=<0 0>}
     }
     evenRowBackgroundHoveredBrush {
      bIsDynamicallyLoaded=false
      drawAs=Image
      imageType=NoImage
      mirroring=NoMirror
      resourceName=None
      resourceObject=null
      tiling=NoTile
      imageSize=<0 0>
      margin=<ltrb 0 0 0 0>
      outlineSettings {
       bUseBrushTransparency=false
       roundingType=HalfHeightRadius
       width=0
       color {colorUseRule=UseColor_Specified, specifiedColor=<rgba 0 0 0 0>}
       cornerRadii=<0 0 0 0>
      }
      tintColor {colorUseRule=UseColor_Specified, specifiedColor=<rgba 0.0176 0.0176 0.0176 1>}
      uVRegion {bIsValid=false, max=<0 0>, min=<0 0>}
     }
     inactiveBrush {
      bIsDynamicallyLoaded=false
      drawAs=Image
      imageType=NoImage
      mirroring=NoMirror
      resourceName=None
      resourceObject=null
      tiling=NoTile
      imageSize=<0 0>
      margin=<ltrb 0 0 0 0>
      outlineSettings {
       bUseBrushTransparency=false
       roundingType=HalfHeightRadius
       width=0
       color {colorUseRule=UseColor_Specified, specifiedColor=<rgba 0 0 0 0>}
       cornerRadii=<0 0 0 0>
      }
      tintColor {colorUseRule=UseColor_Specified, specifiedColor=<rgba 0.0513 0.0953 0.159 1>}
      uVRegion {bIsValid=false, max=<0 0>, min=<0 0>}
     }
     inactiveHighlightedBrush {
      bIsDynamicallyLoaded=false
      drawAs=Image
      imageType=NoImage
      mirroring=NoMirror
      resourceName=None
      resourceObject=null
      tiling=NoTile
      imageSize=<0 0>
      margin=<ltrb 0 0 0 0>
      outlineSettings {
       bUseBrushTransparency=false
       roundingType=HalfHeightRadius
       width=0
       color {colorUseRule=UseColor_Specified, specifiedColor=<rgba 0 0 0 0>}
       cornerRadii=<0 0 0 0>
      }
      tintColor {colorUseRule=UseColor_Specified, specifiedColor=<rgba 0.0252 0.0319 0.0423 1>}
      uVRegion {bIsValid=false, max=<0 0>, min=<0 0>}
     }
     inactiveHoveredBrush {
      bIsDynamicallyLoaded=false
      drawAs=Image
      imageType=NoImage
      mirroring=NoMirror
      resourceName=None
      resourceObject=null
      tiling=NoTile
      imageSize=<0 0>
      margin=<ltrb 0 0 0 0>
      outlineSettings {
       bUseBrushTransparency=false
       roundingType=HalfHeightRadius
       width=0
       color {colorUseRule=UseColor_Specified, specifiedColor=<rgba 0 0 0 0>}
       cornerRadii=<0 0 0 0>
      }
      tintColor {colorUseRule=UseColor_Specified, specifiedColor=<rgba 0.0513 0.0953 0.159 1>}
      uVRegion {bIsValid=false, max=<0 0>, min=<0 0>}
     }
     oddRowBackgroundBrush {
      bIsDynamicallyLoaded=false
      drawAs=Image
      imageType=NoImage
      mirroring=NoMirror
      resourceName=None
      resourceObject=null
      tiling=NoTile
      imageSize=<0 0>
      margin=<ltrb 0 0 0 0>
      outlineSettings {
       bUseBrushTransparency=false
       roundingType=HalfHeightRadius
       width=0
       color {colorUseRule=UseColor_Specified, specifiedColor=<rgba 0 0 0 0>}
       cornerRadii=<0 0 0 0>
      }
      tintColor {colorUseRule=UseColor_Specified, specifiedColor=<rgba 0.0103 0.0103 0.0103 1>}
      uVRegion {bIsValid=false, max=<0 0>, min=<0 0>}
     }
     oddRowBackgroundHoveredBrush {
      bIsDynamicallyLoaded=false
      drawAs=Image
      imageType=NoImage
      mirroring=NoMirror
      resourceName=None
      resourceObject=null
      tiling=NoTile
      imageSize=<0 0>
      margin=<ltrb 0 0 0 0>
      outlineSettings {
       bUseBrushTransparency=false
       roundingType=HalfHeightRadius
       width=0
       color {colorUseRule=UseColor_Specified, specifiedColor=<rgba 0 0 0 0>}
       cornerRadii=<0 0 0 0>
      }
      tintColor {colorUseRule=UseColor_Specified, specifiedColor=<rgba 0.0176 0.0176 0.0176 1>}
      uVRegion {bIsValid=false, max=<0 0>, min=<0 0>}
     }
     parentRowBackgroundBrush {
      bIsDynamicallyLoaded=false
      drawAs=Image
      imageType=NoImage
      mirroring=NoMirror
      resourceName=None
      resourceObject=null
      tiling=NoTile
      imageSize=<32 32>
      margin=<ltrb 0 0 0 0>
      outlineSettings {
       bUseBrushTransparency=false
       roundingType=HalfHeightRadius
       width=0
       color {colorUseRule=UseColor_Specified, specifiedColor=<rgba 0 0 0 0>}
       cornerRadii=<0 0 0 0>
      }
      tintColor {colorUseRule=UseColor_Specified, specifiedColor=<rgba 1 1 1 1>}
      uVRegion {bIsValid=false, max=<0 0>, min=<0 0>}
     }
     parentRowBackgroundHoveredBrush {
      bIsDynamicallyLoaded=false
      drawAs=Image
      imageType=NoImage
      mirroring=NoMirror
      resourceName=None
      resourceObject=null
      tiling=NoTile
      imageSize=<32 32>
      margin=<ltrb 0 0 0 0>
      outlineSettings {
       bUseBrushTransparency=false
       roundingType=HalfHeightRadius
       width=0
       color {colorUseRule=UseColor_Specified, specifiedColor=<rgba 0 0 0 0>}
       cornerRadii=<0 0 0 0>
      }
      tintColor {colorUseRule=UseColor_Specified, specifiedColor=<rgba 1 1 1 1>}
      uVRegion {bIsValid=false, max=<0 0>, min=<0 0>}
     }
     selectedTextColor {colorUseRule=UseColor_Specified, specifiedColor=<rgba 1 1 1 1>}
     selectorFocusedBrush {
      bIsDynamicallyLoaded=false
      drawAs=Border
      imageType=FullColor
      mirroring=NoMirror
      resourceName=../../../Engine/Content/Slate/Common/Selector.png
      resourceObject=null
      tiling=Both
      imageSize=<0 0>
      margin=<ltrb 0.25 0.25 0.25 0.25>
      outlineSettings {
       bUseBrushTransparency=false
       roundingType=HalfHeightRadius
       width=0
       color {colorUseRule=UseColor_Specified, specifiedColor=<rgba 0 0 0 0>}
       cornerRadii=<0 0 0 0>
      }
      tintColor {colorUseRule=UseColor_Specified, specifiedColor=<rgba 0.701 0.225 0.003 1>}
      uVRegion {bIsValid=false, max=<0 0>, min=<0 0>}
     }
     textColor {colorUseRule=UseColor_Specified, specifiedColor=<rgba 0.5271 0.5271 0.5271 1>}
    }
    scrollBarStyle {
     thickness=8
     draggedThumbImage {
      bIsDynamicallyLoaded=false
      drawAs=RoundedBox
      imageType=NoImage
      mirroring=NoMirror
      resourceName=None
      resourceObject=null
      tiling=NoTile
      imageSize=<0 0>
      margin=<ltrb 0 0 0 0>
      outlineSettings {
       bUseBrushTransparency=false
       roundingType=FixedRadius
       width=0
       color {colorUseRule=UseColor_Specified, specifiedColor=<rgba 0 0 0 0>}
       cornerRadii=<4 4 4 4>
      }
      tintColor {colorUseRule=UseColor_Specified, specifiedColor=<rgba 0.2159 0.2159 0.2159 1>}
      uVRegion {bIsValid=false, max=<0 0>, min=<0 0>}
     }
     horizontalBackgroundImage {
      bIsDynamicallyLoaded=false
      drawAs=NoDrawType
      imageType=NoImage
      mirroring=NoMirror
      resourceName=None
      resourceObject=null
      tiling=NoTile
      imageSize=<0 0>
      margin=<ltrb 0 0 0 0>
      outlineSettings {
       bUseBrushTransparency=false
       roundingType=HalfHeightRadius
       width=0
       color {colorUseRule=UseColor_Specified, specifiedColor=<rgba 0 0 0 0>}
       cornerRadii=<0 0 0 0>
      }
      tintColor {colorUseRule=UseColor_Specified, specifiedColor=<rgba 1 1 1 1>}
      uVRegion {bIsValid=false, max=<0 0>, min=<0 0>}
     }
     horizontalBottomSlotImage {
      bIsDynamicallyLoaded=false
      drawAs=NoDrawType
      imageType=NoImage
      mirroring=NoMirror
      resourceName=None
      resourceObject=null
      tiling=NoTile
      imageSize=<0 0>
      margin=<ltrb 0 0 0 0>
      outlineSettings {
       bUseBrushTransparency=false
       roundingType=HalfHeightRadius
       width=0
       color {colorUseRule=UseColor_Specified, specifiedColor=<rgba 0 0 0 0>}
       cornerRadii=<0 0 0 0>
      }
      tintColor {colorUseRule=UseColor_Specified, specifiedColor=<rgba 1 1 1 1>}
      uVRegion {bIsValid=false, max=<0 0>, min=<0 0>}
     }
     horizontalTopSlotImage {
      bIsDynamicallyLoaded=false
      drawAs=NoDrawType
      imageType=NoImage
      mirroring=NoMirror
      resourceName=None
      resourceObject=null
      tiling=NoTile
      imageSize=<0 0>
      margin=<ltrb 0 0 0 0>
      outlineSettings {
       bUseBrushTransparency=false
       roundingType=HalfHeightRadius
       width=0
       color {colorUseRule=UseColor_Specified, specifiedColor=<rgba 0 0 0 0>}
       cornerRadii=<0 0 0 0>
      }
      tintColor {colorUseRule=UseColor_Specified, specifiedColor=<rgba 1 1 1 1>}
      uVRegion {bIsValid=false, max=<0 0>, min=<0 0>}
     }
     hoveredThumbImage {
      bIsDynamicallyLoaded=false
      drawAs=RoundedBox
      imageType=NoImage
      mirroring=NoMirror
      resourceName=None
      resourceObject=null
      tiling=NoTile
      imageSize=<0 0>
      margin=<ltrb 0 0 0 0>
      outlineSettings {
       bUseBrushTransparency=false
       roundingType=FixedRadius
       width=0
       color {colorUseRule=UseColor_Specified, specifiedColor=<rgba 0 0 0 0>}
       cornerRadii=<4 4 4 4>
      }
      tintColor {colorUseRule=UseColor_Specified, specifiedColor=<rgba 0.2159 0.2159 0.2159 1>}
      uVRegion {bIsValid=false, max=<0 0>, min=<0 0>}
     }
     normalThumbImage {
      bIsDynamicallyLoaded=false
      drawAs=RoundedBox
      imageType=NoImage
      mirroring=NoMirror
      resourceName=None
      resourceObject=null
      tiling=NoTile
      imageSize=<0 0>
      margin=<ltrb 0 0 0 0>
      outlineSettings {
       bUseBrushTransparency=false
       roundingType=FixedRadius
       width=0
       color {colorUseRule=UseColor_Specified, specifiedColor=<rgba 0 0 0 0>}
       cornerRadii=<4 4 4 4>
      }
      tintColor {colorUseRule=UseColor_Specified, specifiedColor=<rgba 0.0953 0.0953 0.0953 1>}
      uVRegion {bIsValid=false, max=<0 0>, min=<0 0>}
     }
     verticalBackgroundImage {
      bIsDynamicallyLoaded=false
      drawAs=NoDrawType
      imageType=NoImage
      mirroring=NoMirror
      resourceName=None
      resourceObject=null
      tiling=NoTile
      imageSize=<0 0>
      margin=<ltrb 0 0 0 0>
      outlineSettings {
       bUseBrushTransparency=false
       roundingType=HalfHeightRadius
       width=0
       color {colorUseRule=UseColor_Specified, specifiedColor=<rgba 0 0 0 0>}
       cornerRadii=<0 0 0 0>
      }
      tintColor {colorUseRule=UseColor_Specified, specifiedColor=<rgba 1 1 1 1>}
      uVRegion {bIsValid=false, max=<0 0>, min=<0 0>}
     }
     verticalBottomSlotImage {
      bIsDynamicallyLoaded=false
      drawAs=NoDrawType
      imageType=NoImage
      mirroring=NoMirror
      resourceName=None
      resourceObject=null
      tiling=NoTile
      imageSize=<0 0>
      margin=<ltrb 0 0 0 0>
      outlineSettings {
       bUseBrushTransparency=false
       roundingType=HalfHeightRadius
       width=0
       color {colorUseRule=UseColor_Specified, specifiedColor=<rgba 0 0 0 0>}
       cornerRadii=<0 0 0 0>
      }
      tintColor {colorUseRule=UseColor_Specified, specifiedColor=<rgba 1 1 1 1>}
      uVRegion {bIsValid=false, max=<0 0>, min=<0 0>}
     }
     verticalTopSlotImage {
      bIsDynamicallyLoaded=false
      drawAs=NoDrawType
      imageType=NoImage
      mirroring=NoMirror
      resourceName=None
      resourceObject=null
      tiling=NoTile
      imageSize=<0 0>
      margin=<ltrb 0 0 0 0>
      outlineSettings {
       bUseBrushTransparency=false
       roundingType=HalfHeightRadius
       width=0
       color {colorUseRule=UseColor_Specified, specifiedColor=<rgba 0 0 0 0>}
       cornerRadii=<0 0 0 0>
      }
      tintColor {colorUseRule=UseColor_Specified, specifiedColor=<rgba 1 1 1 1>}
      uVRegion {bIsValid=false, max=<0 0>, min=<0 0>}
     }
    }
    widgetStyle {
     comboButtonStyle {
      downArrowAlign=VAlign_Center
      buttonStyle {
       clickedSlateSound {resourceObject=null}
       disabled {
        bIsDynamicallyLoaded=false
        drawAs=RoundedBox
        imageType=NoImage
        mirroring=NoMirror
        resourceName=None
        resourceObject=null
        tiling=NoTile
        imageSize=<0 0>
        margin=<ltrb 0 0 0 0>
        outlineSettings {
         bUseBrushTransparency=false
         roundingType=FixedRadius
         width=1
         color {colorUseRule=UseColor_ColorTable, specifiedColor=<rgba 0 0 0 0>}
         cornerRadii=<4 4 4 4>
        }
        tintColor {colorUseRule=UseColor_Specified, specifiedColor=<rgba 0.0395 0.0395 0.0395 1>}
        uVRegion {bIsValid=false, max=<0 0>, min=<0 0>}
       }
       disabledForeground {colorUseRule=UseColor_ColorTable, specifiedColor=<rgba 0 0 0 0>}
       hovered {
        bIsDynamicallyLoaded=false
        drawAs=RoundedBox
        imageType=NoImage
        mirroring=NoMirror
        resourceName=None
        resourceObject=null
        tiling=NoTile
        imageSize=<0 0>
        margin=<ltrb 0 0 0 0>
        outlineSettings {
         bUseBrushTransparency=false
         roundingType=FixedRadius
         width=1
         color {colorUseRule=UseColor_ColorTable, specifiedColor=<rgba 0 0 0 0>}
         cornerRadii=<4 4 4 4>
        }
        tintColor {colorUseRule=UseColor_Specified, specifiedColor=<rgba 0.0048 0.0048 0.0048 1>}
        uVRegion {bIsValid=false, max=<0 0>, min=<0 0>}
       }
       hoveredForeground {colorUseRule=UseColor_ColorTable, specifiedColor=<rgba 0 0 0 0>}
       hoveredSlateSound {resourceObject=null}
       normal {
        bIsDynamicallyLoaded=false
        drawAs=RoundedBox
        imageType=NoImage
        mirroring=NoMirror
        resourceName=None
        resourceObject=null
        tiling=NoTile
        imageSize=<0 0>
        margin=<ltrb 0 0 0 0>
        outlineSettings {
         bUseBrushTransparency=false
         roundingType=FixedRadius
         width=1
         color {colorUseRule=UseColor_ColorTable, specifiedColor=<rgba 0 0 0 0>}
         cornerRadii=<4 4 4 4>
        }
        tintColor {colorUseRule=UseColor_Specified, specifiedColor=<rgba 0.0048 0.0048 0.0048 1>}
        uVRegion {bIsValid=false, max=<0 0>, min=<0 0>}
       }
       normalForeground {colorUseRule=UseColor_ColorTable, specifiedColor=<rgba 0 0 0 0>}
       normalPadding=<ltrb 8 1 8 1>
       pressed {
        bIsDynamicallyLoaded=false
        drawAs=RoundedBox
        imageType=NoImage
        mirroring=NoMirror
        resourceName=None
        resourceObject=null
        tiling=NoTile
        imageSize=<0 0>
        margin=<ltrb 0 0 0 0>
        outlineSettings {
         bUseBrushTransparency=false
         roundingType=FixedRadius
         width=1
         color {colorUseRule=UseColor_ColorTable, specifiedColor=<rgba 0 0 0 0>}
         cornerRadii=<4 4 4 4>
        }
        tintColor {colorUseRule=UseColor_Specified, specifiedColor=<rgba 0.0048 0.0048 0.0048 1>}
        uVRegion {bIsValid=false, max=<0 0>, min=<0 0>}
       }
       pressedForeground {colorUseRule=UseColor_ColorTable, specifiedColor=<rgba 0 0 0 0>}
       pressedPadding=<ltrb 8 2 8 0>
       pressedSlateSound {resourceObject=null}
      }
      contentPadding=<ltrb 0 0 0 0>
      downArrowImage {
       bIsDynamicallyLoaded=false
       drawAs=Image
       imageType=Vector
       mirroring=NoMirror
       resourceName=../../../Engine/Content/Slate/Starship/CoreWidgets/ComboBox/wide-chevron-down.svg
       resourceObject=null
       tiling=NoTile
       imageSize=<16 16>
       margin=<ltrb 0 0 0 0>
       outlineSettings {
        bUseBrushTransparency=false
        roundingType=HalfHeightRadius
        width=0
        color {colorUseRule=UseColor_Specified, specifiedColor=<rgba 0 0 0 0>}
        cornerRadii=<0 0 0 0>
       }
       tintColor {colorUseRule=UseColor_Specified, specifiedColor=<rgba 1 1 1 1>}
       uVRegion {bIsValid=false, max=<0 0>, min=<0 0>}
      }
      downArrowPadding=<ltrb 2 2 2 2>
      menuBorderBrush {
       bIsDynamicallyLoaded=false
       drawAs=RoundedBox
       imageType=NoImage
       mirroring=NoMirror
       resourceName=None
       resourceObject=null
       tiling=NoTile
       imageSize=<0 0>
       margin=<ltrb 0 0 0 0>
       outlineSettings {
        bUseBrushTransparency=false
        roundingType=FixedRadius
        width=1
        color {colorUseRule=UseColor_ColorTable, specifiedColor=<rgba 0 0 0 0>}
        cornerRadii=<0 0 0 0>
       }
       tintColor {colorUseRule=UseColor_Specified, specifiedColor=<rgba 0.0395 0.0395 0.0395 1>}
       uVRegion {bIsValid=false, max=<0 0>, min=<0 0>}
      }
      menuBorderPadding=<ltrb 0 0 0 0>
      shadowColorAndOpacity=<rgba 0 0 0 1>
      shadowOffset=<0 0>
     }
     contentPadding=<ltrb 0 0 0 0>
     menuRowPadding=<ltrb 8 3 8 3>
     pressedSlateSound {resourceObject=null}
     selectionChangeSlateSound {resourceObject=null}
    }
   }
   slot {class=HorizontalBoxSlot}
  }
  SequenceSelector {
   widgetClassPath=&ComboBoxString
   bIsVariable=true
   parent=HorizontalBox_0
   widget=SequenceSelector
   properties {
    selectedOption=Standard Characters
    defaultOptions [
     Standard Characters
     Toon Characters
     Toon Characters - BR_Art_Current
     Modelling Sequence
     Morph Target
     Turntable Status Verification
     Vehicles
     ExecBeautyShots
     FlipbookCapture
     Standard Characters - IP Headshot
     Toon Characters - IP Headshot
    ]
    foregroundColor {colorUseRule=UseColor_Specified, specifiedColor=<rgba 1 1 1 1>}
    itemStyle {
     bUseParentRowBrush=false
     activeBrush {
      bIsDynamicallyLoaded=false
      drawAs=Image
      imageType=NoImage
      mirroring=NoMirror
      resourceName=None
      resourceObject=null
      tiling=NoTile
      imageSize=<0 0>
      margin=<ltrb 0 0 0 0>
      outlineSettings {
       bUseBrushTransparency=false
       roundingType=HalfHeightRadius
       width=0
       color {colorUseRule=UseColor_Specified, specifiedColor=<rgba 0 0 0 0>}
       cornerRadii=<0 0 0 0>
      }
      tintColor {colorUseRule=UseColor_Specified, specifiedColor=<rgba 0 0.162 0.7454 1>}
      uVRegion {bIsValid=false, max=<0 0>, min=<0 0>}
     }
     activeHighlightedBrush {
      bIsDynamicallyLoaded=false
      drawAs=Image
      imageType=NoImage
      mirroring=NoMirror
      resourceName=None
      resourceObject=null
      tiling=NoTile
      imageSize=<0 0>
      margin=<ltrb 0 0 0 0>
      outlineSettings {
       bUseBrushTransparency=false
       roundingType=HalfHeightRadius
       width=0
       color {colorUseRule=UseColor_Specified, specifiedColor=<rgba 0 0 0 0>}
       cornerRadii=<0 0 0 0>
      }
      tintColor {colorUseRule=UseColor_Specified, specifiedColor=<rgba 0.0252 0.0319 0.0423 1>}
      uVRegion {bIsValid=false, max=<0 0>, min=<0 0>}
     }
     activeHoveredBrush {
      bIsDynamicallyLoaded=false
      drawAs=Image
      imageType=NoImage
      mirroring=NoMirror
      resourceName=None
      resourceObject=null
      tiling=NoTile
      imageSize=<0 0>
      margin=<ltrb 0 0 0 0>
      outlineSettings {
       bUseBrushTransparency=false
       roundingType=HalfHeightRadius
       width=0
       color {colorUseRule=UseColor_Specified, specifiedColor=<rgba 0 0 0 0>}
       cornerRadii=<0 0 0 0>
      }
      tintColor {colorUseRule=UseColor_Specified, specifiedColor=<rgba 0 0.162 0.7454 1>}
      uVRegion {bIsValid=false, max=<0 0>, min=<0 0>}
     }
     dropIndicator_Above {
      bIsDynamicallyLoaded=false
      drawAs=Box
      imageType=FullColor
      mirroring=NoMirror
      resourceName=../../../Engine/Content/Slate/Common/DropZoneIndicator_Above.png
      resourceObject=null
      tiling=NoTile
      imageSize=<0 0>
      margin=<ltrb 0.625 0.625 0 0>
      outlineSettings {
       bUseBrushTransparency=false
       roundingType=HalfHeightRadius
       width=0
       color {colorUseRule=UseColor_Specified, specifiedColor=<rgba 0 0 0 0>}
       cornerRadii=<0 0 0 0>
      }
      tintColor {colorUseRule=UseColor_Specified, specifiedColor=<rgba 0.0091 0.3515 0.9647 1>}
      uVRegion {bIsValid=false, max=<0 0>, min=<0 0>}
     }
     dropIndicator_Below {
      bIsDynamicallyLoaded=false
      drawAs=Box
      imageType=FullColor
      mirroring=NoMirror
      resourceName=../../../Engine/Content/Slate/Common/DropZoneIndicator_Below.png
      resourceObject=null
      tiling=NoTile
      imageSize=<0 0>
      margin=<ltrb 0.625 0 0 0.625>
      outlineSettings {
       bUseBrushTransparency=false
       roundingType=HalfHeightRadius
       width=0
       color {colorUseRule=UseColor_Specified, specifiedColor=<rgba 0 0 0 0>}
       cornerRadii=<0 0 0 0>
      }
      tintColor {colorUseRule=UseColor_Specified, specifiedColor=<rgba 0.0091 0.3515 0.9647 1>}
      uVRegion {bIsValid=false, max=<0 0>, min=<0 0>}
     }
     dropIndicator_Onto {
      bIsDynamicallyLoaded=false
      drawAs=Box
      imageType=FullColor
      mirroring=NoMirror
      resourceName=../../../Engine/Content/Slate/Common/DropZoneIndicator_Onto.png
      resourceObject=null
      tiling=NoTile
      imageSize=<0 0>
      margin=<ltrb 0.25 0.25 0.25 0.25>
      outlineSettings {
       bUseBrushTransparency=false
       roundingType=HalfHeightRadius
       width=0
       color {colorUseRule=UseColor_Specified, specifiedColor=<rgba 0 0 0 0>}
       cornerRadii=<0 0 0 0>
      }
      tintColor {colorUseRule=UseColor_Specified, specifiedColor=<rgba 0.0091 0.3515 0.9647 1>}
      uVRegion {bIsValid=false, max=<0 0>, min=<0 0>}
     }
     evenRowBackgroundBrush {
      bIsDynamicallyLoaded=false
      drawAs=Image
      imageType=NoImage
      mirroring=NoMirror
      resourceName=None
      resourceObject=null
      tiling=NoTile
      imageSize=<0 0>
      margin=<ltrb 0 0 0 0>
      outlineSettings {
       bUseBrushTransparency=false
       roundingType=HalfHeightRadius
       width=0
       color {colorUseRule=UseColor_Specified, specifiedColor=<rgba 0 0 0 0>}
       cornerRadii=<0 0 0 0>
      }
      tintColor {colorUseRule=UseColor_Specified, specifiedColor=<rgba 0.0103 0.0103 0.0103 1>}
      uVRegion {bIsValid=false, max=<0 0>, min=<0 0>}
     }
     evenRowBackgroundHoveredBrush {
      bIsDynamicallyLoaded=false
      drawAs=Image
      imageType=NoImage
      mirroring=NoMirror
      resourceName=None
      resourceObject=null
      tiling=NoTile
      imageSize=<0 0>
      margin=<ltrb 0 0 0 0>
      outlineSettings {
       bUseBrushTransparency=false
       roundingType=HalfHeightRadius
       width=0
       color {colorUseRule=UseColor_Specified, specifiedColor=<rgba 0 0 0 0>}
       cornerRadii=<0 0 0 0>
      }
      tintColor {colorUseRule=UseColor_Specified, specifiedColor=<rgba 0.0176 0.0176 0.0176 1>}
      uVRegion {bIsValid=false, max=<0 0>, min=<0 0>}
     }
     inactiveBrush {
      bIsDynamicallyLoaded=false
      drawAs=Image
      imageType=NoImage
      mirroring=NoMirror
      resourceName=None
      resourceObject=null
      tiling=NoTile
      imageSize=<0 0>
      margin=<ltrb 0 0 0 0>
      outlineSettings {
       bUseBrushTransparency=false
       roundingType=HalfHeightRadius
       width=0
       color {colorUseRule=UseColor_Specified, specifiedColor=<rgba 0 0 0 0>}
       cornerRadii=<0 0 0 0>
      }
      tintColor {colorUseRule=UseColor_Specified, specifiedColor=<rgba 0.0513 0.0953 0.159 1>}
      uVRegion {bIsValid=false, max=<0 0>, min=<0 0>}
     }
     inactiveHighlightedBrush {
      bIsDynamicallyLoaded=false
      drawAs=Image
      imageType=NoImage
      mirroring=NoMirror
      resourceName=None
      resourceObject=null
      tiling=NoTile
      imageSize=<0 0>
      margin=<ltrb 0 0 0 0>
      outlineSettings {
       bUseBrushTransparency=false
       roundingType=HalfHeightRadius
       width=0
       color {colorUseRule=UseColor_Specified, specifiedColor=<rgba 0 0 0 0>}
       cornerRadii=<0 0 0 0>
      }
      tintColor {colorUseRule=UseColor_Specified, specifiedColor=<rgba 0.0252 0.0319 0.0423 1>}
      uVRegion {bIsValid=false, max=<0 0>, min=<0 0>}
     }
     inactiveHoveredBrush {
      bIsDynamicallyLoaded=false
      drawAs=Image
      imageType=NoImage
      mirroring=NoMirror
      resourceName=None
      resourceObject=null
      tiling=NoTile
      imageSize=<0 0>
      margin=<ltrb 0 0 0 0>
      outlineSettings {
       bUseBrushTransparency=false
       roundingType=HalfHeightRadius
       width=0
       color {colorUseRule=UseColor_Specified, specifiedColor=<rgba 0 0 0 0>}
       cornerRadii=<0 0 0 0>
      }
      tintColor {colorUseRule=UseColor_Specified, specifiedColor=<rgba 0.0513 0.0953 0.159 1>}
      uVRegion {bIsValid=false, max=<0 0>, min=<0 0>}
     }
     oddRowBackgroundBrush {
      bIsDynamicallyLoaded=false
      drawAs=Image
      imageType=NoImage
      mirroring=NoMirror
      resourceName=None
      resourceObject=null
      tiling=NoTile
      imageSize=<0 0>
      margin=<ltrb 0 0 0 0>
      outlineSettings {
       bUseBrushTransparency=false
       roundingType=HalfHeightRadius
       width=0
       color {colorUseRule=UseColor_Specified, specifiedColor=<rgba 0 0 0 0>}
       cornerRadii=<0 0 0 0>
      }
      tintColor {colorUseRule=UseColor_Specified, specifiedColor=<rgba 0.0103 0.0103 0.0103 1>}
      uVRegion {bIsValid=false, max=<0 0>, min=<0 0>}
     }
     oddRowBackgroundHoveredBrush {
      bIsDynamicallyLoaded=false
      drawAs=Image
      imageType=NoImage
      mirroring=NoMirror
      resourceName=None
      resourceObject=null
      tiling=NoTile
      imageSize=<0 0>
      margin=<ltrb 0 0 0 0>
      outlineSettings {
       bUseBrushTransparency=false
       roundingType=HalfHeightRadius
       width=0
       color {colorUseRule=UseColor_Specified, specifiedColor=<rgba 0 0 0 0>}
       cornerRadii=<0 0 0 0>
      }
      tintColor {colorUseRule=UseColor_Specified, specifiedColor=<rgba 0.0176 0.0176 0.0176 1>}
      uVRegion {bIsValid=false, max=<0 0>, min=<0 0>}
     }
     parentRowBackgroundBrush {
      bIsDynamicallyLoaded=false
      drawAs=Image
      imageType=NoImage
      mirroring=NoMirror
      resourceName=None
      resourceObject=null
      tiling=NoTile
      imageSize=<32 32>
      margin=<ltrb 0 0 0 0>
      outlineSettings {
       bUseBrushTransparency=false
       roundingType=HalfHeightRadius
       width=0
       color {colorUseRule=UseColor_Specified, specifiedColor=<rgba 0 0 0 0>}
       cornerRadii=<0 0 0 0>
      }
      tintColor {colorUseRule=UseColor_Specified, specifiedColor=<rgba 1 1 1 1>}
      uVRegion {bIsValid=false, max=<0 0>, min=<0 0>}
     }
     parentRowBackgroundHoveredBrush {
      bIsDynamicallyLoaded=false
      drawAs=Image
      imageType=NoImage
      mirroring=NoMirror
      resourceName=None
      resourceObject=null
      tiling=NoTile
      imageSize=<32 32>
      margin=<ltrb 0 0 0 0>
      outlineSettings {
       bUseBrushTransparency=false
       roundingType=HalfHeightRadius
       width=0
       color {colorUseRule=UseColor_Specified, specifiedColor=<rgba 0 0 0 0>}
       cornerRadii=<0 0 0 0>
      }
      tintColor {colorUseRule=UseColor_Specified, specifiedColor=<rgba 1 1 1 1>}
      uVRegion {bIsValid=false, max=<0 0>, min=<0 0>}
     }
     selectedTextColor {colorUseRule=UseColor_Specified, specifiedColor=<rgba 1 1 1 1>}
     selectorFocusedBrush {
      bIsDynamicallyLoaded=false
      drawAs=Border
      imageType=FullColor
      mirroring=NoMirror
      resourceName=../../../Engine/Content/Slate/Common/Selector.png
      resourceObject=null
      tiling=Both
      imageSize=<0 0>
      margin=<ltrb 0.25 0.25 0.25 0.25>
      outlineSettings {
       bUseBrushTransparency=false
       roundingType=HalfHeightRadius
       width=0
       color {colorUseRule=UseColor_Specified, specifiedColor=<rgba 0 0 0 0>}
       cornerRadii=<0 0 0 0>
      }
      tintColor {colorUseRule=UseColor_Specified, specifiedColor=<rgba 0.701 0.225 0.003 1>}
      uVRegion {bIsValid=false, max=<0 0>, min=<0 0>}
     }
     textColor {colorUseRule=UseColor_Specified, specifiedColor=<rgba 0.5271 0.5271 0.5271 1>}
    }
    scrollBarStyle {
     thickness=8
     draggedThumbImage {
      bIsDynamicallyLoaded=false
      drawAs=RoundedBox
      imageType=NoImage
      mirroring=NoMirror
      resourceName=None
      resourceObject=null
      tiling=NoTile
      imageSize=<0 0>
      margin=<ltrb 0 0 0 0>
      outlineSettings {
       bUseBrushTransparency=false
       roundingType=FixedRadius
       width=0
       color {colorUseRule=UseColor_Specified, specifiedColor=<rgba 0 0 0 0>}
       cornerRadii=<4 4 4 4>
      }
      tintColor {colorUseRule=UseColor_Specified, specifiedColor=<rgba 0.2159 0.2159 0.2159 1>}
      uVRegion {bIsValid=false, max=<0 0>, min=<0 0>}
     }
     horizontalBackgroundImage {
      bIsDynamicallyLoaded=false
      drawAs=NoDrawType
      imageType=NoImage
      mirroring=NoMirror
      resourceName=None
      resourceObject=null
      tiling=NoTile
      imageSize=<0 0>
      margin=<ltrb 0 0 0 0>
      outlineSettings {
       bUseBrushTransparency=false
       roundingType=HalfHeightRadius
       width=0
       color {colorUseRule=UseColor_Specified, specifiedColor=<rgba 0 0 0 0>}
       cornerRadii=<0 0 0 0>
      }
      tintColor {colorUseRule=UseColor_Specified, specifiedColor=<rgba 1 1 1 1>}
      uVRegion {bIsValid=false, max=<0 0>, min=<0 0>}
     }
     horizontalBottomSlotImage {
      bIsDynamicallyLoaded=false
      drawAs=NoDrawType
      imageType=NoImage
      mirroring=NoMirror
      resourceName=None
      resourceObject=null
      tiling=NoTile
      imageSize=<0 0>
      margin=<ltrb 0 0 0 0>
      outlineSettings {
       bUseBrushTransparency=false
       roundingType=HalfHeightRadius
       width=0
       color {colorUseRule=UseColor_Specified, specifiedColor=<rgba 0 0 0 0>}
       cornerRadii=<0 0 0 0>
      }
      tintColor {colorUseRule=UseColor_Specified, specifiedColor=<rgba 1 1 1 1>}
      uVRegion {bIsValid=false, max=<0 0>, min=<0 0>}
     }
     horizontalTopSlotImage {
      bIsDynamicallyLoaded=false
      drawAs=NoDrawType
      imageType=NoImage
      mirroring=NoMirror
      resourceName=None
      resourceObject=null
      tiling=NoTile
      imageSize=<0 0>
      margin=<ltrb 0 0 0 0>
      outlineSettings {
       bUseBrushTransparency=false
       roundingType=HalfHeightRadius
       width=0
       color {colorUseRule=UseColor_Specified, specifiedColor=<rgba 0 0 0 0>}
       cornerRadii=<0 0 0 0>
      }
      tintColor {colorUseRule=UseColor_Specified, specifiedColor=<rgba 1 1 1 1>}
      uVRegion {bIsValid=false, max=<0 0>, min=<0 0>}
     }
     hoveredThumbImage {
      bIsDynamicallyLoaded=false
      drawAs=RoundedBox
      imageType=NoImage
      mirroring=NoMirror
      resourceName=None
      resourceObject=null
      tiling=NoTile
      imageSize=<0 0>
      margin=<ltrb 0 0 0 0>
      outlineSettings {
       bUseBrushTransparency=false
       roundingType=FixedRadius
       width=0
       color {colorUseRule=UseColor_Specified, specifiedColor=<rgba 0 0 0 0>}
       cornerRadii=<4 4 4 4>
      }
      tintColor {colorUseRule=UseColor_Specified, specifiedColor=<rgba 0.2159 0.2159 0.2159 1>}
      uVRegion {bIsValid=false, max=<0 0>, min=<0 0>}
     }
     normalThumbImage {
      bIsDynamicallyLoaded=false
      drawAs=RoundedBox
      imageType=NoImage
      mirroring=NoMirror
      resourceName=None
      resourceObject=null
      tiling=NoTile
      imageSize=<0 0>
      margin=<ltrb 0 0 0 0>
      outlineSettings {
       bUseBrushTransparency=false
       roundingType=FixedRadius
       width=0
       color {colorUseRule=UseColor_Specified, specifiedColor=<rgba 0 0 0 0>}
       cornerRadii=<4 4 4 4>
      }
      tintColor {colorUseRule=UseColor_Specified, specifiedColor=<rgba 0.0953 0.0953 0.0953 1>}
      uVRegion {bIsValid=false, max=<0 0>, min=<0 0>}
     }
     verticalBackgroundImage {
      bIsDynamicallyLoaded=false
      drawAs=NoDrawType
      imageType=NoImage
      mirroring=NoMirror
      resourceName=None
      resourceObject=null
      tiling=NoTile
      imageSize=<0 0>
      margin=<ltrb 0 0 0 0>
      outlineSettings {
       bUseBrushTransparency=false
       roundingType=HalfHeightRadius
       width=0
       color {colorUseRule=UseColor_Specified, specifiedColor=<rgba 0 0 0 0>}
       cornerRadii=<0 0 0 0>
      }
      tintColor {colorUseRule=UseColor_Specified, specifiedColor=<rgba 1 1 1 1>}
      uVRegion {bIsValid=false, max=<0 0>, min=<0 0>}
     }
     verticalBottomSlotImage {
      bIsDynamicallyLoaded=false
      drawAs=NoDrawType
      imageType=NoImage
      mirroring=NoMirror
      resourceName=None
      resourceObject=null
      tiling=NoTile
      imageSize=<0 0>
      margin=<ltrb 0 0 0 0>
      outlineSettings {
       bUseBrushTransparency=false
       roundingType=HalfHeightRadius
       width=0
       color {colorUseRule=UseColor_Specified, specifiedColor=<rgba 0 0 0 0>}
       cornerRadii=<0 0 0 0>
      }
      tintColor {colorUseRule=UseColor_Specified, specifiedColor=<rgba 1 1 1 1>}
      uVRegion {bIsValid=false, max=<0 0>, min=<0 0>}
     }
     verticalTopSlotImage {
      bIsDynamicallyLoaded=false
      drawAs=NoDrawType
      imageType=NoImage
      mirroring=NoMirror
      resourceName=None
      resourceObject=null
      tiling=NoTile
      imageSize=<0 0>
      margin=<ltrb 0 0 0 0>
      outlineSettings {
       bUseBrushTransparency=false
       roundingType=HalfHeightRadius
       width=0
       color {colorUseRule=UseColor_Specified, specifiedColor=<rgba 0 0 0 0>}
       cornerRadii=<0 0 0 0>
      }
      tintColor {colorUseRule=UseColor_Specified, specifiedColor=<rgba 1 1 1 1>}
      uVRegion {bIsValid=false, max=<0 0>, min=<0 0>}
     }
    }
    widgetStyle {
     comboButtonStyle {
      downArrowAlign=VAlign_Center
      buttonStyle {
       clickedSlateSound {resourceObject=null}
       disabled {
        bIsDynamicallyLoaded=false
        drawAs=RoundedBox
        imageType=NoImage
        mirroring=NoMirror
        resourceName=None
        resourceObject=null
        tiling=NoTile
        imageSize=<0 0>
        margin=<ltrb 0 0 0 0>
        outlineSettings {
         bUseBrushTransparency=false
         roundingType=FixedRadius
         width=1
         color {colorUseRule=UseColor_ColorTable, specifiedColor=<rgba 0 0 0 0>}
         cornerRadii=<4 4 4 4>
        }
        tintColor {colorUseRule=UseColor_Specified, specifiedColor=<rgba 0.0395 0.0395 0.0395 1>}
        uVRegion {bIsValid=false, max=<0 0>, min=<0 0>}
       }
       disabledForeground {colorUseRule=UseColor_ColorTable, specifiedColor=<rgba 0 0 0 0>}
       hovered {
        bIsDynamicallyLoaded=false
        drawAs=RoundedBox
        imageType=NoImage
        mirroring=NoMirror
        resourceName=None
        resourceObject=null
        tiling=NoTile
        imageSize=<0 0>
        margin=<ltrb 0 0 0 0>
        outlineSettings {
         bUseBrushTransparency=false
         roundingType=FixedRadius
         width=1
         color {colorUseRule=UseColor_ColorTable, specifiedColor=<rgba 0 0 0 0>}
         cornerRadii=<4 4 4 4>
        }
        tintColor {colorUseRule=UseColor_Specified, specifiedColor=<rgba 0.0048 0.0048 0.0048 1>}
        uVRegion {bIsValid=false, max=<0 0>, min=<0 0>}
       }
       hoveredForeground {colorUseRule=UseColor_ColorTable, specifiedColor=<rgba 0 0 0 0>}
       hoveredSlateSound {resourceObject=null}
       normal {
        bIsDynamicallyLoaded=false
        drawAs=RoundedBox
        imageType=NoImage
        mirroring=NoMirror
        resourceName=None
        resourceObject=null
        tiling=NoTile
        imageSize=<0 0>
        margin=<ltrb 0 0 0 0>
        outlineSettings {
         bUseBrushTransparency=false
         roundingType=FixedRadius
         width=1
         color {colorUseRule=UseColor_ColorTable, specifiedColor=<rgba 0 0 0 0>}
         cornerRadii=<4 4 4 4>
        }
        tintColor {colorUseRule=UseColor_Specified, specifiedColor=<rgba 0.0048 0.0048 0.0048 1>}
        uVRegion {bIsValid=false, max=<0 0>, min=<0 0>}
       }
       normalForeground {colorUseRule=UseColor_ColorTable, specifiedColor=<rgba 0 0 0 0>}
       normalPadding=<ltrb 8 1 8 1>
       pressed {
        bIsDynamicallyLoaded=false
        drawAs=RoundedBox
        imageType=NoImage
        mirroring=NoMirror
        resourceName=None
        resourceObject=null
        tiling=NoTile
        imageSize=<0 0>
        margin=<ltrb 0 0 0 0>
        outlineSettings {
         bUseBrushTransparency=false
         roundingType=FixedRadius
         width=1
         color {colorUseRule=UseColor_ColorTable, specifiedColor=<rgba 0 0 0 0>}
         cornerRadii=<4 4 4 4>
        }
        tintColor {colorUseRule=UseColor_Specified, specifiedColor=<rgba 0.0048 0.0048 0.0048 1>}
        uVRegion {bIsValid=false, max=<0 0>, min=<0 0>}
       }
       pressedForeground {colorUseRule=UseColor_ColorTable, specifiedColor=<rgba 0 0 0 0>}
       pressedPadding=<ltrb 8 2 8 0>
       pressedSlateSound {resourceObject=null}
      }
      contentPadding=<ltrb 0 0 0 0>
      downArrowImage {
       bIsDynamicallyLoaded=false
       drawAs=Image
       imageType=Vector
       mirroring=NoMirror
       resourceName=../../../Engine/Content/Slate/Starship/CoreWidgets/ComboBox/wide-chevron-down.svg
       resourceObject=null
       tiling=NoTile
       imageSize=<16 16>
       margin=<ltrb 0 0 0 0>
       outlineSettings {
        bUseBrushTransparency=false
        roundingType=HalfHeightRadius
        width=0
        color {colorUseRule=UseColor_Specified, specifiedColor=<rgba 0 0 0 0>}
        cornerRadii=<0 0 0 0>
       }
       tintColor {colorUseRule=UseColor_Specified, specifiedColor=<rgba 1 1 1 1>}
       uVRegion {bIsValid=false, max=<0 0>, min=<0 0>}
      }
      downArrowPadding=<ltrb 2 2 2 2>
      menuBorderBrush {
       bIsDynamicallyLoaded=false
       drawAs=RoundedBox
       imageType=NoImage
       mirroring=NoMirror
       resourceName=None
       resourceObject=null
       tiling=NoTile
       imageSize=<0 0>
       margin=<ltrb 0 0 0 0>
       outlineSettings {
        bUseBrushTransparency=false
        roundingType=FixedRadius
        width=1
        color {colorUseRule=UseColor_ColorTable, specifiedColor=<rgba 0 0 0 0>}
        cornerRadii=<0 0 0 0>
       }
       tintColor {colorUseRule=UseColor_Specified, specifiedColor=<rgba 0.0395 0.0395 0.0395 1>}
       uVRegion {bIsValid=false, max=<0 0>, min=<0 0>}
      }
      menuBorderPadding=<ltrb 0 0 0 0>
      shadowColorAndOpacity=<rgba 0 0 0 1>
      shadowOffset=<0 0>
     }
     contentPadding=<ltrb 0 0 0 0>
     menuRowPadding=<ltrb 8 3 8 3>
     pressedSlateSound {resourceObject=null}
     selectionChangeSlateSound {resourceObject=null}
    }
   }
   slot {class=HorizontalBoxSlot}
  }
  BR_Art_LightingBox {
   widgetClassPath=&HorizontalBox
   bIsVariable=true
   parent=VerticalBox_6
   widget=BR_Art_LightingBox
   properties {visibility=Hidden}
   slot {class=VerticalBoxSlot}
  }
  ToggleToCurrentTODM {
   widgetClassPath=/Script/UMG.CheckBox
   bIsVariable=true
   parent=BR_Art_LightingBox
   widget=ToggleToCurrentTODM
   properties {widgetStyle {
     checkBoxType=CheckBox
     backgroundHoveredImage {
      bIsDynamicallyLoaded=false
      drawAs=RoundedBox
      imageType=NoImage
      mirroring=NoMirror
      resourceName=None
      resourceObject=null
      tiling=NoTile
      imageSize=<18 18>
      margin=<ltrb 0 0 0 0>
      outlineSettings {
       bUseBrushTransparency=false
       roundingType=FixedRadius
       width=1
       color {colorUseRule=UseColor_ColorTable, specifiedColor=<rgba 0 0 0 0>}
       cornerRadii=<3 3 3 3>
      }
      tintColor {colorUseRule=UseColor_Specified, specifiedColor=<rgba 0.0048 0.0048 0.0048 1>}
      uVRegion {bIsValid=false, max=<0 0>, min=<0 0>}
     }
     backgroundImage {
      bIsDynamicallyLoaded=false
      drawAs=RoundedBox
      imageType=NoImage
      mirroring=NoMirror
      resourceName=None
      resourceObject=null
      tiling=NoTile
      imageSize=<18 18>
      margin=<ltrb 0 0 0 0>
      outlineSettings {
       bUseBrushTransparency=false
       roundingType=FixedRadius
       width=1
       color {colorUseRule=UseColor_ColorTable, specifiedColor=<rgba 0 0 0 0>}
       cornerRadii=<3 3 3 3>
      }
      tintColor {colorUseRule=UseColor_Specified, specifiedColor=<rgba 0.0048 0.0048 0.0048 1>}
      uVRegion {bIsValid=false, max=<0 0>, min=<0 0>}
     }
     backgroundPressedImage {
      bIsDynamicallyLoaded=false
      drawAs=RoundedBox
      imageType=NoImage
      mirroring=NoMirror
      resourceName=None
      resourceObject=null
      tiling=NoTile
      imageSize=<18 18>
      margin=<ltrb 0 0 0 0>
      outlineSettings {
       bUseBrushTransparency=false
       roundingType=FixedRadius
       width=1
       color {colorUseRule=UseColor_ColorTable, specifiedColor=<rgba 0 0 0 0>}
       cornerRadii=<3 3 3 3>
      }
      tintColor {colorUseRule=UseColor_Specified, specifiedColor=<rgba 0.0048 0.0048 0.0048 1>}
      uVRegion {bIsValid=false, max=<0 0>, min=<0 0>}
     }
     borderBackgroundColor {colorUseRule=UseColor_Specified, specifiedColor=<rgba 1 1 1 1>}
     checkedForeground {colorUseRule=UseColor_Specified, specifiedColor=<rgba 1 1 1 1>}
     checkedHoveredForeground {colorUseRule=UseColor_Specified, specifiedColor=<rgba 1 1 1 1>}
     checkedHoveredImage {
      bIsDynamicallyLoaded=false
      drawAs=Image
      imageType=Vector
      mirroring=NoMirror
      resourceName=../../../Engine/Content/Slate/Starship/CoreWidgets/CheckBox/check.svg
      resourceObject=null
      tiling=NoTile
      imageSize=<16 16>
      margin=<ltrb 0 0 0 0>
      outlineSettings {
       bUseBrushTransparency=false
       roundingType=HalfHeightRadius
       width=0
       color {colorUseRule=UseColor_Specified, specifiedColor=<rgba 0 0 0 0>}
       cornerRadii=<0 0 0 0>
      }
      tintColor {colorUseRule=UseColor_Specified, specifiedColor=<rgba 0.0044 0.2384 1 1>}
      uVRegion {bIsValid=false, max=<0 0>, min=<0 0>}
     }
     checkedImage {
      bIsDynamicallyLoaded=false
      drawAs=Image
      imageType=Vector
      mirroring=NoMirror
      resourceName=../../../Engine/Content/Slate/Starship/CoreWidgets/CheckBox/check.svg
      resourceObject=null
      tiling=NoTile
      imageSize=<16 16>
      margin=<ltrb 0 0 0 0>
      outlineSettings {
       bUseBrushTransparency=false
       roundingType=HalfHeightRadius
       width=0
       color {colorUseRule=UseColor_Specified, specifiedColor=<rgba 0 0 0 0>}
       cornerRadii=<0 0 0 0>
      }
      tintColor {colorUseRule=UseColor_Specified, specifiedColor=<rgba 0 0.162 0.7454 1>}
      uVRegion {bIsValid=false, max=<0 0>, min=<0 0>}
     }
     checkedPressedForeground {colorUseRule=UseColor_Specified, specifiedColor=<rgba 1 1 1 1>}
     checkedPressedImage {
      bIsDynamicallyLoaded=false
      drawAs=Image
      imageType=Vector
      mirroring=NoMirror
      resourceName=../../../Engine/Content/Slate/Starship/CoreWidgets/CheckBox/check.svg
      resourceObject=null
      tiling=NoTile
      imageSize=<16 16>
      margin=<ltrb 0 0 0 0>
      outlineSettings {
       bUseBrushTransparency=false
       roundingType=HalfHeightRadius
       width=0
       color {colorUseRule=UseColor_Specified, specifiedColor=<rgba 0 0 0 0>}
       cornerRadii=<0 0 0 0>
      }
      tintColor {colorUseRule=UseColor_Specified, specifiedColor=<rgba 0 0.0802 0.3515 1>}
      uVRegion {bIsValid=false, max=<0 0>, min=<0 0>}
     }
     checkedSlateSound {resourceObject=null}
     foregroundColor {colorUseRule=UseColor_Specified, specifiedColor=<rgba 1 1 1 1>}
     hoveredForeground {colorUseRule=UseColor_Specified, specifiedColor=<rgba 1 1 1 1>}
     hoveredSlateSound {resourceObject=null}
     padding=<ltrb 2 0 0 0>
     pressedForeground {colorUseRule=UseColor_Specified, specifiedColor=<rgba 1 1 1 1>}
     uncheckedHoveredImage {
      bIsDynamicallyLoaded=false
      drawAs=NoDrawType
      imageType=NoImage
      mirroring=NoMirror
      resourceName=None
      resourceObject=null
      tiling=NoTile
      imageSize=<0 0>
      margin=<ltrb 0 0 0 0>
      outlineSettings {
       bUseBrushTransparency=false
       roundingType=HalfHeightRadius
       width=0
       color {colorUseRule=UseColor_Specified, specifiedColor=<rgba 0 0 0 0>}
       cornerRadii=<0 0 0 0>
      }
      tintColor {colorUseRule=UseColor_Specified, specifiedColor=<rgba 1 1 1 1>}
      uVRegion {bIsValid=false, max=<0 0>, min=<0 0>}
     }
     uncheckedImage {
      bIsDynamicallyLoaded=false
      drawAs=NoDrawType
      imageType=NoImage
      mirroring=NoMirror
      resourceName=None
      resourceObject=null
      tiling=NoTile
      imageSize=<0 0>
      margin=<ltrb 0 0 0 0>
      outlineSettings {
       bUseBrushTransparency=false
       roundingType=HalfHeightRadius
       width=0
       color {colorUseRule=UseColor_Specified, specifiedColor=<rgba 0 0 0 0>}
       cornerRadii=<0 0 0 0>
      }
      tintColor {colorUseRule=UseColor_Specified, specifiedColor=<rgba 1 1 1 1>}
      uVRegion {bIsValid=false, max=<0 0>, min=<0 0>}
     }
     uncheckedPressedImage {
      bIsDynamicallyLoaded=false
      drawAs=NoDrawType
      imageType=NoImage
      mirroring=NoMirror
      resourceName=None
      resourceObject=null
      tiling=NoTile
      imageSize=<0 0>
      margin=<ltrb 0 0 0 0>
      outlineSettings {
       bUseBrushTransparency=false
       roundingType=HalfHeightRadius
       width=0
       color {colorUseRule=UseColor_Specified, specifiedColor=<rgba 0 0 0 0>}
       cornerRadii=<0 0 0 0>
      }
      tintColor {colorUseRule=UseColor_Specified, specifiedColor=<rgba 1 1 1 1>}
      uVRegion {bIsValid=false, max=<0 0>, min=<0 0>}
     }
     uncheckedSlateSound {resourceObject=null}
     undeterminedForeground {colorUseRule=UseColor_Specified, specifiedColor=<rgba 1 1 1 1>}
     undeterminedHoveredImage {
      bIsDynamicallyLoaded=false
      drawAs=Image
      imageType=Vector
      mirroring=NoMirror
      resourceName=../../../Engine/Content/Slate/Starship/CoreWidgets/CheckBox/indeterminate.svg
      resourceObject=null
      tiling=NoTile
      imageSize=<16 16>
      margin=<ltrb 0 0 0 0>
      outlineSettings {
       bUseBrushTransparency=false
       roundingType=HalfHeightRadius
       width=0
       color {colorUseRule=UseColor_Specified, specifiedColor=<rgba 0 0 0 0>}
       cornerRadii=<0 0 0 0>
      }
      tintColor {colorUseRule=UseColor_Specified, specifiedColor=<rgba 0.0044 0.2384 1 1>}
      uVRegion {bIsValid=false, max=<0 0>, min=<0 0>}
     }
     undeterminedImage {
      bIsDynamicallyLoaded=false
      drawAs=Image
      imageType=Vector
      mirroring=NoMirror
      resourceName=../../../Engine/Content/Slate/Starship/CoreWidgets/CheckBox/indeterminate.svg
      resourceObject=null
      tiling=NoTile
      imageSize=<16 16>
      margin=<ltrb 0 0 0 0>
      outlineSettings {
       bUseBrushTransparency=false
       roundingType=HalfHeightRadius
       width=0
       color {colorUseRule=UseColor_Specified, specifiedColor=<rgba 0 0 0 0>}
       cornerRadii=<0 0 0 0>
      }
      tintColor {colorUseRule=UseColor_Specified, specifiedColor=<rgba 0 0.162 0.7454 1>}
      uVRegion {bIsValid=false, max=<0 0>, min=<0 0>}
     }
     undeterminedPressedImage {
      bIsDynamicallyLoaded=false
      drawAs=Image
      imageType=Vector
      mirroring=NoMirror
      resourceName=../../../Engine/Content/Slate/Starship/CoreWidgets/CheckBox/indeterminate.svg
      resourceObject=null
      tiling=NoTile
      imageSize=<16 16>
      margin=<ltrb 0 0 0 0>
      outlineSettings {
       bUseBrushTransparency=false
       roundingType=HalfHeightRadius
       width=0
       color {colorUseRule=UseColor_Specified, specifiedColor=<rgba 0 0 0 0>}
       cornerRadii=<0 0 0 0>
      }
      tintColor {colorUseRule=UseColor_Specified, specifiedColor=<rgba 0 0.0802 0.3515 1>}
      uVRegion {bIsValid=false, max=<0 0>, min=<0 0>}
     }
    }}
   slot {class=HorizontalBoxSlot}
  }
  Spacer_12 {
   widgetClassPath=&Spacer
   parent=BR_Art_LightingBox
   widget=Spacer_12
   slot {class=HorizontalBoxSlot, properties {padding=<ltrb 0 0 10 0>}}
  }
  TextBlock_7 {
   widgetClassPath=&TextBlock
   parent=BR_Art_LightingBox
   widget=TextBlock_7
   properties {
    text=Toggle TODM to BR_Art_Current
    font {
     bForceMonospaced=false
     bMaterialIsStencil=false
     fontMaterial=null
     fontObject=&Roboto
     letterSpacing=0
     monospacedWidth=1
     size=9
     skewAmount=0
     typefaceFontName=Regular
     outlineSettings {
      bApplyOutlineToDropShadows=false
      bMiteredCorners=false
      bSeparateFillAlpha=false
      outlineMaterial=null
      outlineSize=0
      outlineColor=<rgba 0 0 0 1>
     }
    }
   }
   slot {class=HorizontalBoxSlot, properties {verticalAlignment=VAlign_Bottom}}
  }
  BackpackBox {widgetClassPath=&HorizontalBox, parent=HorizontalBox_2, widget=BackpackBox, slot {class=HorizontalBoxSlot}}
  Spacer_284 {
   widgetClassPath=&Spacer
   parent=BackpackBox
   widget=Spacer_284
   slot {class=HorizontalBoxSlot, properties {padding=<ltrb 0 0 20 0>}}
  }
  BackpackMode {
   widgetClassPath=&CheckBox
   bIsVariable=true
   parent=BackpackBox
   widget=BackpackMode
   properties {widgetStyle {
     checkBoxType=CheckBox
     backgroundHoveredImage {
      bIsDynamicallyLoaded=false
      drawAs=RoundedBox
      imageType=NoImage
      mirroring=NoMirror
      resourceName=None
      resourceObject=null
      tiling=NoTile
      imageSize=<18 18>
      margin=<ltrb 0 0 0 0>
      outlineSettings {
       bUseBrushTransparency=false
       roundingType=FixedRadius
       width=1
       color {colorUseRule=UseColor_ColorTable, specifiedColor=<rgba 0 0 0 0>}
       cornerRadii=<3 3 3 3>
      }
      tintColor {colorUseRule=UseColor_Specified, specifiedColor=<rgba 0.0048 0.0048 0.0048 1>}
      uVRegion {bIsValid=false, max=<0 0>, min=<0 0>}
     }
     backgroundImage {
      bIsDynamicallyLoaded=false
      drawAs=RoundedBox
      imageType=NoImage
      mirroring=NoMirror
      resourceName=None
      resourceObject=null
      tiling=NoTile
      imageSize=<18 18>
      margin=<ltrb 0 0 0 0>
      outlineSettings {
       bUseBrushTransparency=false
       roundingType=FixedRadius
       width=1
       color {colorUseRule=UseColor_ColorTable, specifiedColor=<rgba 0 0 0 0>}
       cornerRadii=<3 3 3 3>
      }
      tintColor {colorUseRule=UseColor_Specified, specifiedColor=<rgba 0.0048 0.0048 0.0048 1>}
      uVRegion {bIsValid=false, max=<0 0>, min=<0 0>}
     }
     backgroundPressedImage {
      bIsDynamicallyLoaded=false
      drawAs=RoundedBox
      imageType=NoImage
      mirroring=NoMirror
      resourceName=None
      resourceObject=null
      tiling=NoTile
      imageSize=<18 18>
      margin=<ltrb 0 0 0 0>
      outlineSettings {
       bUseBrushTransparency=false
       roundingType=FixedRadius
       width=1
       color {colorUseRule=UseColor_ColorTable, specifiedColor=<rgba 0 0 0 0>}
       cornerRadii=<3 3 3 3>
      }
      tintColor {colorUseRule=UseColor_Specified, specifiedColor=<rgba 0.0048 0.0048 0.0048 1>}
      uVRegion {bIsValid=false, max=<0 0>, min=<0 0>}
     }
     borderBackgroundColor {colorUseRule=UseColor_Specified, specifiedColor=<rgba 1 1 1 1>}
     checkedForeground {colorUseRule=UseColor_Specified, specifiedColor=<rgba 1 1 1 1>}
     checkedHoveredForeground {colorUseRule=UseColor_Specified, specifiedColor=<rgba 1 1 1 1>}
     checkedHoveredImage {
      bIsDynamicallyLoaded=false
      drawAs=Image
      imageType=Vector
      mirroring=NoMirror
      resourceName=../../../Engine/Content/Slate/Starship/CoreWidgets/CheckBox/check.svg
      resourceObject=null
      tiling=NoTile
      imageSize=<16 16>
      margin=<ltrb 0 0 0 0>
      outlineSettings {
       bUseBrushTransparency=false
       roundingType=HalfHeightRadius
       width=0
       color {colorUseRule=UseColor_Specified, specifiedColor=<rgba 0 0 0 0>}
       cornerRadii=<0 0 0 0>
      }
      tintColor {colorUseRule=UseColor_Specified, specifiedColor=<rgba 0.0044 0.2384 1 1>}
      uVRegion {bIsValid=false, max=<0 0>, min=<0 0>}
     }
     checkedImage {
      bIsDynamicallyLoaded=false
      drawAs=Image
      imageType=Vector
      mirroring=NoMirror
      resourceName=../../../Engine/Content/Slate/Starship/CoreWidgets/CheckBox/check.svg
      resourceObject=null
      tiling=NoTile
      imageSize=<16 16>
      margin=<ltrb 0 0 0 0>
      outlineSettings {
       bUseBrushTransparency=false
       roundingType=HalfHeightRadius
       width=0
       color {colorUseRule=UseColor_Specified, specifiedColor=<rgba 0 0 0 0>}
       cornerRadii=<0 0 0 0>
      }
      tintColor {colorUseRule=UseColor_Specified, specifiedColor=<rgba 0 0.162 0.7454 1>}
      uVRegion {bIsValid=false, max=<0 0>, min=<0 0>}
     }
     checkedPressedForeground {colorUseRule=UseColor_Specified, specifiedColor=<rgba 1 1 1 1>}
     checkedPressedImage {
      bIsDynamicallyLoaded=false
      drawAs=Image
      imageType=Vector
      mirroring=NoMirror
      resourceName=../../../Engine/Content/Slate/Starship/CoreWidgets/CheckBox/check.svg
      resourceObject=null
      tiling=NoTile
      imageSize=<16 16>
      margin=<ltrb 0 0 0 0>
      outlineSettings {
       bUseBrushTransparency=false
       roundingType=HalfHeightRadius
       width=0
       color {colorUseRule=UseColor_Specified, specifiedColor=<rgba 0 0 0 0>}
       cornerRadii=<0 0 0 0>
      }
      tintColor {colorUseRule=UseColor_Specified, specifiedColor=<rgba 0 0.0802 0.3515 1>}
      uVRegion {bIsValid=false, max=<0 0>, min=<0 0>}
     }
     checkedSlateSound {resourceObject=null}
     foregroundColor {colorUseRule=UseColor_Specified, specifiedColor=<rgba 1 1 1 1>}
     hoveredForeground {colorUseRule=UseColor_Specified, specifiedColor=<rgba 1 1 1 1>}
     hoveredSlateSound {resourceObject=null}
     padding=<ltrb 2 0 0 0>
     pressedForeground {colorUseRule=UseColor_Specified, specifiedColor=<rgba 1 1 1 1>}
     uncheckedHoveredImage {
      bIsDynamicallyLoaded=false
      drawAs=NoDrawType
      imageType=NoImage
      mirroring=NoMirror
      resourceName=None
      resourceObject=null
      tiling=NoTile
      imageSize=<0 0>
      margin=<ltrb 0 0 0 0>
      outlineSettings {
       bUseBrushTransparency=false
       roundingType=HalfHeightRadius
       width=0
       color {colorUseRule=UseColor_Specified, specifiedColor=<rgba 0 0 0 0>}
       cornerRadii=<0 0 0 0>
      }
      tintColor {colorUseRule=UseColor_Specified, specifiedColor=<rgba 1 1 1 1>}
      uVRegion {bIsValid=false, max=<0 0>, min=<0 0>}
     }
     uncheckedImage {
      bIsDynamicallyLoaded=false
      drawAs=NoDrawType
      imageType=NoImage
      mirroring=NoMirror
      resourceName=None
      resourceObject=null
      tiling=NoTile
      imageSize=<0 0>
      margin=<ltrb 0 0 0 0>
      outlineSettings {
       bUseBrushTransparency=false
       roundingType=HalfHeightRadius
       width=0
       color {colorUseRule=UseColor_Specified, specifiedColor=<rgba 0 0 0 0>}
       cornerRadii=<0 0 0 0>
      }
      tintColor {colorUseRule=UseColor_Specified, specifiedColor=<rgba 1 1 1 1>}
      uVRegion {bIsValid=false, max=<0 0>, min=<0 0>}
     }
     uncheckedPressedImage {
      bIsDynamicallyLoaded=false
      drawAs=NoDrawType
      imageType=NoImage
      mirroring=NoMirror
      resourceName=None
      resourceObject=null
      tiling=NoTile
      imageSize=<0 0>
      margin=<ltrb 0 0 0 0>
      outlineSettings {
       bUseBrushTransparency=false
       roundingType=HalfHeightRadius
       width=0
       color {colorUseRule=UseColor_Specified, specifiedColor=<rgba 0 0 0 0>}
       cornerRadii=<0 0 0 0>
      }
      tintColor {colorUseRule=UseColor_Specified, specifiedColor=<rgba 1 1 1 1>}
      uVRegion {bIsValid=false, max=<0 0>, min=<0 0>}
     }
     uncheckedSlateSound {resourceObject=null}
     undeterminedForeground {colorUseRule=UseColor_Specified, specifiedColor=<rgba 1 1 1 1>}
     undeterminedHoveredImage {
      bIsDynamicallyLoaded=false
      drawAs=Image
      imageType=Vector
      mirroring=NoMirror
      resourceName=../../../Engine/Content/Slate/Starship/CoreWidgets/CheckBox/indeterminate.svg
      resourceObject=null
      tiling=NoTile
      imageSize=<16 16>
      margin=<ltrb 0 0 0 0>
      outlineSettings {
       bUseBrushTransparency=false
       roundingType=HalfHeightRadius
       width=0
       color {colorUseRule=UseColor_Specified, specifiedColor=<rgba 0 0 0 0>}
       cornerRadii=<0 0 0 0>
      }
      tintColor {colorUseRule=UseColor_Specified, specifiedColor=<rgba 0.0044 0.2384 1 1>}
      uVRegion {bIsValid=false, max=<0 0>, min=<0 0>}
     }
     undeterminedImage {
      bIsDynamicallyLoaded=false
      drawAs=Image
      imageType=Vector
      mirroring=NoMirror
      resourceName=../../../Engine/Content/Slate/Starship/CoreWidgets/CheckBox/indeterminate.svg
      resourceObject=null
      tiling=NoTile
      imageSize=<16 16>
      margin=<ltrb 0 0 0 0>
      outlineSettings {
       bUseBrushTransparency=false
       roundingType=HalfHeightRadius
       width=0
       color {colorUseRule=UseColor_Specified, specifiedColor=<rgba 0 0 0 0>}
       cornerRadii=<0 0 0 0>
      }
      tintColor {colorUseRule=UseColor_Specified, specifiedColor=<rgba 0 0.162 0.7454 1>}
      uVRegion {bIsValid=false, max=<0 0>, min=<0 0>}
     }
     undeterminedPressedImage {
      bIsDynamicallyLoaded=false
      drawAs=Image
      imageType=Vector
      mirroring=NoMirror
      resourceName=../../../Engine/Content/Slate/Starship/CoreWidgets/CheckBox/indeterminate.svg
      resourceObject=null
      tiling=NoTile
      imageSize=<16 16>
      margin=<ltrb 0 0 0 0>
      outlineSettings {
       bUseBrushTransparency=false
       roundingType=HalfHeightRadius
       width=0
       color {colorUseRule=UseColor_Specified, specifiedColor=<rgba 0 0 0 0>}
       cornerRadii=<0 0 0 0>
      }
      tintColor {colorUseRule=UseColor_Specified, specifiedColor=<rgba 0 0.0802 0.3515 1>}
      uVRegion {bIsValid=false, max=<0 0>, min=<0 0>}
     }
    }}
   slot {class=HorizontalBoxSlot}
  }
  Spacer_383 {
   widgetClassPath=&Spacer
   parent=BackpackBox
   widget=Spacer_383
   slot {class=HorizontalBoxSlot, properties {padding=<ltrb 0 0 10 0>}}
  }
  TextBlock_174 {
   widgetClassPath=&TextBlock
   parent=BackpackBox
   widget=TextBlock_174
   properties {
    text=Backpack Mode
    font {
     bForceMonospaced=false
     bMaterialIsStencil=false
     fontMaterial=null
     fontObject=&Roboto
     letterSpacing=0
     monospacedWidth=1
     size=14
     skewAmount=0
     typefaceFontName=Bold
     outlineSettings {
      bApplyOutlineToDropShadows=false
      bMiteredCorners=false
      bSeparateFillAlpha=false
      outlineMaterial=null
      outlineSize=0
      outlineColor=<rgba 0 0 0 1>
     }
    }
   }
   slot {class=HorizontalBoxSlot, properties {verticalAlignment=VAlign_Center}}
  }
  VehicleMode {widgetClassPath=&HorizontalBox, parent=HorizontalBox_2, widget=VehicleMode, slot {class=HorizontalBoxSlot}}
  Spacer_10 {
   widgetClassPath=&Spacer
   parent=VehicleMode
   widget=Spacer_10
   slot {class=HorizontalBoxSlot, properties {padding=<ltrb 0 0 20 0>}}
  }
  VehicleModeSelect {
   widgetClassPath=&CheckBox
   bIsVariable=true
   parent=VehicleMode
   widget=VehicleModeSelect
   properties {widgetStyle {
     checkBoxType=CheckBox
     backgroundHoveredImage {
      bIsDynamicallyLoaded=false
      drawAs=RoundedBox
      imageType=NoImage
      mirroring=NoMirror
      resourceName=None
      resourceObject=null
      tiling=NoTile
      imageSize=<18 18>
      margin=<ltrb 0 0 0 0>
      outlineSettings {
       bUseBrushTransparency=false
       roundingType=FixedRadius
       width=1
       color {colorUseRule=UseColor_ColorTable, specifiedColor=<rgba 0 0 0 0>}
       cornerRadii=<3 3 3 3>
      }
      tintColor {colorUseRule=UseColor_Specified, specifiedColor=<rgba 0.0048 0.0048 0.0048 1>}
      uVRegion {bIsValid=false, max=<0 0>, min=<0 0>}
     }
     backgroundImage {
      bIsDynamicallyLoaded=false
      drawAs=RoundedBox
      imageType=NoImage
      mirroring=NoMirror
      resourceName=None
      resourceObject=null
      tiling=NoTile
      imageSize=<18 18>
      margin=<ltrb 0 0 0 0>
      outlineSettings {
       bUseBrushTransparency=false
       roundingType=FixedRadius
       width=1
       color {colorUseRule=UseColor_ColorTable, specifiedColor=<rgba 0 0 0 0>}
       cornerRadii=<3 3 3 3>
      }
      tintColor {colorUseRule=UseColor_Specified, specifiedColor=<rgba 0.0048 0.0048 0.0048 1>}
      uVRegion {bIsValid=false, max=<0 0>, min=<0 0>}
     }
     backgroundPressedImage {
      bIsDynamicallyLoaded=false
      drawAs=RoundedBox
      imageType=NoImage
      mirroring=NoMirror
      resourceName=None
      resourceObject=null
      tiling=NoTile
      imageSize=<18 18>
      margin=<ltrb 0 0 0 0>
      outlineSettings {
       bUseBrushTransparency=false
       roundingType=FixedRadius
       width=1
       color {colorUseRule=UseColor_ColorTable, specifiedColor=<rgba 0 0 0 0>}
       cornerRadii=<3 3 3 3>
      }
      tintColor {colorUseRule=UseColor_Specified, specifiedColor=<rgba 0.0048 0.0048 0.0048 1>}
      uVRegion {bIsValid=false, max=<0 0>, min=<0 0>}
     }
     borderBackgroundColor {colorUseRule=UseColor_Specified, specifiedColor=<rgba 1 1 1 1>}
     checkedForeground {colorUseRule=UseColor_Specified, specifiedColor=<rgba 1 1 1 1>}
     checkedHoveredForeground {colorUseRule=UseColor_Specified, specifiedColor=<rgba 1 1 1 1>}
     checkedHoveredImage {
      bIsDynamicallyLoaded=false
      drawAs=Image
      imageType=Vector
      mirroring=NoMirror
      resourceName=../../../Engine/Content/Slate/Starship/CoreWidgets/CheckBox/check.svg
      resourceObject=null
      tiling=NoTile
      imageSize=<16 16>
      margin=<ltrb 0 0 0 0>
      outlineSettings {
       bUseBrushTransparency=false
       roundingType=HalfHeightRadius
       width=0
       color {colorUseRule=UseColor_Specified, specifiedColor=<rgba 0 0 0 0>}
       cornerRadii=<0 0 0 0>
      }
      tintColor {colorUseRule=UseColor_Specified, specifiedColor=<rgba 0.0044 0.2384 1 1>}
      uVRegion {bIsValid=false, max=<0 0>, min=<0 0>}
     }
     checkedImage {
      bIsDynamicallyLoaded=false
      drawAs=Image
      imageType=Vector
      mirroring=NoMirror
      resourceName=../../../Engine/Content/Slate/Starship/CoreWidgets/CheckBox/check.svg
      resourceObject=null
      tiling=NoTile
      imageSize=<16 16>
      margin=<ltrb 0 0 0 0>
      outlineSettings {
       bUseBrushTransparency=false
       roundingType=HalfHeightRadius
       width=0
       color {colorUseRule=UseColor_Specified, specifiedColor=<rgba 0 0 0 0>}
       cornerRadii=<0 0 0 0>
      }
      tintColor {colorUseRule=UseColor_Specified, specifiedColor=<rgba 0 0.162 0.7454 1>}
      uVRegion {bIsValid=false, max=<0 0>, min=<0 0>}
     }
     checkedPressedForeground {colorUseRule=UseColor_Specified, specifiedColor=<rgba 1 1 1 1>}
     checkedPressedImage {
      bIsDynamicallyLoaded=false
      drawAs=Image
      imageType=Vector
      mirroring=NoMirror
      resourceName=../../../Engine/Content/Slate/Starship/CoreWidgets/CheckBox/check.svg
      resourceObject=null
      tiling=NoTile
      imageSize=<16 16>
      margin=<ltrb 0 0 0 0>
      outlineSettings {
       bUseBrushTransparency=false
       roundingType=HalfHeightRadius
       width=0
       color {colorUseRule=UseColor_Specified, specifiedColor=<rgba 0 0 0 0>}
       cornerRadii=<0 0 0 0>
      }
      tintColor {colorUseRule=UseColor_Specified, specifiedColor=<rgba 0 0.0802 0.3515 1>}
      uVRegion {bIsValid=false, max=<0 0>, min=<0 0>}
     }
     checkedSlateSound {resourceObject=null}
     foregroundColor {colorUseRule=UseColor_Specified, specifiedColor=<rgba 1 1 1 1>}
     hoveredForeground {colorUseRule=UseColor_Specified, specifiedColor=<rgba 1 1 1 1>}
     hoveredSlateSound {resourceObject=null}
     padding=<ltrb 2 0 0 0>
     pressedForeground {colorUseRule=UseColor_Specified, specifiedColor=<rgba 1 1 1 1>}
     uncheckedHoveredImage {
      bIsDynamicallyLoaded=false
      drawAs=NoDrawType
      imageType=NoImage
      mirroring=NoMirror
      resourceName=None
      resourceObject=null
      tiling=NoTile
      imageSize=<0 0>
      margin=<ltrb 0 0 0 0>
      outlineSettings {
       bUseBrushTransparency=false
       roundingType=HalfHeightRadius
       width=0
       color {colorUseRule=UseColor_Specified, specifiedColor=<rgba 0 0 0 0>}
       cornerRadii=<0 0 0 0>
      }
      tintColor {colorUseRule=UseColor_Specified, specifiedColor=<rgba 1 1 1 1>}
      uVRegion {bIsValid=false, max=<0 0>, min=<0 0>}
     }
     uncheckedImage {
      bIsDynamicallyLoaded=false
      drawAs=NoDrawType
      imageType=NoImage
      mirroring=NoMirror
      resourceName=None
      resourceObject=null
      tiling=NoTile
      imageSize=<0 0>
      margin=<ltrb 0 0 0 0>
      outlineSettings {
       bUseBrushTransparency=false
       roundingType=HalfHeightRadius
       width=0
       color {colorUseRule=UseColor_Specified, specifiedColor=<rgba 0 0 0 0>}
       cornerRadii=<0 0 0 0>
      }
      tintColor {colorUseRule=UseColor_Specified, specifiedColor=<rgba 1 1 1 1>}
      uVRegion {bIsValid=false, max=<0 0>, min=<0 0>}
     }
     uncheckedPressedImage {
      bIsDynamicallyLoaded=false
      drawAs=NoDrawType
      imageType=NoImage
      mirroring=NoMirror
      resourceName=None
      resourceObject=null
      tiling=NoTile
      imageSize=<0 0>
      margin=<ltrb 0 0 0 0>
      outlineSettings {
       bUseBrushTransparency=false
       roundingType=HalfHeightRadius
       width=0
       color {colorUseRule=UseColor_Specified, specifiedColor=<rgba 0 0 0 0>}
       cornerRadii=<0 0 0 0>
      }
      tintColor {colorUseRule=UseColor_Specified, specifiedColor=<rgba 1 1 1 1>}
      uVRegion {bIsValid=false, max=<0 0>, min=<0 0>}
     }
     uncheckedSlateSound {resourceObject=null}
     undeterminedForeground {colorUseRule=UseColor_Specified, specifiedColor=<rgba 1 1 1 1>}
     undeterminedHoveredImage {
      bIsDynamicallyLoaded=false
      drawAs=Image
      imageType=Vector
      mirroring=NoMirror
      resourceName=../../../Engine/Content/Slate/Starship/CoreWidgets/CheckBox/indeterminate.svg
      resourceObject=null
      tiling=NoTile
      imageSize=<16 16>
      margin=<ltrb 0 0 0 0>
      outlineSettings {
       bUseBrushTransparency=false
       roundingType=HalfHeightRadius
       width=0
       color {colorUseRule=UseColor_Specified, specifiedColor=<rgba 0 0 0 0>}
       cornerRadii=<0 0 0 0>
      }
      tintColor {colorUseRule=UseColor_Specified, specifiedColor=<rgba 0.0044 0.2384 1 1>}
      uVRegion {bIsValid=false, max=<0 0>, min=<0 0>}
     }
     undeterminedImage {
      bIsDynamicallyLoaded=false
      drawAs=Image
      imageType=Vector
      mirroring=NoMirror
      resourceName=../../../Engine/Content/Slate/Starship/CoreWidgets/CheckBox/indeterminate.svg
      resourceObject=null
      tiling=NoTile
      imageSize=<16 16>
      margin=<ltrb 0 0 0 0>
      outlineSettings {
       bUseBrushTransparency=false
       roundingType=HalfHeightRadius
       width=0
       color {colorUseRule=UseColor_Specified, specifiedColor=<rgba 0 0 0 0>}
       cornerRadii=<0 0 0 0>
      }
      tintColor {colorUseRule=UseColor_Specified, specifiedColor=<rgba 0 0.162 0.7454 1>}
      uVRegion {bIsValid=false, max=<0 0>, min=<0 0>}
     }
     undeterminedPressedImage {
      bIsDynamicallyLoaded=false
      drawAs=Image
      imageType=Vector
      mirroring=NoMirror
      resourceName=../../../Engine/Content/Slate/Starship/CoreWidgets/CheckBox/indeterminate.svg
      resourceObject=null
      tiling=NoTile
      imageSize=<16 16>
      margin=<ltrb 0 0 0 0>
      outlineSettings {
       bUseBrushTransparency=false
       roundingType=HalfHeightRadius
       width=0
       color {colorUseRule=UseColor_Specified, specifiedColor=<rgba 0 0 0 0>}
       cornerRadii=<0 0 0 0>
      }
      tintColor {colorUseRule=UseColor_Specified, specifiedColor=<rgba 0 0.0802 0.3515 1>}
      uVRegion {bIsValid=false, max=<0 0>, min=<0 0>}
     }
    }}
   slot {class=HorizontalBoxSlot}
  }
  Spacer_11 {
   widgetClassPath=&Spacer
   parent=VehicleMode
   widget=Spacer_11
   slot {class=HorizontalBoxSlot, properties {padding=<ltrb 0 0 10 0>}}
  }
  TextBlock_6 {
   widgetClassPath=&TextBlock
   parent=VehicleMode
   widget=TextBlock_6
   properties {
    text=VehicleMode
    font {
     bForceMonospaced=false
     bMaterialIsStencil=false
     fontMaterial=null
     fontObject=&Roboto
     letterSpacing=0
     monospacedWidth=1
     size=14
     skewAmount=0
     typefaceFontName=Bold
     outlineSettings {
      bApplyOutlineToDropShadows=false
      bMiteredCorners=false
      bSeparateFillAlpha=false
      outlineMaterial=null
      outlineSize=0
      outlineColor=<rgba 0 0 0 1>
     }
    }
   }
   slot {class=HorizontalBoxSlot, properties {verticalAlignment=VAlign_Center}}
  }
  AdvancedControlsBox {widgetClassPath=&HorizontalBox, parent=HorizontalBox_2, widget=AdvancedControlsBox, slot {class=HorizontalBoxSlot}}
  Spacer_3 {
   widgetClassPath=&Spacer
   parent=AdvancedControlsBox
   widget=Spacer_3
   slot {class=HorizontalBoxSlot, properties {padding=<ltrb 0 0 40 0>}}
  }
  ShowAdvanced_Checkbox {
   widgetClassPath=&CheckBox
   bIsVariable=true
   parent=AdvancedControlsBox
   widget=ShowAdvanced_Checkbox
   properties {widgetStyle {
     checkBoxType=CheckBox
     backgroundHoveredImage {
      bIsDynamicallyLoaded=false
      drawAs=RoundedBox
      imageType=NoImage
      mirroring=NoMirror
      resourceName=None
      resourceObject=null
      tiling=NoTile
      imageSize=<18 18>
      margin=<ltrb 0 0 0 0>
      outlineSettings {
       bUseBrushTransparency=false
       roundingType=FixedRadius
       width=1
       color {colorUseRule=UseColor_ColorTable, specifiedColor=<rgba 0 0 0 0>}
       cornerRadii=<3 3 3 3>
      }
      tintColor {colorUseRule=UseColor_Specified, specifiedColor=<rgba 0.0048 0.0048 0.0048 1>}
      uVRegion {bIsValid=false, max=<0 0>, min=<0 0>}
     }
     backgroundImage {
      bIsDynamicallyLoaded=false
      drawAs=RoundedBox
      imageType=NoImage
      mirroring=NoMirror
      resourceName=None
      resourceObject=null
      tiling=NoTile
      imageSize=<18 18>
      margin=<ltrb 0 0 0 0>
      outlineSettings {
       bUseBrushTransparency=false
       roundingType=FixedRadius
       width=1
       color {colorUseRule=UseColor_ColorTable, specifiedColor=<rgba 0 0 0 0>}
       cornerRadii=<3 3 3 3>
      }
      tintColor {colorUseRule=UseColor_Specified, specifiedColor=<rgba 0.0048 0.0048 0.0048 1>}
      uVRegion {bIsValid=false, max=<0 0>, min=<0 0>}
     }
     backgroundPressedImage {
      bIsDynamicallyLoaded=false
      drawAs=RoundedBox
      imageType=NoImage
      mirroring=NoMirror
      resourceName=None
      resourceObject=null
      tiling=NoTile
      imageSize=<18 18>
      margin=<ltrb 0 0 0 0>
      outlineSettings {
       bUseBrushTransparency=false
       roundingType=FixedRadius
       width=1
       color {colorUseRule=UseColor_ColorTable, specifiedColor=<rgba 0 0 0 0>}
       cornerRadii=<3 3 3 3>
      }
      tintColor {colorUseRule=UseColor_Specified, specifiedColor=<rgba 0.0048 0.0048 0.0048 1>}
      uVRegion {bIsValid=false, max=<0 0>, min=<0 0>}
     }
     borderBackgroundColor {colorUseRule=UseColor_Specified, specifiedColor=<rgba 1 1 1 1>}
     checkedForeground {colorUseRule=UseColor_Specified, specifiedColor=<rgba 1 1 1 1>}
     checkedHoveredForeground {colorUseRule=UseColor_Specified, specifiedColor=<rgba 1 1 1 1>}
     checkedHoveredImage {
      bIsDynamicallyLoaded=false
      drawAs=Image
      imageType=Vector
      mirroring=NoMirror
      resourceName=../../../Engine/Content/Slate/Starship/CoreWidgets/CheckBox/check.svg
      resourceObject=null
      tiling=NoTile
      imageSize=<16 16>
      margin=<ltrb 0 0 0 0>
      outlineSettings {
       bUseBrushTransparency=false
       roundingType=HalfHeightRadius
       width=0
       color {colorUseRule=UseColor_Specified, specifiedColor=<rgba 0 0 0 0>}
       cornerRadii=<0 0 0 0>
      }
      tintColor {colorUseRule=UseColor_Specified, specifiedColor=<rgba 0.0044 0.2384 1 1>}
      uVRegion {bIsValid=false, max=<0 0>, min=<0 0>}
     }
     checkedImage {
      bIsDynamicallyLoaded=false
      drawAs=Image
      imageType=Vector
      mirroring=NoMirror
      resourceName=../../../Engine/Content/Slate/Starship/CoreWidgets/CheckBox/check.svg
      resourceObject=null
      tiling=NoTile
      imageSize=<16 16>
      margin=<ltrb 0 0 0 0>
      outlineSettings {
       bUseBrushTransparency=false
       roundingType=HalfHeightRadius
       width=0
       color {colorUseRule=UseColor_Specified, specifiedColor=<rgba 0 0 0 0>}
       cornerRadii=<0 0 0 0>
      }
      tintColor {colorUseRule=UseColor_Specified, specifiedColor=<rgba 0 0.162 0.7454 1>}
      uVRegion {bIsValid=false, max=<0 0>, min=<0 0>}
     }
     checkedPressedForeground {colorUseRule=UseColor_Specified, specifiedColor=<rgba 1 1 1 1>}
     checkedPressedImage {
      bIsDynamicallyLoaded=false
      drawAs=Image
      imageType=Vector
      mirroring=NoMirror
      resourceName=../../../Engine/Content/Slate/Starship/CoreWidgets/CheckBox/check.svg
      resourceObject=null
      tiling=NoTile
      imageSize=<16 16>
      margin=<ltrb 0 0 0 0>
      outlineSettings {
       bUseBrushTransparency=false
       roundingType=HalfHeightRadius
       width=0
       color {colorUseRule=UseColor_Specified, specifiedColor=<rgba 0 0 0 0>}
       cornerRadii=<0 0 0 0>
      }
      tintColor {colorUseRule=UseColor_Specified, specifiedColor=<rgba 0 0.0802 0.3515 1>}
      uVRegion {bIsValid=false, max=<0 0>, min=<0 0>}
     }
     checkedSlateSound {resourceObject=null}
     foregroundColor {colorUseRule=UseColor_Specified, specifiedColor=<rgba 1 1 1 1>}
     hoveredForeground {colorUseRule=UseColor_Specified, specifiedColor=<rgba 1 1 1 1>}
     hoveredSlateSound {resourceObject=null}
     padding=<ltrb 2 0 0 0>
     pressedForeground {colorUseRule=UseColor_Specified, specifiedColor=<rgba 1 1 1 1>}
     uncheckedHoveredImage {
      bIsDynamicallyLoaded=false
      drawAs=NoDrawType
      imageType=NoImage
      mirroring=NoMirror
      resourceName=None
      resourceObject=null
      tiling=NoTile
      imageSize=<0 0>
      margin=<ltrb 0 0 0 0>
      outlineSettings {
       bUseBrushTransparency=false
       roundingType=HalfHeightRadius
       width=0
       color {colorUseRule=UseColor_Specified, specifiedColor=<rgba 0 0 0 0>}
       cornerRadii=<0 0 0 0>
      }
      tintColor {colorUseRule=UseColor_Specified, specifiedColor=<rgba 1 1 1 1>}
      uVRegion {bIsValid=false, max=<0 0>, min=<0 0>}
     }
     uncheckedImage {
      bIsDynamicallyLoaded=false
      drawAs=NoDrawType
      imageType=NoImage
      mirroring=NoMirror
      resourceName=None
      resourceObject=null
      tiling=NoTile
      imageSize=<0 0>
      margin=<ltrb 0 0 0 0>
      outlineSettings {
       bUseBrushTransparency=false
       roundingType=HalfHeightRadius
       width=0
       color {colorUseRule=UseColor_Specified, specifiedColor=<rgba 0 0 0 0>}
       cornerRadii=<0 0 0 0>
      }
      tintColor {colorUseRule=UseColor_Specified, specifiedColor=<rgba 1 1 1 1>}
      uVRegion {bIsValid=false, max=<0 0>, min=<0 0>}
     }
     uncheckedPressedImage {
      bIsDynamicallyLoaded=false
      drawAs=NoDrawType
      imageType=NoImage
      mirroring=NoMirror
      resourceName=None
      resourceObject=null
      tiling=NoTile
      imageSize=<0 0>
      margin=<ltrb 0 0 0 0>
      outlineSettings {
       bUseBrushTransparency=false
       roundingType=HalfHeightRadius
       width=0
       color {colorUseRule=UseColor_Specified, specifiedColor=<rgba 0 0 0 0>}
       cornerRadii=<0 0 0 0>
      }
      tintColor {colorUseRule=UseColor_Specified, specifiedColor=<rgba 1 1 1 1>}
      uVRegion {bIsValid=false, max=<0 0>, min=<0 0>}
     }
     uncheckedSlateSound {resourceObject=null}
     undeterminedForeground {colorUseRule=UseColor_Specified, specifiedColor=<rgba 1 1 1 1>}
     undeterminedHoveredImage {
      bIsDynamicallyLoaded=false
      drawAs=Image
      imageType=Vector
      mirroring=NoMirror
      resourceName=../../../Engine/Content/Slate/Starship/CoreWidgets/CheckBox/indeterminate.svg
      resourceObject=null
      tiling=NoTile
      imageSize=<16 16>
      margin=<ltrb 0 0 0 0>
      outlineSettings {
       bUseBrushTransparency=false
       roundingType=HalfHeightRadius
       width=0
       color {colorUseRule=UseColor_Specified, specifiedColor=<rgba 0 0 0 0>}
       cornerRadii=<0 0 0 0>
      }
      tintColor {colorUseRule=UseColor_Specified, specifiedColor=<rgba 0.0044 0.2384 1 1>}
      uVRegion {bIsValid=false, max=<0 0>, min=<0 0>}
     }
     undeterminedImage {
      bIsDynamicallyLoaded=false
      drawAs=Image
      imageType=Vector
      mirroring=NoMirror
      resourceName=../../../Engine/Content/Slate/Starship/CoreWidgets/CheckBox/indeterminate.svg
      resourceObject=null
      tiling=NoTile
      imageSize=<16 16>
      margin=<ltrb 0 0 0 0>
      outlineSettings {
       bUseBrushTransparency=false
       roundingType=HalfHeightRadius
       width=0
       color {colorUseRule=UseColor_Specified, specifiedColor=<rgba 0 0 0 0>}
       cornerRadii=<0 0 0 0>
      }
      tintColor {colorUseRule=UseColor_Specified, specifiedColor=<rgba 0 0.162 0.7454 1>}
      uVRegion {bIsValid=false, max=<0 0>, min=<0 0>}
     }
     undeterminedPressedImage {
      bIsDynamicallyLoaded=false
      drawAs=Image
      imageType=Vector
      mirroring=NoMirror
      resourceName=../../../Engine/Content/Slate/Starship/CoreWidgets/CheckBox/indeterminate.svg
      resourceObject=null
      tiling=NoTile
      imageSize=<16 16>
      margin=<ltrb 0 0 0 0>
      outlineSettings {
       bUseBrushTransparency=false
       roundingType=HalfHeightRadius
       width=0
       color {colorUseRule=UseColor_Specified, specifiedColor=<rgba 0 0 0 0>}
       cornerRadii=<0 0 0 0>
      }
      tintColor {colorUseRule=UseColor_Specified, specifiedColor=<rgba 0 0.0802 0.3515 1>}
      uVRegion {bIsValid=false, max=<0 0>, min=<0 0>}
     }
    }}
   slot {class=HorizontalBoxSlot}
  }
  Spacer_4 {
   widgetClassPath=&Spacer
   parent=AdvancedControlsBox
   widget=Spacer_4
   slot {class=HorizontalBoxSlot, properties {padding=<ltrb 0 0 10 0>}}
  }
  TextBlock_1 {
   widgetClassPath=&TextBlock
   parent=AdvancedControlsBox
   widget=TextBlock_1
   properties {
    text=Show Advanced Controls
    font {
     bForceMonospaced=false
     bMaterialIsStencil=false
     fontMaterial=null
     fontObject=&Roboto
     letterSpacing=0
     monospacedWidth=1
     size=14
     skewAmount=0
     typefaceFontName=Bold
     outlineSettings {
      bApplyOutlineToDropShadows=false
      bMiteredCorners=false
      bSeparateFillAlpha=false
      outlineMaterial=null
      outlineSize=0
      outlineColor=<rgba 0 0 0 1>
     }
    }
   }
   slot {class=HorizontalBoxSlot, properties {verticalAlignment=VAlign_Center}}
  }
  HorizontalBox_1 {
   widgetClassPath=&HorizontalBox
   parent=VerticalBox_5
   widget=HorizontalBox_1
   slot {class=VerticalBoxSlot, properties {size {sizeRule=Fill, value=1}}}
  }
  VerticalBox_2 {widgetClassPath=&VerticalBox, parent=HorizontalBox_1, widget=VerticalBox_2, slot {class=HorizontalBoxSlot}}
  VerticalBox_0 {widgetClassPath=&VerticalBox, parent=VerticalBox_2, widget=VerticalBox_0, slot {class=VerticalBoxSlot}}
  Render {
   widgetClassPath=&Button
   parent=VerticalBox_0
   widget=Render
   properties {widgetStyle {
     clickedSlateSound {resourceObject=null}
     disabled {
      bIsDynamicallyLoaded=false
      drawAs=RoundedBox
      imageType=NoImage
      mirroring=NoMirror
      resourceName=None
      resourceObject=null
      tiling=NoTile
      imageSize=<0 0>
      margin=<ltrb 0 0 0 0>
      outlineSettings {
       bUseBrushTransparency=false
       roundingType=FixedRadius
       width=1
       color {colorUseRule=UseColor_ColorTable, specifiedColor=<rgba 0 0 0 0>}
       cornerRadii=<4 4 4 4>
      }
      tintColor {colorUseRule=UseColor_Specified, specifiedColor=<rgba 0.0395 0.0395 0.0395 1>}
      uVRegion {bIsValid=false, max=<0 0>, min=<0 0>}
     }
     disabledForeground {colorUseRule=UseColor_ColorTable, specifiedColor=<rgba 0 0 0 0>}
     hovered {
      bIsDynamicallyLoaded=false
      drawAs=RoundedBox
      imageType=NoImage
      mirroring=NoMirror
      resourceName=None
      resourceObject=null
      tiling=NoTile
      imageSize=<32 32>
      margin=<ltrb 0 0 0 0>
      outlineSettings {
       bUseBrushTransparency=false
       roundingType=FixedRadius
       width=1
       color {colorUseRule=UseColor_ColorTable, specifiedColor=<rgba 0 0 0 0>}
       cornerRadii=<4 4 4 4>
      }
      tintColor {colorUseRule=UseColor_Specified, specifiedColor=<rgba 0.0953 0.0953 0.0953 1>}
      uVRegion {bIsValid=false, max=<0 0>, min=<0 0>}
     }
     hoveredForeground {colorUseRule=UseColor_ColorTable, specifiedColor=<rgba 0 0 0 0>}
     hoveredSlateSound {resourceObject=null}
     normal {
      bIsDynamicallyLoaded=false
      drawAs=RoundedBox
      imageType=NoImage
      mirroring=NoMirror
      resourceName=None
      resourceObject=null
      tiling=NoTile
      imageSize=<32 32>
      margin=<ltrb 0 0 0 0>
      outlineSettings {
       bUseBrushTransparency=false
       roundingType=FixedRadius
       width=1
       color {colorUseRule=UseColor_ColorTable, specifiedColor=<rgba 0 0 0 0>}
       cornerRadii=<4 4 4 4>
      }
      tintColor {colorUseRule=UseColor_Specified, specifiedColor=<rgba 0.0395 0.0395 0.0395 1>}
      uVRegion {bIsValid=false, max=<0 0>, min=<0 0>}
     }
     normalForeground {colorUseRule=UseColor_ColorTable, specifiedColor=<rgba 0 0 0 0>}
     normalPadding=<ltrb 12 1.5 12 1.5>
     pressed {
      bIsDynamicallyLoaded=false
      drawAs=RoundedBox
      imageType=NoImage
      mirroring=NoMirror
      resourceName=None
      resourceObject=null
      tiling=NoTile
      imageSize=<32 32>
      margin=<ltrb 0 0 0 0>
      outlineSettings {
       bUseBrushTransparency=false
       roundingType=FixedRadius
       width=1
       color {colorUseRule=UseColor_ColorTable, specifiedColor=<rgba 0 0 0 0>}
       cornerRadii=<4 4 4 4>
      }
      tintColor {colorUseRule=UseColor_Specified, specifiedColor=<rgba 0.0284 0.0284 0.0284 1>}
      uVRegion {bIsValid=false, max=<0 0>, min=<0 0>}
     }
     pressedForeground {colorUseRule=UseColor_ColorTable, specifiedColor=<rgba 0 0 0 0>}
     pressedPadding=<ltrb 12 2.5 12 0.5>
     pressedSlateSound {resourceObject=null}
    }}
   slot {class=VerticalBoxSlot}
  }
  TextBlock_98 {widgetClassPath=&TextBlock, parent=Render, widget=TextBlock_98, properties {text=Render Sequence}, slot {class=ButtonSlot}}
  OutputFolder_DV {
   widgetClassPath=&DetailsView
   parent=VerticalBox_0
   widget=OutputFolder_DV
   properties {categoriesToShow [Nocategory]}
   slot {class=VerticalBoxSlot}
  }
  VerticalBox_1 {
   widgetClassPath=&VerticalBox
   parent=VerticalBox_2
   widget=VerticalBox_1
   slot {class=VerticalBoxSlot, properties {horizontalAlignment=HAlign_Left}}
  }
  Spacer_110 {widgetClassPath=&Spacer, parent=VerticalBox_1, widget=Spacer_110, properties {size=<1 20>}, slot {class=VerticalBoxSlot}}
  RotationSelector {
   widgetClassPath=&DetailsView
   parent=VerticalBox_1
   widget=RotationSelector
   properties {categoriesToShow [Actor]}
   slot {class=VerticalBoxSlot}
  }
  Spacer_287 {
   widgetClassPath=&Spacer
   parent=VerticalBox_2
   widget=Spacer_287
   slot {class=VerticalBoxSlot, properties {padding=<ltrb 2 2 2 2>}}
  }
  Rotation_Slider {
   widgetClassPath=/Script/UMG.Slider
   bIsVariable=true
   parent=VerticalBox_2
   widget=Rotation_Slider
   properties {
    maxValue=180
    minValue=-180
    toolTipText=Set ACTOR REFERENCE to enable Rotation Slider
    widgetStyle {
     barThickness=4
     disabledBarImage {
      bIsDynamicallyLoaded=false
      drawAs=RoundedBox
      imageType=NoImage
      mirroring=NoMirror
      resourceName=None
      resourceObject=null
      tiling=NoTile
      imageSize=<32 32>
      margin=<ltrb 0 0 0 0>
      outlineSettings {
       bUseBrushTransparency=false
       roundingType=FixedRadius
       width=1
       color {colorUseRule=UseColor_ColorTable, specifiedColor=<rgba 0 0 0 0>}
       cornerRadii=<2 2 2 2>
      }
      tintColor {colorUseRule=UseColor_Specified, specifiedColor=<rgba 1 1 1 1>}
      uVRegion {bIsValid=false, max=<0 0>, min=<0 0>}
     }
     disabledThumbImage {
      bIsDynamicallyLoaded=false
      drawAs=RoundedBox
      imageType=NoImage
      mirroring=NoMirror
      resourceName=None
      resourceObject=null
      tiling=NoTile
      imageSize=<32 32>
      margin=<ltrb 0 0 0 0>
      outlineSettings {
       bUseBrushTransparency=false
       roundingType=HalfHeightRadius
       width=0
       color {colorUseRule=UseColor_Specified, specifiedColor=<rgba 0 0 0 0>}
       cornerRadii=<0 0 0 0>
      }
      tintColor {colorUseRule=UseColor_Specified, specifiedColor=<rgba 0.5271 0.5271 0.5271 1>}
      uVRegion {bIsValid=false, max=<0 0>, min=<0 0>}
     }
     hoveredBarImage {
      bIsDynamicallyLoaded=false
      drawAs=RoundedBox
      imageType=NoImage
      mirroring=NoMirror
      resourceName=None
      resourceObject=null
      tiling=NoTile
      imageSize=<0 0>
      margin=<ltrb 0 0 0 0>
      outlineSettings {
       bUseBrushTransparency=false
       roundingType=FixedRadius
       width=1
       color {colorUseRule=UseColor_ColorTable, specifiedColor=<rgba 0 0 0 0>}
       cornerRadii=<2 2 2 2>
      }
      tintColor {colorUseRule=UseColor_Specified, specifiedColor=<rgba 0.0048 0.0048 0.0048 1>}
      uVRegion {bIsValid=false, max=<0 0>, min=<0 0>}
     }
     hoveredThumbImage {
      bIsDynamicallyLoaded=false
      drawAs=RoundedBox
      imageType=NoImage
      mirroring=NoMirror
      resourceName=None
      resourceObject=null
      tiling=NoTile
      imageSize=<8 8>
      margin=<ltrb 0 0 0 0>
      outlineSettings {
       bUseBrushTransparency=false
       roundingType=HalfHeightRadius
       width=0
       color {colorUseRule=UseColor_Specified, specifiedColor=<rgba 0 0 0 0>}
       cornerRadii=<0 0 0 0>
      }
      tintColor {colorUseRule=UseColor_Specified, specifiedColor=<rgba 1 1 1 1>}
      uVRegion {bIsValid=false, max=<0 0>, min=<0 0>}
     }
     normalBarImage {
      bIsDynamicallyLoaded=false
      drawAs=RoundedBox
      imageType=NoImage
      mirroring=NoMirror
      resourceName=None
      resourceObject=null
      tiling=NoTile
      imageSize=<0 0>
      margin=<ltrb 0 0 0 0>
      outlineSettings {
       bUseBrushTransparency=false
       roundingType=FixedRadius
       width=1
       color {colorUseRule=UseColor_ColorTable, specifiedColor=<rgba 0 0 0 0>}
       cornerRadii=<2 2 2 2>
      }
      tintColor {colorUseRule=UseColor_Specified, specifiedColor=<rgba 0.0048 0.0048 0.0048 1>}
      uVRegion {bIsValid=false, max=<0 0>, min=<0 0>}
     }
     normalThumbImage {
      bIsDynamicallyLoaded=false
      drawAs=RoundedBox
      imageType=NoImage
      mirroring=NoMirror
      resourceName=None
      resourceObject=null
      tiling=NoTile
      imageSize=<8 8>
      margin=<ltrb 0 0 0 0>
      outlineSettings {
       bUseBrushTransparency=false
       roundingType=HalfHeightRadius
       width=0
       color {colorUseRule=UseColor_Specified, specifiedColor=<rgba 0 0 0 0>}
       cornerRadii=<0 0 0 0>
      }
      tintColor {colorUseRule=UseColor_Specified, specifiedColor=<rgba 0.5271 0.5271 0.5271 1>}
      uVRegion {bIsValid=false, max=<0 0>, min=<0 0>}
     }
    }
   }
   slot {class=VerticalBoxSlot}
  }
  AdvancedBox {
   widgetClassPath=&HorizontalBox
   bIsVariable=true
   parent=HorizontalBox_1
   widget=AdvancedBox
   properties {visibility=Collapsed}
   slot {class=HorizontalBoxSlot, properties {verticalAlignment=VAlign_Top, size {sizeRule=Fill, value=1}}}
  }
  VerticalBox_4 {
   widgetClassPath=&VerticalBox
   parent=AdvancedBox
   widget=VerticalBox_4
   slot {class=HorizontalBoxSlot, properties {size {sizeRule=Fill, value=1}}}
  }
  HorizontalBox_4 {
   widgetClassPath=&HorizontalBox
   parent=VerticalBox_4
   widget=HorizontalBox_4
   slot {class=VerticalBoxSlot, properties {size {sizeRule=Fill, value=1}}}
  }
  Border_0 {
   widgetClassPath=&Border
   parent=HorizontalBox_4
   widget=Border_0
   properties {horizontalAlignment=HAlign_Left, verticalAlignment=VAlign_Top, brushColor=<rgba 0.151 0.151 0.151 0.453>}
   slot {class=HorizontalBoxSlot}
  }
  VerticalBox_519 {
   widgetClassPath=&VerticalBox
   parent=Border_0
   widget=VerticalBox_519
   slot {class=BorderSlot, properties {horizontalAlignment=HAlign_Left, verticalAlignment=VAlign_Top}}
  }
  QualityBox {widgetClassPath=&HorizontalBox, parent=VerticalBox_519, widget=QualityBox, slot {class=VerticalBoxSlot}}
  Spacer_1 {widgetClassPath=&Spacer, parent=QualityBox, widget=Spacer_1, slot {class=HorizontalBoxSlot, properties {padding=<ltrb 0 0 40 0>}}}
  ForceQuality_Checkbox {
   widgetClassPath=&CheckBox
   bIsVariable=true
   parent=QualityBox
   widget=ForceQuality_Checkbox
   properties {
    checkedState=Checked
    widgetStyle {
     checkBoxType=CheckBox
     backgroundHoveredImage {
      bIsDynamicallyLoaded=false
      drawAs=RoundedBox
      imageType=NoImage
      mirroring=NoMirror
      resourceName=None
      resourceObject=null
      tiling=NoTile
      imageSize=<18 18>
      margin=<ltrb 0 0 0 0>
      outlineSettings {
       bUseBrushTransparency=false
       roundingType=FixedRadius
       width=1
       color {colorUseRule=UseColor_ColorTable, specifiedColor=<rgba 0 0 0 0>}
       cornerRadii=<3 3 3 3>
      }
      tintColor {colorUseRule=UseColor_Specified, specifiedColor=<rgba 0.0048 0.0048 0.0048 1>}
      uVRegion {bIsValid=false, max=<0 0>, min=<0 0>}
     }
     backgroundImage {
      bIsDynamicallyLoaded=false
      drawAs=RoundedBox
      imageType=NoImage
      mirroring=NoMirror
      resourceName=None
      resourceObject=null
      tiling=NoTile
      imageSize=<18 18>
      margin=<ltrb 0 0 0 0>
      outlineSettings {
       bUseBrushTransparency=false
       roundingType=FixedRadius
       width=1
       color {colorUseRule=UseColor_ColorTable, specifiedColor=<rgba 0 0 0 0>}
       cornerRadii=<3 3 3 3>
      }
      tintColor {colorUseRule=UseColor_Specified, specifiedColor=<rgba 0.0048 0.0048 0.0048 1>}
      uVRegion {bIsValid=false, max=<0 0>, min=<0 0>}
     }
     backgroundPressedImage {
      bIsDynamicallyLoaded=false
      drawAs=RoundedBox
      imageType=NoImage
      mirroring=NoMirror
      resourceName=None
      resourceObject=null
      tiling=NoTile
      imageSize=<18 18>
      margin=<ltrb 0 0 0 0>
      outlineSettings {
       bUseBrushTransparency=false
       roundingType=FixedRadius
       width=1
       color {colorUseRule=UseColor_ColorTable, specifiedColor=<rgba 0 0 0 0>}
       cornerRadii=<3 3 3 3>
      }
      tintColor {colorUseRule=UseColor_Specified, specifiedColor=<rgba 0.0048 0.0048 0.0048 1>}
      uVRegion {bIsValid=false, max=<0 0>, min=<0 0>}
     }
     borderBackgroundColor {colorUseRule=UseColor_Specified, specifiedColor=<rgba 1 1 1 1>}
     checkedForeground {colorUseRule=UseColor_Specified, specifiedColor=<rgba 1 1 1 1>}
     checkedHoveredForeground {colorUseRule=UseColor_Specified, specifiedColor=<rgba 1 1 1 1>}
     checkedHoveredImage {
      bIsDynamicallyLoaded=false
      drawAs=Image
      imageType=Vector
      mirroring=NoMirror
      resourceName=../../../Engine/Content/Slate/Starship/CoreWidgets/CheckBox/check.svg
      resourceObject=null
      tiling=NoTile
      imageSize=<16 16>
      margin=<ltrb 0 0 0 0>
      outlineSettings {
       bUseBrushTransparency=false
       roundingType=HalfHeightRadius
       width=0
       color {colorUseRule=UseColor_Specified, specifiedColor=<rgba 0 0 0 0>}
       cornerRadii=<0 0 0 0>
      }
      tintColor {colorUseRule=UseColor_Specified, specifiedColor=<rgba 0.0044 0.2384 1 1>}
      uVRegion {bIsValid=false, max=<0 0>, min=<0 0>}
     }
     checkedImage {
      bIsDynamicallyLoaded=false
      drawAs=Image
      imageType=Vector
      mirroring=NoMirror
      resourceName=../../../Engine/Content/Slate/Starship/CoreWidgets/CheckBox/check.svg
      resourceObject=null
      tiling=NoTile
      imageSize=<16 16>
      margin=<ltrb 0 0 0 0>
      outlineSettings {
       bUseBrushTransparency=false
       roundingType=HalfHeightRadius
       width=0
       color {colorUseRule=UseColor_Specified, specifiedColor=<rgba 0 0 0 0>}
       cornerRadii=<0 0 0 0>
      }
      tintColor {colorUseRule=UseColor_Specified, specifiedColor=<rgba 0 0.162 0.7454 1>}
      uVRegion {bIsValid=false, max=<0 0>, min=<0 0>}
     }
     checkedPressedForeground {colorUseRule=UseColor_Specified, specifiedColor=<rgba 1 1 1 1>}
     checkedPressedImage {
      bIsDynamicallyLoaded=false
      drawAs=Image
      imageType=Vector
      mirroring=NoMirror
      resourceName=../../../Engine/Content/Slate/Starship/CoreWidgets/CheckBox/check.svg
      resourceObject=null
      tiling=NoTile
      imageSize=<16 16>
      margin=<ltrb 0 0 0 0>
      outlineSettings {
       bUseBrushTransparency=false
       roundingType=HalfHeightRadius
       width=0
       color {colorUseRule=UseColor_Specified, specifiedColor=<rgba 0 0 0 0>}
       cornerRadii=<0 0 0 0>
      }
      tintColor {colorUseRule=UseColor_Specified, specifiedColor=<rgba 0 0.0802 0.3515 1>}
      uVRegion {bIsValid=false, max=<0 0>, min=<0 0>}
     }
     checkedSlateSound {resourceObject=null}
     foregroundColor {colorUseRule=UseColor_Specified, specifiedColor=<rgba 1 1 1 1>}
     hoveredForeground {colorUseRule=UseColor_Specified, specifiedColor=<rgba 1 1 1 1>}
     hoveredSlateSound {resourceObject=null}
     padding=<ltrb 2 0 0 0>
     pressedForeground {colorUseRule=UseColor_Specified, specifiedColor=<rgba 1 1 1 1>}
     uncheckedHoveredImage {
      bIsDynamicallyLoaded=false
      drawAs=NoDrawType
      imageType=NoImage
      mirroring=NoMirror
      resourceName=None
      resourceObject=null
      tiling=NoTile
      imageSize=<0 0>
      margin=<ltrb 0 0 0 0>
      outlineSettings {
       bUseBrushTransparency=false
       roundingType=HalfHeightRadius
       width=0
       color {colorUseRule=UseColor_Specified, specifiedColor=<rgba 0 0 0 0>}
       cornerRadii=<0 0 0 0>
      }
      tintColor {colorUseRule=UseColor_Specified, specifiedColor=<rgba 1 1 1 1>}
      uVRegion {bIsValid=false, max=<0 0>, min=<0 0>}
     }
     uncheckedImage {
      bIsDynamicallyLoaded=false
      drawAs=NoDrawType
      imageType=NoImage
      mirroring=NoMirror
      resourceName=None
      resourceObject=null
      tiling=NoTile
      imageSize=<0 0>
      margin=<ltrb 0 0 0 0>
      outlineSettings {
       bUseBrushTransparency=false
       roundingType=HalfHeightRadius
       width=0
       color {colorUseRule=UseColor_Specified, specifiedColor=<rgba 0 0 0 0>}
       cornerRadii=<0 0 0 0>
      }
      tintColor {colorUseRule=UseColor_Specified, specifiedColor=<rgba 1 1 1 1>}
      uVRegion {bIsValid=false, max=<0 0>, min=<0 0>}
     }
     uncheckedPressedImage {
      bIsDynamicallyLoaded=false
      drawAs=NoDrawType
      imageType=NoImage
      mirroring=NoMirror
      resourceName=None
      resourceObject=null
      tiling=NoTile
      imageSize=<0 0>
      margin=<ltrb 0 0 0 0>
      outlineSettings {
       bUseBrushTransparency=false
       roundingType=HalfHeightRadius
       width=0
       color {colorUseRule=UseColor_Specified, specifiedColor=<rgba 0 0 0 0>}
       cornerRadii=<0 0 0 0>
      }
      tintColor {colorUseRule=UseColor_Specified, specifiedColor=<rgba 1 1 1 1>}
      uVRegion {bIsValid=false, max=<0 0>, min=<0 0>}
     }
     uncheckedSlateSound {resourceObject=null}
     undeterminedForeground {colorUseRule=UseColor_Specified, specifiedColor=<rgba 1 1 1 1>}
     undeterminedHoveredImage {
      bIsDynamicallyLoaded=false
      drawAs=Image
      imageType=Vector
      mirroring=NoMirror
      resourceName=../../../Engine/Content/Slate/Starship/CoreWidgets/CheckBox/indeterminate.svg
      resourceObject=null
      tiling=NoTile
      imageSize=<16 16>
      margin=<ltrb 0 0 0 0>
      outlineSettings {
       bUseBrushTransparency=false
       roundingType=HalfHeightRadius
       width=0
       color {colorUseRule=UseColor_Specified, specifiedColor=<rgba 0 0 0 0>}
       cornerRadii=<0 0 0 0>
      }
      tintColor {colorUseRule=UseColor_Specified, specifiedColor=<rgba 0.0044 0.2384 1 1>}
      uVRegion {bIsValid=false, max=<0 0>, min=<0 0>}
     }
     undeterminedImage {
      bIsDynamicallyLoaded=false
      drawAs=Image
      imageType=Vector
      mirroring=NoMirror
      resourceName=../../../Engine/Content/Slate/Starship/CoreWidgets/CheckBox/indeterminate.svg
      resourceObject=null
      tiling=NoTile
      imageSize=<16 16>
      margin=<ltrb 0 0 0 0>
      outlineSettings {
       bUseBrushTransparency=false
       roundingType=HalfHeightRadius
       width=0
       color {colorUseRule=UseColor_Specified, specifiedColor=<rgba 0 0 0 0>}
       cornerRadii=<0 0 0 0>
      }
      tintColor {colorUseRule=UseColor_Specified, specifiedColor=<rgba 0 0.162 0.7454 1>}
      uVRegion {bIsValid=false, max=<0 0>, min=<0 0>}
     }
     undeterminedPressedImage {
      bIsDynamicallyLoaded=false
      drawAs=Image
      imageType=Vector
      mirroring=NoMirror
      resourceName=../../../Engine/Content/Slate/Starship/CoreWidgets/CheckBox/indeterminate.svg
      resourceObject=null
      tiling=NoTile
      imageSize=<16 16>
      margin=<ltrb 0 0 0 0>
      outlineSettings {
       bUseBrushTransparency=false
       roundingType=HalfHeightRadius
       width=0
       color {colorUseRule=UseColor_Specified, specifiedColor=<rgba 0 0 0 0>}
       cornerRadii=<0 0 0 0>
      }
      tintColor {colorUseRule=UseColor_Specified, specifiedColor=<rgba 0 0.0802 0.3515 1>}
      uVRegion {bIsValid=false, max=<0 0>, min=<0 0>}
     }
    }
   }
   slot {class=HorizontalBoxSlot}
  }
  Spacer {widgetClassPath=&Spacer, parent=QualityBox, widget=Spacer, slot {class=HorizontalBoxSlot, properties {padding=<ltrb 0 0 10 0>}}}
  TextBlock {
   widgetClassPath=&TextBlock
   parent=QualityBox
   widget=TextBlock
   properties {
    text=Force Quality Settings
    font {
     bForceMonospaced=false
     bMaterialIsStencil=false
     fontMaterial=null
     fontObject=&Roboto
     letterSpacing=0
     monospacedWidth=1
     size=14
     skewAmount=0
     typefaceFontName=Bold
     outlineSettings {
      bApplyOutlineToDropShadows=false
      bMiteredCorners=false
      bSeparateFillAlpha=false
      outlineMaterial=null
      outlineSize=0
      outlineColor=<rgba 0 0 0 1>
     }
    }
   }
   slot {class=HorizontalBoxSlot, properties {verticalAlignment=VAlign_Bottom}}
  }
  Spacer_2 {widgetClassPath=&Spacer, parent=QualityBox, widget=Spacer_2, slot {class=HorizontalBoxSlot, properties {padding=<ltrb 0 0 10 0>}}}
  ForceQualityMode {
   widgetClassPath=&ComboBoxString
   bIsVariable=true
   parent=QualityBox
   widget=ForceQualityMode
   properties {
    selectedOption=Epic
    defaultOptions [Epic,Low]
    foregroundColor {colorUseRule=UseColor_Specified, specifiedColor=<rgba 1 1 1 1>}
    itemStyle {
     bUseParentRowBrush=false
     activeBrush {
      bIsDynamicallyLoaded=false
      drawAs=Image
      imageType=NoImage
      mirroring=NoMirror
      resourceName=None
      resourceObject=null
      tiling=NoTile
      imageSize=<0 0>
      margin=<ltrb 0 0 0 0>
      outlineSettings {
       bUseBrushTransparency=false
       roundingType=HalfHeightRadius
       width=0
       color {colorUseRule=UseColor_Specified, specifiedColor=<rgba 0 0 0 0>}
       cornerRadii=<0 0 0 0>
      }
      tintColor {colorUseRule=UseColor_Specified, specifiedColor=<rgba 0 0.162 0.7454 1>}
      uVRegion {bIsValid=false, max=<0 0>, min=<0 0>}
     }
     activeHighlightedBrush {
      bIsDynamicallyLoaded=false
      drawAs=Image
      imageType=NoImage
      mirroring=NoMirror
      resourceName=None
      resourceObject=null
      tiling=NoTile
      imageSize=<0 0>
      margin=<ltrb 0 0 0 0>
      outlineSettings {
       bUseBrushTransparency=false
       roundingType=HalfHeightRadius
       width=0
       color {colorUseRule=UseColor_Specified, specifiedColor=<rgba 0 0 0 0>}
       cornerRadii=<0 0 0 0>
      }
      tintColor {colorUseRule=UseColor_Specified, specifiedColor=<rgba 0.0252 0.0319 0.0423 1>}
      uVRegion {bIsValid=false, max=<0 0>, min=<0 0>}
     }
     activeHoveredBrush {
      bIsDynamicallyLoaded=false
      drawAs=Image
      imageType=NoImage
      mirroring=NoMirror
      resourceName=None
      resourceObject=null
      tiling=NoTile
      imageSize=<0 0>
      margin=<ltrb 0 0 0 0>
      outlineSettings {
       bUseBrushTransparency=false
       roundingType=HalfHeightRadius
       width=0
       color {colorUseRule=UseColor_Specified, specifiedColor=<rgba 0 0 0 0>}
       cornerRadii=<0 0 0 0>
      }
      tintColor {colorUseRule=UseColor_Specified, specifiedColor=<rgba 0 0.162 0.7454 1>}
      uVRegion {bIsValid=false, max=<0 0>, min=<0 0>}
     }
     dropIndicator_Above {
      bIsDynamicallyLoaded=false
      drawAs=Box
      imageType=FullColor
      mirroring=NoMirror
      resourceName=../../../Engine/Content/Slate/Common/DropZoneIndicator_Above.png
      resourceObject=null
      tiling=NoTile
      imageSize=<0 0>
      margin=<ltrb 0.625 0.625 0 0>
      outlineSettings {
       bUseBrushTransparency=false
       roundingType=HalfHeightRadius
       width=0
       color {colorUseRule=UseColor_Specified, specifiedColor=<rgba 0 0 0 0>}
       cornerRadii=<0 0 0 0>
      }
      tintColor {colorUseRule=UseColor_Specified, specifiedColor=<rgba 0.0091 0.3515 0.9647 1>}
      uVRegion {bIsValid=false, max=<0 0>, min=<0 0>}
     }
     dropIndicator_Below {
      bIsDynamicallyLoaded=false
      drawAs=Box
      imageType=FullColor
      mirroring=NoMirror
      resourceName=../../../Engine/Content/Slate/Common/DropZoneIndicator_Below.png
      resourceObject=null
      tiling=NoTile
      imageSize=<0 0>
      margin=<ltrb 0.625 0 0 0.625>
      outlineSettings {
       bUseBrushTransparency=false
       roundingType=HalfHeightRadius
       width=0
       color {colorUseRule=UseColor_Specified, specifiedColor=<rgba 0 0 0 0>}
       cornerRadii=<0 0 0 0>
      }
      tintColor {colorUseRule=UseColor_Specified, specifiedColor=<rgba 0.0091 0.3515 0.9647 1>}
      uVRegion {bIsValid=false, max=<0 0>, min=<0 0>}
     }
     dropIndicator_Onto {
      bIsDynamicallyLoaded=false
      drawAs=Box
      imageType=FullColor
      mirroring=NoMirror
      resourceName=../../../Engine/Content/Slate/Common/DropZoneIndicator_Onto.png
      resourceObject=null
      tiling=NoTile
      imageSize=<0 0>
      margin=<ltrb 0.25 0.25 0.25 0.25>
      outlineSettings {
       bUseBrushTransparency=false
       roundingType=HalfHeightRadius
       width=0
       color {colorUseRule=UseColor_Specified, specifiedColor=<rgba 0 0 0 0>}
       cornerRadii=<0 0 0 0>
      }
      tintColor {colorUseRule=UseColor_Specified, specifiedColor=<rgba 0.0091 0.3515 0.9647 1>}
      uVRegion {bIsValid=false, max=<0 0>, min=<0 0>}
     }
     evenRowBackgroundBrush {
      bIsDynamicallyLoaded=false
      drawAs=Image
      imageType=NoImage
      mirroring=NoMirror
      resourceName=None
      resourceObject=null
      tiling=NoTile
      imageSize=<0 0>
      margin=<ltrb 0 0 0 0>
      outlineSettings {
       bUseBrushTransparency=false
       roundingType=HalfHeightRadius
       width=0
       color {colorUseRule=UseColor_Specified, specifiedColor=<rgba 0 0 0 0>}
       cornerRadii=<0 0 0 0>
      }
      tintColor {colorUseRule=UseColor_Specified, specifiedColor=<rgba 0.0103 0.0103 0.0103 1>}
      uVRegion {bIsValid=false, max=<0 0>, min=<0 0>}
     }
     evenRowBackgroundHoveredBrush {
      bIsDynamicallyLoaded=false
      drawAs=Image
      imageType=NoImage
      mirroring=NoMirror
      resourceName=None
      resourceObject=null
      tiling=NoTile
      imageSize=<0 0>
      margin=<ltrb 0 0 0 0>
      outlineSettings {
       bUseBrushTransparency=false
       roundingType=HalfHeightRadius
       width=0
       color {colorUseRule=UseColor_Specified, specifiedColor=<rgba 0 0 0 0>}
       cornerRadii=<0 0 0 0>
      }
      tintColor {colorUseRule=UseColor_Specified, specifiedColor=<rgba 0.0176 0.0176 0.0176 1>}
      uVRegion {bIsValid=false, max=<0 0>, min=<0 0>}
     }
     inactiveBrush {
      bIsDynamicallyLoaded=false
      drawAs=Image
      imageType=NoImage
      mirroring=NoMirror
      resourceName=None
      resourceObject=null
      tiling=NoTile
      imageSize=<0 0>
      margin=<ltrb 0 0 0 0>
      outlineSettings {
       bUseBrushTransparency=false
       roundingType=HalfHeightRadius
       width=0
       color {colorUseRule=UseColor_Specified, specifiedColor=<rgba 0 0 0 0>}
       cornerRadii=<0 0 0 0>
      }
      tintColor {colorUseRule=UseColor_Specified, specifiedColor=<rgba 0.0513 0.0953 0.159 1>}
      uVRegion {bIsValid=false, max=<0 0>, min=<0 0>}
     }
     inactiveHighlightedBrush {
      bIsDynamicallyLoaded=false
      drawAs=Image
      imageType=NoImage
      mirroring=NoMirror
      resourceName=None
      resourceObject=null
      tiling=NoTile
      imageSize=<0 0>
      margin=<ltrb 0 0 0 0>
      outlineSettings {
       bUseBrushTransparency=false
       roundingType=HalfHeightRadius
       width=0
       color {colorUseRule=UseColor_Specified, specifiedColor=<rgba 0 0 0 0>}
       cornerRadii=<0 0 0 0>
      }
      tintColor {colorUseRule=UseColor_Specified, specifiedColor=<rgba 0.0252 0.0319 0.0423 1>}
      uVRegion {bIsValid=false, max=<0 0>, min=<0 0>}
     }
     inactiveHoveredBrush {
      bIsDynamicallyLoaded=false
      drawAs=Image
      imageType=NoImage
      mirroring=NoMirror
      resourceName=None
      resourceObject=null
      tiling=NoTile
      imageSize=<0 0>
      margin=<ltrb 0 0 0 0>
      outlineSettings {
       bUseBrushTransparency=false
       roundingType=HalfHeightRadius
       width=0
       color {colorUseRule=UseColor_Specified, specifiedColor=<rgba 0 0 0 0>}
       cornerRadii=<0 0 0 0>
      }
      tintColor {colorUseRule=UseColor_Specified, specifiedColor=<rgba 0.0513 0.0953 0.159 1>}
      uVRegion {bIsValid=false, max=<0 0>, min=<0 0>}
     }
     oddRowBackgroundBrush {
      bIsDynamicallyLoaded=false
      drawAs=Image
      imageType=NoImage
      mirroring=NoMirror
      resourceName=None
      resourceObject=null
      tiling=NoTile
      imageSize=<0 0>
      margin=<ltrb 0 0 0 0>
      outlineSettings {
       bUseBrushTransparency=false
       roundingType=HalfHeightRadius
       width=0
       color {colorUseRule=UseColor_Specified, specifiedColor=<rgba 0 0 0 0>}
       cornerRadii=<0 0 0 0>
      }
      tintColor {colorUseRule=UseColor_Specified, specifiedColor=<rgba 0.0103 0.0103 0.0103 1>}
      uVRegion {bIsValid=false, max=<0 0>, min=<0 0>}
     }
     oddRowBackgroundHoveredBrush {
      bIsDynamicallyLoaded=false
      drawAs=Image
      imageType=NoImage
      mirroring=NoMirror
      resourceName=None
      resourceObject=null
      tiling=NoTile
      imageSize=<0 0>
      margin=<ltrb 0 0 0 0>
      outlineSettings {
       bUseBrushTransparency=false
       roundingType=HalfHeightRadius
       width=0
       color {colorUseRule=UseColor_Specified, specifiedColor=<rgba 0 0 0 0>}
       cornerRadii=<0 0 0 0>
      }
      tintColor {colorUseRule=UseColor_Specified, specifiedColor=<rgba 0.0176 0.0176 0.0176 1>}
      uVRegion {bIsValid=false, max=<0 0>, min=<0 0>}
     }
     parentRowBackgroundBrush {
      bIsDynamicallyLoaded=false
      drawAs=Image
      imageType=NoImage
      mirroring=NoMirror
      resourceName=None
      resourceObject=null
      tiling=NoTile
      imageSize=<32 32>
      margin=<ltrb 0 0 0 0>
      outlineSettings {
       bUseBrushTransparency=false
       roundingType=HalfHeightRadius
       width=0
       color {colorUseRule=UseColor_Specified, specifiedColor=<rgba 0 0 0 0>}
       cornerRadii=<0 0 0 0>
      }
      tintColor {colorUseRule=UseColor_Specified, specifiedColor=<rgba 1 1 1 1>}
      uVRegion {bIsValid=false, max=<0 0>, min=<0 0>}
     }
     parentRowBackgroundHoveredBrush {
      bIsDynamicallyLoaded=false
      drawAs=Image
      imageType=NoImage
      mirroring=NoMirror
      resourceName=None
      resourceObject=null
      tiling=NoTile
      imageSize=<32 32>
      margin=<ltrb 0 0 0 0>
      outlineSettings {
       bUseBrushTransparency=false
       roundingType=HalfHeightRadius
       width=0
       color {colorUseRule=UseColor_Specified, specifiedColor=<rgba 0 0 0 0>}
       cornerRadii=<0 0 0 0>
      }
      tintColor {colorUseRule=UseColor_Specified, specifiedColor=<rgba 1 1 1 1>}
      uVRegion {bIsValid=false, max=<0 0>, min=<0 0>}
     }
     selectedTextColor {colorUseRule=UseColor_Specified, specifiedColor=<rgba 1 1 1 1>}
     selectorFocusedBrush {
      bIsDynamicallyLoaded=false
      drawAs=Border
      imageType=FullColor
      mirroring=NoMirror
      resourceName=../../../Engine/Content/Slate/Common/Selector.png
      resourceObject=null
      tiling=Both
      imageSize=<0 0>
      margin=<ltrb 0.25 0.25 0.25 0.25>
      outlineSettings {
       bUseBrushTransparency=false
       roundingType=HalfHeightRadius
       width=0
       color {colorUseRule=UseColor_Specified, specifiedColor=<rgba 0 0 0 0>}
       cornerRadii=<0 0 0 0>
      }
      tintColor {colorUseRule=UseColor_Specified, specifiedColor=<rgba 0.701 0.225 0.003 1>}
      uVRegion {bIsValid=false, max=<0 0>, min=<0 0>}
     }
     textColor {colorUseRule=UseColor_Specified, specifiedColor=<rgba 0.5271 0.5271 0.5271 1>}
    }
    scrollBarStyle {
     thickness=8
     draggedThumbImage {
      bIsDynamicallyLoaded=false
      drawAs=RoundedBox
      imageType=NoImage
      mirroring=NoMirror
      resourceName=None
      resourceObject=null
      tiling=NoTile
      imageSize=<0 0>
      margin=<ltrb 0 0 0 0>
      outlineSettings {
       bUseBrushTransparency=false
       roundingType=FixedRadius
       width=0
       color {colorUseRule=UseColor_Specified, specifiedColor=<rgba 0 0 0 0>}
       cornerRadii=<4 4 4 4>
      }
      tintColor {colorUseRule=UseColor_Specified, specifiedColor=<rgba 0.2159 0.2159 0.2159 1>}
      uVRegion {bIsValid=false, max=<0 0>, min=<0 0>}
     }
     horizontalBackgroundImage {
      bIsDynamicallyLoaded=false
      drawAs=NoDrawType
      imageType=NoImage
      mirroring=NoMirror
      resourceName=None
      resourceObject=null
      tiling=NoTile
      imageSize=<0 0>
      margin=<ltrb 0 0 0 0>
      outlineSettings {
       bUseBrushTransparency=false
       roundingType=HalfHeightRadius
       width=0
       color {colorUseRule=UseColor_Specified, specifiedColor=<rgba 0 0 0 0>}
       cornerRadii=<0 0 0 0>
      }
      tintColor {colorUseRule=UseColor_Specified, specifiedColor=<rgba 1 1 1 1>}
      uVRegion {bIsValid=false, max=<0 0>, min=<0 0>}
     }
     horizontalBottomSlotImage {
      bIsDynamicallyLoaded=false
      drawAs=NoDrawType
      imageType=NoImage
      mirroring=NoMirror
      resourceName=None
      resourceObject=null
      tiling=NoTile
      imageSize=<0 0>
      margin=<ltrb 0 0 0 0>
      outlineSettings {
       bUseBrushTransparency=false
       roundingType=HalfHeightRadius
       width=0
       color {colorUseRule=UseColor_Specified, specifiedColor=<rgba 0 0 0 0>}
       cornerRadii=<0 0 0 0>
      }
      tintColor {colorUseRule=UseColor_Specified, specifiedColor=<rgba 1 1 1 1>}
      uVRegion {bIsValid=false, max=<0 0>, min=<0 0>}
     }
     horizontalTopSlotImage {
      bIsDynamicallyLoaded=false
      drawAs=NoDrawType
      imageType=NoImage
      mirroring=NoMirror
      resourceName=None
      resourceObject=null
      tiling=NoTile
      imageSize=<0 0>
      margin=<ltrb 0 0 0 0>
      outlineSettings {
       bUseBrushTransparency=false
       roundingType=HalfHeightRadius
       width=0
       color {colorUseRule=UseColor_Specified, specifiedColor=<rgba 0 0 0 0>}
       cornerRadii=<0 0 0 0>
      }
      tintColor {colorUseRule=UseColor_Specified, specifiedColor=<rgba 1 1 1 1>}
      uVRegion {bIsValid=false, max=<0 0>, min=<0 0>}
     }
     hoveredThumbImage {
      bIsDynamicallyLoaded=false
      drawAs=RoundedBox
      imageType=NoImage
      mirroring=NoMirror
      resourceName=None
      resourceObject=null
      tiling=NoTile
      imageSize=<0 0>
      margin=<ltrb 0 0 0 0>
      outlineSettings {
       bUseBrushTransparency=false
       roundingType=FixedRadius
       width=0
       color {colorUseRule=UseColor_Specified, specifiedColor=<rgba 0 0 0 0>}
       cornerRadii=<4 4 4 4>
      }
      tintColor {colorUseRule=UseColor_Specified, specifiedColor=<rgba 0.2159 0.2159 0.2159 1>}
      uVRegion {bIsValid=false, max=<0 0>, min=<0 0>}
     }
     normalThumbImage {
      bIsDynamicallyLoaded=false
      drawAs=RoundedBox
      imageType=NoImage
      mirroring=NoMirror
      resourceName=None
      resourceObject=null
      tiling=NoTile
      imageSize=<0 0>
      margin=<ltrb 0 0 0 0>
      outlineSettings {
       bUseBrushTransparency=false
       roundingType=FixedRadius
       width=0
       color {colorUseRule=UseColor_Specified, specifiedColor=<rgba 0 0 0 0>}
       cornerRadii=<4 4 4 4>
      }
      tintColor {colorUseRule=UseColor_Specified, specifiedColor=<rgba 0.0953 0.0953 0.0953 1>}
      uVRegion {bIsValid=false, max=<0 0>, min=<0 0>}
     }
     verticalBackgroundImage {
      bIsDynamicallyLoaded=false
      drawAs=NoDrawType
      imageType=NoImage
      mirroring=NoMirror
      resourceName=None
      resourceObject=null
      tiling=NoTile
      imageSize=<0 0>
      margin=<ltrb 0 0 0 0>
      outlineSettings {
       bUseBrushTransparency=false
       roundingType=HalfHeightRadius
       width=0
       color {colorUseRule=UseColor_Specified, specifiedColor=<rgba 0 0 0 0>}
       cornerRadii=<0 0 0 0>
      }
      tintColor {colorUseRule=UseColor_Specified, specifiedColor=<rgba 1 1 1 1>}
      uVRegion {bIsValid=false, max=<0 0>, min=<0 0>}
     }
     verticalBottomSlotImage {
      bIsDynamicallyLoaded=false
      drawAs=NoDrawType
      imageType=NoImage
      mirroring=NoMirror
      resourceName=None
      resourceObject=null
      tiling=NoTile
      imageSize=<0 0>
      margin=<ltrb 0 0 0 0>
      outlineSettings {
       bUseBrushTransparency=false
       roundingType=HalfHeightRadius
       width=0
       color {colorUseRule=UseColor_Specified, specifiedColor=<rgba 0 0 0 0>}
       cornerRadii=<0 0 0 0>
      }
      tintColor {colorUseRule=UseColor_Specified, specifiedColor=<rgba 1 1 1 1>}
      uVRegion {bIsValid=false, max=<0 0>, min=<0 0>}
     }
     verticalTopSlotImage {
      bIsDynamicallyLoaded=false
      drawAs=NoDrawType
      imageType=NoImage
      mirroring=NoMirror
      resourceName=None
      resourceObject=null
      tiling=NoTile
      imageSize=<0 0>
      margin=<ltrb 0 0 0 0>
      outlineSettings {
       bUseBrushTransparency=false
       roundingType=HalfHeightRadius
       width=0
       color {colorUseRule=UseColor_Specified, specifiedColor=<rgba 0 0 0 0>}
       cornerRadii=<0 0 0 0>
      }
      tintColor {colorUseRule=UseColor_Specified, specifiedColor=<rgba 1 1 1 1>}
      uVRegion {bIsValid=false, max=<0 0>, min=<0 0>}
     }
    }
    widgetStyle {
     comboButtonStyle {
      downArrowAlign=VAlign_Center
      buttonStyle {
       clickedSlateSound {resourceObject=null}
       disabled {
        bIsDynamicallyLoaded=false
        drawAs=RoundedBox
        imageType=NoImage
        mirroring=NoMirror
        resourceName=None
        resourceObject=null
        tiling=NoTile
        imageSize=<0 0>
        margin=<ltrb 0 0 0 0>
        outlineSettings {
         bUseBrushTransparency=false
         roundingType=FixedRadius
         width=1
         color {colorUseRule=UseColor_ColorTable, specifiedColor=<rgba 0 0 0 0>}
         cornerRadii=<4 4 4 4>
        }
        tintColor {colorUseRule=UseColor_Specified, specifiedColor=<rgba 0.0395 0.0395 0.0395 1>}
        uVRegion {bIsValid=false, max=<0 0>, min=<0 0>}
       }
       disabledForeground {colorUseRule=UseColor_ColorTable, specifiedColor=<rgba 0 0 0 0>}
       hovered {
        bIsDynamicallyLoaded=false
        drawAs=RoundedBox
        imageType=NoImage
        mirroring=NoMirror
        resourceName=None
        resourceObject=null
        tiling=NoTile
        imageSize=<0 0>
        margin=<ltrb 0 0 0 0>
        outlineSettings {
         bUseBrushTransparency=false
         roundingType=FixedRadius
         width=1
         color {colorUseRule=UseColor_ColorTable, specifiedColor=<rgba 0 0 0 0>}
         cornerRadii=<4 4 4 4>
        }
        tintColor {colorUseRule=UseColor_Specified, specifiedColor=<rgba 0.0048 0.0048 0.0048 1>}
        uVRegion {bIsValid=false, max=<0 0>, min=<0 0>}
       }
       hoveredForeground {colorUseRule=UseColor_ColorTable, specifiedColor=<rgba 0 0 0 0>}
       hoveredSlateSound {resourceObject=null}
       normal {
        bIsDynamicallyLoaded=false
        drawAs=RoundedBox
        imageType=NoImage
        mirroring=NoMirror
        resourceName=None
        resourceObject=null
        tiling=NoTile
        imageSize=<0 0>
        margin=<ltrb 0 0 0 0>
        outlineSettings {
         bUseBrushTransparency=false
         roundingType=FixedRadius
         width=1
         color {colorUseRule=UseColor_ColorTable, specifiedColor=<rgba 0 0 0 0>}
         cornerRadii=<4 4 4 4>
        }
        tintColor {colorUseRule=UseColor_Specified, specifiedColor=<rgba 0.0048 0.0048 0.0048 1>}
        uVRegion {bIsValid=false, max=<0 0>, min=<0 0>}
       }
       normalForeground {colorUseRule=UseColor_ColorTable, specifiedColor=<rgba 0 0 0 0>}
       normalPadding=<ltrb 8 1 8 1>
       pressed {
        bIsDynamicallyLoaded=false
        drawAs=RoundedBox
        imageType=NoImage
        mirroring=NoMirror
        resourceName=None
        resourceObject=null
        tiling=NoTile
        imageSize=<0 0>
        margin=<ltrb 0 0 0 0>
        outlineSettings {
         bUseBrushTransparency=false
         roundingType=FixedRadius
         width=1
         color {colorUseRule=UseColor_ColorTable, specifiedColor=<rgba 0 0 0 0>}
         cornerRadii=<4 4 4 4>
        }
        tintColor {colorUseRule=UseColor_Specified, specifiedColor=<rgba 0.0048 0.0048 0.0048 1>}
        uVRegion {bIsValid=false, max=<0 0>, min=<0 0>}
       }
       pressedForeground {colorUseRule=UseColor_ColorTable, specifiedColor=<rgba 0 0 0 0>}
       pressedPadding=<ltrb 8 2 8 0>
       pressedSlateSound {resourceObject=null}
      }
      contentPadding=<ltrb 0 0 0 0>
      downArrowImage {
       bIsDynamicallyLoaded=false
       drawAs=Image
       imageType=Vector
       mirroring=NoMirror
       resourceName=../../../Engine/Content/Slate/Starship/CoreWidgets/ComboBox/wide-chevron-down.svg
       resourceObject=null
       tiling=NoTile
       imageSize=<16 16>
       margin=<ltrb 0 0 0 0>
       outlineSettings {
        bUseBrushTransparency=false
        roundingType=HalfHeightRadius
        width=0
        color {colorUseRule=UseColor_Specified, specifiedColor=<rgba 0 0 0 0>}
        cornerRadii=<0 0 0 0>
       }
       tintColor {colorUseRule=UseColor_Specified, specifiedColor=<rgba 1 1 1 1>}
       uVRegion {bIsValid=false, max=<0 0>, min=<0 0>}
      }
      downArrowPadding=<ltrb 2 2 2 2>
      menuBorderBrush {
       bIsDynamicallyLoaded=false
       drawAs=RoundedBox
       imageType=NoImage
       mirroring=NoMirror
       resourceName=None
       resourceObject=null
       tiling=NoTile
       imageSize=<0 0>
       margin=<ltrb 0 0 0 0>
       outlineSettings {
        bUseBrushTransparency=false
        roundingType=FixedRadius
        width=1
        color {colorUseRule=UseColor_ColorTable, specifiedColor=<rgba 0 0 0 0>}
        cornerRadii=<0 0 0 0>
       }
       tintColor {colorUseRule=UseColor_Specified, specifiedColor=<rgba 0.0395 0.0395 0.0395 1>}
       uVRegion {bIsValid=false, max=<0 0>, min=<0 0>}
      }
      menuBorderPadding=<ltrb 0 0 0 0>
      shadowColorAndOpacity=<rgba 0 0 0 1>
      shadowOffset=<0 0>
     }
     contentPadding=<ltrb 0 0 0 0>
     menuRowPadding=<ltrb 8 3 8 3>
     pressedSlateSound {resourceObject=null}
     selectionChangeSlateSound {resourceObject=null}
    }
   }
   slot {class=HorizontalBoxSlot}
  }
  DisableTAA {widgetClassPath=&HorizontalBox, parent=VerticalBox_519, widget=DisableTAA, slot {class=VerticalBoxSlot}}
  Spacer_5 {widgetClassPath=&Spacer, parent=DisableTAA, widget=Spacer_5, slot {class=HorizontalBoxSlot, properties {padding=<ltrb 0 0 40 0>}}}
  EnableTAA_Checkbox {
   widgetClassPath=&CheckBox
   bIsVariable=true
   parent=DisableTAA
   widget=EnableTAA_Checkbox
   properties {
    checkedState=Checked
    widgetStyle {
     checkBoxType=CheckBox
     backgroundHoveredImage {
      bIsDynamicallyLoaded=false
      drawAs=RoundedBox
      imageType=NoImage
      mirroring=NoMirror
      resourceName=None
      resourceObject=null
      tiling=NoTile
      imageSize=<18 18>
      margin=<ltrb 0 0 0 0>
      outlineSettings {
       bUseBrushTransparency=false
       roundingType=FixedRadius
       width=1
       color {colorUseRule=UseColor_ColorTable, specifiedColor=<rgba 0 0 0 0>}
       cornerRadii=<3 3 3 3>
      }
      tintColor {colorUseRule=UseColor_Specified, specifiedColor=<rgba 0.0048 0.0048 0.0048 1>}
      uVRegion {bIsValid=false, max=<0 0>, min=<0 0>}
     }
     backgroundImage {
      bIsDynamicallyLoaded=false
      drawAs=RoundedBox
      imageType=NoImage
      mirroring=NoMirror
      resourceName=None
      resourceObject=null
      tiling=NoTile
      imageSize=<18 18>
      margin=<ltrb 0 0 0 0>
      outlineSettings {
       bUseBrushTransparency=false
       roundingType=FixedRadius
       width=1
       color {colorUseRule=UseColor_ColorTable, specifiedColor=<rgba 0 0 0 0>}
       cornerRadii=<3 3 3 3>
      }
      tintColor {colorUseRule=UseColor_Specified, specifiedColor=<rgba 0.0048 0.0048 0.0048 1>}
      uVRegion {bIsValid=false, max=<0 0>, min=<0 0>}
     }
     backgroundPressedImage {
      bIsDynamicallyLoaded=false
      drawAs=RoundedBox
      imageType=NoImage
      mirroring=NoMirror
      resourceName=None
      resourceObject=null
      tiling=NoTile
      imageSize=<18 18>
      margin=<ltrb 0 0 0 0>
      outlineSettings {
       bUseBrushTransparency=false
       roundingType=FixedRadius
       width=1
       color {colorUseRule=UseColor_ColorTable, specifiedColor=<rgba 0 0 0 0>}
       cornerRadii=<3 3 3 3>
      }
      tintColor {colorUseRule=UseColor_Specified, specifiedColor=<rgba 0.0048 0.0048 0.0048 1>}
      uVRegion {bIsValid=false, max=<0 0>, min=<0 0>}
     }
     borderBackgroundColor {colorUseRule=UseColor_Specified, specifiedColor=<rgba 1 1 1 1>}
     checkedForeground {colorUseRule=UseColor_Specified, specifiedColor=<rgba 1 1 1 1>}
     checkedHoveredForeground {colorUseRule=UseColor_Specified, specifiedColor=<rgba 1 1 1 1>}
     checkedHoveredImage {
      bIsDynamicallyLoaded=false
      drawAs=Image
      imageType=Vector
      mirroring=NoMirror
      resourceName=../../../Engine/Content/Slate/Starship/CoreWidgets/CheckBox/check.svg
      resourceObject=null
      tiling=NoTile
      imageSize=<16 16>
      margin=<ltrb 0 0 0 0>
      outlineSettings {
       bUseBrushTransparency=false
       roundingType=HalfHeightRadius
       width=0
       color {colorUseRule=UseColor_Specified, specifiedColor=<rgba 0 0 0 0>}
       cornerRadii=<0 0 0 0>
      }
      tintColor {colorUseRule=UseColor_Specified, specifiedColor=<rgba 0.0044 0.2384 1 1>}
      uVRegion {bIsValid=false, max=<0 0>, min=<0 0>}
     }
     checkedImage {
      bIsDynamicallyLoaded=false
      drawAs=Image
      imageType=Vector
      mirroring=NoMirror
      resourceName=../../../Engine/Content/Slate/Starship/CoreWidgets/CheckBox/check.svg
      resourceObject=null
      tiling=NoTile
      imageSize=<16 16>
      margin=<ltrb 0 0 0 0>
      outlineSettings {
       bUseBrushTransparency=false
       roundingType=HalfHeightRadius
       width=0
       color {colorUseRule=UseColor_Specified, specifiedColor=<rgba 0 0 0 0>}
       cornerRadii=<0 0 0 0>
      }
      tintColor {colorUseRule=UseColor_Specified, specifiedColor=<rgba 0 0.162 0.7454 1>}
      uVRegion {bIsValid=false, max=<0 0>, min=<0 0>}
     }
     checkedPressedForeground {colorUseRule=UseColor_Specified, specifiedColor=<rgba 1 1 1 1>}
     checkedPressedImage {
      bIsDynamicallyLoaded=false
      drawAs=Image
      imageType=Vector
      mirroring=NoMirror
      resourceName=../../../Engine/Content/Slate/Starship/CoreWidgets/CheckBox/check.svg
      resourceObject=null
      tiling=NoTile
      imageSize=<16 16>
      margin=<ltrb 0 0 0 0>
      outlineSettings {
       bUseBrushTransparency=false
       roundingType=HalfHeightRadius
       width=0
       color {colorUseRule=UseColor_Specified, specifiedColor=<rgba 0 0 0 0>}
       cornerRadii=<0 0 0 0>
      }
      tintColor {colorUseRule=UseColor_Specified, specifiedColor=<rgba 0 0.0802 0.3515 1>}
      uVRegion {bIsValid=false, max=<0 0>, min=<0 0>}
     }
     checkedSlateSound {resourceObject=null}
     foregroundColor {colorUseRule=UseColor_Specified, specifiedColor=<rgba 1 1 1 1>}
     hoveredForeground {colorUseRule=UseColor_Specified, specifiedColor=<rgba 1 1 1 1>}
     hoveredSlateSound {resourceObject=null}
     padding=<ltrb 2 0 0 0>
     pressedForeground {colorUseRule=UseColor_Specified, specifiedColor=<rgba 1 1 1 1>}
     uncheckedHoveredImage {
      bIsDynamicallyLoaded=false
      drawAs=NoDrawType
      imageType=NoImage
      mirroring=NoMirror
      resourceName=None
      resourceObject=null
      tiling=NoTile
      imageSize=<0 0>
      margin=<ltrb 0 0 0 0>
      outlineSettings {
       bUseBrushTransparency=false
       roundingType=HalfHeightRadius
       width=0
       color {colorUseRule=UseColor_Specified, specifiedColor=<rgba 0 0 0 0>}
       cornerRadii=<0 0 0 0>
      }
      tintColor {colorUseRule=UseColor_Specified, specifiedColor=<rgba 1 1 1 1>}
      uVRegion {bIsValid=false, max=<0 0>, min=<0 0>}
     }
     uncheckedImage {
      bIsDynamicallyLoaded=false
      drawAs=NoDrawType
      imageType=NoImage
      mirroring=NoMirror
      resourceName=None
      resourceObject=null
      tiling=NoTile
      imageSize=<0 0>
      margin=<ltrb 0 0 0 0>
      outlineSettings {
       bUseBrushTransparency=false
       roundingType=HalfHeightRadius
       width=0
       color {colorUseRule=UseColor_Specified, specifiedColor=<rgba 0 0 0 0>}
       cornerRadii=<0 0 0 0>
      }
      tintColor {colorUseRule=UseColor_Specified, specifiedColor=<rgba 1 1 1 1>}
      uVRegion {bIsValid=false, max=<0 0>, min=<0 0>}
     }
     uncheckedPressedImage {
      bIsDynamicallyLoaded=false
      drawAs=NoDrawType
      imageType=NoImage
      mirroring=NoMirror
      resourceName=None
      resourceObject=null
      tiling=NoTile
      imageSize=<0 0>
      margin=<ltrb 0 0 0 0>
      outlineSettings {
       bUseBrushTransparency=false
       roundingType=HalfHeightRadius
       width=0
       color {colorUseRule=UseColor_Specified, specifiedColor=<rgba 0 0 0 0>}
       cornerRadii=<0 0 0 0>
      }
      tintColor {colorUseRule=UseColor_Specified, specifiedColor=<rgba 1 1 1 1>}
      uVRegion {bIsValid=false, max=<0 0>, min=<0 0>}
     }
     uncheckedSlateSound {resourceObject=null}
     undeterminedForeground {colorUseRule=UseColor_Specified, specifiedColor=<rgba 1 1 1 1>}
     undeterminedHoveredImage {
      bIsDynamicallyLoaded=false
      drawAs=Image
      imageType=Vector
      mirroring=NoMirror
      resourceName=../../../Engine/Content/Slate/Starship/CoreWidgets/CheckBox/indeterminate.svg
      resourceObject=null
      tiling=NoTile
      imageSize=<16 16>
      margin=<ltrb 0 0 0 0>
      outlineSettings {
       bUseBrushTransparency=false
       roundingType=HalfHeightRadius
       width=0
       color {colorUseRule=UseColor_Specified, specifiedColor=<rgba 0 0 0 0>}
       cornerRadii=<0 0 0 0>
      }
      tintColor {colorUseRule=UseColor_Specified, specifiedColor=<rgba 0.0044 0.2384 1 1>}
      uVRegion {bIsValid=false, max=<0 0>, min=<0 0>}
     }
     undeterminedImage {
      bIsDynamicallyLoaded=false
      drawAs=Image
      imageType=Vector
      mirroring=NoMirror
      resourceName=../../../Engine/Content/Slate/Starship/CoreWidgets/CheckBox/indeterminate.svg
      resourceObject=null
      tiling=NoTile
      imageSize=<16 16>
      margin=<ltrb 0 0 0 0>
      outlineSettings {
       bUseBrushTransparency=false
       roundingType=HalfHeightRadius
       width=0
       color {colorUseRule=UseColor_Specified, specifiedColor=<rgba 0 0 0 0>}
       cornerRadii=<0 0 0 0>
      }
      tintColor {colorUseRule=UseColor_Specified, specifiedColor=<rgba 0 0.162 0.7454 1>}
      uVRegion {bIsValid=false, max=<0 0>, min=<0 0>}
     }
     undeterminedPressedImage {
      bIsDynamicallyLoaded=false
      drawAs=Image
      imageType=Vector
      mirroring=NoMirror
      resourceName=../../../Engine/Content/Slate/Starship/CoreWidgets/CheckBox/indeterminate.svg
      resourceObject=null
      tiling=NoTile
      imageSize=<16 16>
      margin=<ltrb 0 0 0 0>
      outlineSettings {
       bUseBrushTransparency=false
       roundingType=HalfHeightRadius
       width=0
       color {colorUseRule=UseColor_Specified, specifiedColor=<rgba 0 0 0 0>}
       cornerRadii=<0 0 0 0>
      }
      tintColor {colorUseRule=UseColor_Specified, specifiedColor=<rgba 0 0.0802 0.3515 1>}
      uVRegion {bIsValid=false, max=<0 0>, min=<0 0>}
     }
    }
   }
   slot {class=HorizontalBoxSlot}
  }
  Spacer_6 {widgetClassPath=&Spacer, parent=DisableTAA, widget=Spacer_6, slot {class=HorizontalBoxSlot, properties {padding=<ltrb 0 0 10 0>}}}
  TextBlock_3 {
   widgetClassPath=&TextBlock
   parent=DisableTAA
   widget=TextBlock_3
   properties {
    text="TAA (Disable for Motion Blur Issues)"
    font {
     bForceMonospaced=false
     bMaterialIsStencil=false
     fontMaterial=null
     fontObject=&Roboto
     letterSpacing=0
     monospacedWidth=1
     size=14
     skewAmount=0
     typefaceFontName=Bold
     outlineSettings {
      bApplyOutlineToDropShadows=false
      bMiteredCorners=false
      bSeparateFillAlpha=false
      outlineMaterial=null
      outlineSize=0
      outlineColor=<rgba 0 0 0 1>
     }
    }
   }
   slot {class=HorizontalBoxSlot, properties {verticalAlignment=VAlign_Bottom}}
  }
  GPUDump {widgetClassPath=&HorizontalBox, parent=VerticalBox_519, widget=GPUDump, slot {class=VerticalBoxSlot}}
  Spacer_7 {widgetClassPath=&Spacer, parent=GPUDump, widget=Spacer_7, slot {class=HorizontalBoxSlot, properties {padding=<ltrb 0 0 40 0>}}}
  GPUDump_Checkbox {
   widgetClassPath=&CheckBox
   bIsVariable=true
   parent=GPUDump
   widget=GPUDump_Checkbox
   properties {widgetStyle {
     checkBoxType=CheckBox
     backgroundHoveredImage {
      bIsDynamicallyLoaded=false
      drawAs=RoundedBox
      imageType=NoImage
      mirroring=NoMirror
      resourceName=None
      resourceObject=null
      tiling=NoTile
      imageSize=<18 18>
      margin=<ltrb 0 0 0 0>
      outlineSettings {
       bUseBrushTransparency=false
       roundingType=FixedRadius
       width=1
       color {colorUseRule=UseColor_ColorTable, specifiedColor=<rgba 0 0 0 0>}
       cornerRadii=<3 3 3 3>
      }
      tintColor {colorUseRule=UseColor_Specified, specifiedColor=<rgba 0.0048 0.0048 0.0048 1>}
      uVRegion {bIsValid=false, max=<0 0>, min=<0 0>}
     }
     backgroundImage {
      bIsDynamicallyLoaded=false
      drawAs=RoundedBox
      imageType=NoImage
      mirroring=NoMirror
      resourceName=None
      resourceObject=null
      tiling=NoTile
      imageSize=<18 18>
      margin=<ltrb 0 0 0 0>
      outlineSettings {
       bUseBrushTransparency=false
       roundingType=FixedRadius
       width=1
       color {colorUseRule=UseColor_ColorTable, specifiedColor=<rgba 0 0 0 0>}
       cornerRadii=<3 3 3 3>
      }
      tintColor {colorUseRule=UseColor_Specified, specifiedColor=<rgba 0.0048 0.0048 0.0048 1>}
      uVRegion {bIsValid=false, max=<0 0>, min=<0 0>}
     }
     backgroundPressedImage {
      bIsDynamicallyLoaded=false
      drawAs=RoundedBox
      imageType=NoImage
      mirroring=NoMirror
      resourceName=None
      resourceObject=null
      tiling=NoTile
      imageSize=<18 18>
      margin=<ltrb 0 0 0 0>
      outlineSettings {
       bUseBrushTransparency=false
       roundingType=FixedRadius
       width=1
       color {colorUseRule=UseColor_ColorTable, specifiedColor=<rgba 0 0 0 0>}
       cornerRadii=<3 3 3 3>
      }
      tintColor {colorUseRule=UseColor_Specified, specifiedColor=<rgba 0.0048 0.0048 0.0048 1>}
      uVRegion {bIsValid=false, max=<0 0>, min=<0 0>}
     }
     borderBackgroundColor {colorUseRule=UseColor_Specified, specifiedColor=<rgba 1 1 1 1>}
     checkedForeground {colorUseRule=UseColor_Specified, specifiedColor=<rgba 1 1 1 1>}
     checkedHoveredForeground {colorUseRule=UseColor_Specified, specifiedColor=<rgba 1 1 1 1>}
     checkedHoveredImage {
      bIsDynamicallyLoaded=false
      drawAs=Image
      imageType=Vector
      mirroring=NoMirror
      resourceName=../../../Engine/Content/Slate/Starship/CoreWidgets/CheckBox/check.svg
      resourceObject=null
      tiling=NoTile
      imageSize=<16 16>
      margin=<ltrb 0 0 0 0>
      outlineSettings {
       bUseBrushTransparency=false
       roundingType=HalfHeightRadius
       width=0
       color {colorUseRule=UseColor_Specified, specifiedColor=<rgba 0 0 0 0>}
       cornerRadii=<0 0 0 0>
      }
      tintColor {colorUseRule=UseColor_Specified, specifiedColor=<rgba 0.0044 0.2384 1 1>}
      uVRegion {bIsValid=false, max=<0 0>, min=<0 0>}
     }
     checkedImage {
      bIsDynamicallyLoaded=false
      drawAs=Image
      imageType=Vector
      mirroring=NoMirror
      resourceName=../../../Engine/Content/Slate/Starship/CoreWidgets/CheckBox/check.svg
      resourceObject=null
      tiling=NoTile
      imageSize=<16 16>
      margin=<ltrb 0 0 0 0>
      outlineSettings {
       bUseBrushTransparency=false
       roundingType=HalfHeightRadius
       width=0
       color {colorUseRule=UseColor_Specified, specifiedColor=<rgba 0 0 0 0>}
       cornerRadii=<0 0 0 0>
      }
      tintColor {colorUseRule=UseColor_Specified, specifiedColor=<rgba 0 0.162 0.7454 1>}
      uVRegion {bIsValid=false, max=<0 0>, min=<0 0>}
     }
     checkedPressedForeground {colorUseRule=UseColor_Specified, specifiedColor=<rgba 1 1 1 1>}
     checkedPressedImage {
      bIsDynamicallyLoaded=false
      drawAs=Image
      imageType=Vector
      mirroring=NoMirror
      resourceName=../../../Engine/Content/Slate/Starship/CoreWidgets/CheckBox/check.svg
      resourceObject=null
      tiling=NoTile
      imageSize=<16 16>
      margin=<ltrb 0 0 0 0>
      outlineSettings {
       bUseBrushTransparency=false
       roundingType=HalfHeightRadius
       width=0
       color {colorUseRule=UseColor_Specified, specifiedColor=<rgba 0 0 0 0>}
       cornerRadii=<0 0 0 0>
      }
      tintColor {colorUseRule=UseColor_Specified, specifiedColor=<rgba 0 0.0802 0.3515 1>}
      uVRegion {bIsValid=false, max=<0 0>, min=<0 0>}
     }
     checkedSlateSound {resourceObject=null}
     foregroundColor {colorUseRule=UseColor_Specified, specifiedColor=<rgba 1 1 1 1>}
     hoveredForeground {colorUseRule=UseColor_Specified, specifiedColor=<rgba 1 1 1 1>}
     hoveredSlateSound {resourceObject=null}
     padding=<ltrb 2 0 0 0>
     pressedForeground {colorUseRule=UseColor_Specified, specifiedColor=<rgba 1 1 1 1>}
     uncheckedHoveredImage {
      bIsDynamicallyLoaded=false
      drawAs=NoDrawType
      imageType=NoImage
      mirroring=NoMirror
      resourceName=None
      resourceObject=null
      tiling=NoTile
      imageSize=<0 0>
      margin=<ltrb 0 0 0 0>
      outlineSettings {
       bUseBrushTransparency=false
       roundingType=HalfHeightRadius
       width=0
       color {colorUseRule=UseColor_Specified, specifiedColor=<rgba 0 0 0 0>}
       cornerRadii=<0 0 0 0>
      }
      tintColor {colorUseRule=UseColor_Specified, specifiedColor=<rgba 1 1 1 1>}
      uVRegion {bIsValid=false, max=<0 0>, min=<0 0>}
     }
     uncheckedImage {
      bIsDynamicallyLoaded=false
      drawAs=NoDrawType
      imageType=NoImage
      mirroring=NoMirror
      resourceName=None
      resourceObject=null
      tiling=NoTile
      imageSize=<0 0>
      margin=<ltrb 0 0 0 0>
      outlineSettings {
       bUseBrushTransparency=false
       roundingType=HalfHeightRadius
       width=0
       color {colorUseRule=UseColor_Specified, specifiedColor=<rgba 0 0 0 0>}
       cornerRadii=<0 0 0 0>
      }
      tintColor {colorUseRule=UseColor_Specified, specifiedColor=<rgba 1 1 1 1>}
      uVRegion {bIsValid=false, max=<0 0>, min=<0 0>}
     }
     uncheckedPressedImage {
      bIsDynamicallyLoaded=false
      drawAs=NoDrawType
      imageType=NoImage
      mirroring=NoMirror
      resourceName=None
      resourceObject=null
      tiling=NoTile
      imageSize=<0 0>
      margin=<ltrb 0 0 0 0>
      outlineSettings {
       bUseBrushTransparency=false
       roundingType=HalfHeightRadius
       width=0
       color {colorUseRule=UseColor_Specified, specifiedColor=<rgba 0 0 0 0>}
       cornerRadii=<0 0 0 0>
      }
      tintColor {colorUseRule=UseColor_Specified, specifiedColor=<rgba 1 1 1 1>}
      uVRegion {bIsValid=false, max=<0 0>, min=<0 0>}
     }
     uncheckedSlateSound {resourceObject=null}
     undeterminedForeground {colorUseRule=UseColor_Specified, specifiedColor=<rgba 1 1 1 1>}
     undeterminedHoveredImage {
      bIsDynamicallyLoaded=false
      drawAs=Image
      imageType=Vector
      mirroring=NoMirror
      resourceName=../../../Engine/Content/Slate/Starship/CoreWidgets/CheckBox/indeterminate.svg
      resourceObject=null
      tiling=NoTile
      imageSize=<16 16>
      margin=<ltrb 0 0 0 0>
      outlineSettings {
       bUseBrushTransparency=false
       roundingType=HalfHeightRadius
       width=0
       color {colorUseRule=UseColor_Specified, specifiedColor=<rgba 0 0 0 0>}
       cornerRadii=<0 0 0 0>
      }
      tintColor {colorUseRule=UseColor_Specified, specifiedColor=<rgba 0.0044 0.2384 1 1>}
      uVRegion {bIsValid=false, max=<0 0>, min=<0 0>}
     }
     undeterminedImage {
      bIsDynamicallyLoaded=false
      drawAs=Image
      imageType=Vector
      mirroring=NoMirror
      resourceName=../../../Engine/Content/Slate/Starship/CoreWidgets/CheckBox/indeterminate.svg
      resourceObject=null
      tiling=NoTile
      imageSize=<16 16>
      margin=<ltrb 0 0 0 0>
      outlineSettings {
       bUseBrushTransparency=false
       roundingType=HalfHeightRadius
       width=0
       color {colorUseRule=UseColor_Specified, specifiedColor=<rgba 0 0 0 0>}
       cornerRadii=<0 0 0 0>
      }
      tintColor {colorUseRule=UseColor_Specified, specifiedColor=<rgba 0 0.162 0.7454 1>}
      uVRegion {bIsValid=false, max=<0 0>, min=<0 0>}
     }
     undeterminedPressedImage {
      bIsDynamicallyLoaded=false
      drawAs=Image
      imageType=Vector
      mirroring=NoMirror
      resourceName=../../../Engine/Content/Slate/Starship/CoreWidgets/CheckBox/indeterminate.svg
      resourceObject=null
      tiling=NoTile
      imageSize=<16 16>
      margin=<ltrb 0 0 0 0>
      outlineSettings {
       bUseBrushTransparency=false
       roundingType=HalfHeightRadius
       width=0
       color {colorUseRule=UseColor_Specified, specifiedColor=<rgba 0 0 0 0>}
       cornerRadii=<0 0 0 0>
      }
      tintColor {colorUseRule=UseColor_Specified, specifiedColor=<rgba 0 0.0802 0.3515 1>}
      uVRegion {bIsValid=false, max=<0 0>, min=<0 0>}
     }
    }}
   slot {class=HorizontalBoxSlot}
  }
  Spacer_8 {widgetClassPath=&Spacer, parent=GPUDump, widget=Spacer_8, slot {class=HorizontalBoxSlot, properties {padding=<ltrb 0 0 10 0>}}}
  TextBlock_5 {
   widgetClassPath=&TextBlock
   parent=GPUDump
   widget=TextBlock_5
   properties {
    text=Dump GPU
    font {
     bForceMonospaced=false
     bMaterialIsStencil=false
     fontMaterial=null
     fontObject=&Roboto
     letterSpacing=0
     monospacedWidth=1
     size=14
     skewAmount=0
     typefaceFontName=Bold
     outlineSettings {
      bApplyOutlineToDropShadows=false
      bMiteredCorners=false
      bSeparateFillAlpha=false
      outlineMaterial=null
      outlineSize=0
      outlineColor=<rgba 0 0 0 1>
     }
    }
   }
   slot {class=HorizontalBoxSlot, properties {verticalAlignment=VAlign_Bottom}}
  }
  Spacer_9 {widgetClassPath=&Spacer, parent=GPUDump, widget=Spacer_9, slot {class=HorizontalBoxSlot, properties {padding=<ltrb 0 0 10 0>}}}
  Border_1 {
   widgetClassPath=&Border
   parent=GPUDump
   widget=Border_1
   properties {brushColor=<rgba 0.0312 0.0312 0.0312 1>}
   slot {class=HorizontalBoxSlot}
  }
  GPUDumpFrame {
   widgetClassPath=/Script/UMG.EditableText
   bIsVariable=true
   parent=Border_1
   widget=GPUDumpFrame
   properties {
    bIsEnabled=false
    hintText=Frame to run Dump GPU
    text=0
    widgetStyle {
     backgroundImageComposing {
      bIsDynamicallyLoaded=false
      drawAs=Border
      imageType=FullColor
      mirroring=NoMirror
      resourceName=../../../Engine/Content/Slate/Old/HyperlinkDotted.png
      resourceObject=null
      tiling=Both
      imageSize=<0 0>
      margin=<ltrb 0 0 0 0.1875>
      outlineSettings {
       bUseBrushTransparency=false
       roundingType=HalfHeightRadius
       width=0
       color {colorUseRule=UseColor_Specified, specifiedColor=<rgba 0 0 0 0>}
       cornerRadii=<0 0 0 0>
      }
      tintColor {colorUseRule=UseColor_Specified, specifiedColor=<rgba 1 1 1 1>}
      uVRegion {bIsValid=false, max=<0 0>, min=<0 0>}
     }
     backgroundImageSelected {
      bIsDynamicallyLoaded=false
      drawAs=Image
      imageType=NoImage
      mirroring=NoMirror
      resourceName=None
      resourceObject=null
      tiling=NoTile
      imageSize=<0 0>
      margin=<ltrb 0 0 0 0>
      outlineSettings {
       bUseBrushTransparency=false
       roundingType=HalfHeightRadius
       width=0
       color {colorUseRule=UseColor_Specified, specifiedColor=<rgba 0 0 0 0>}
       cornerRadii=<0 0 0 0>
      }
      tintColor {colorUseRule=UseColor_Specified, specifiedColor=<rgba 0 0.162 0.7454 1>}
      uVRegion {bIsValid=false, max=<0 0>, min=<0 0>}
     }
     caretImage {
      bIsDynamicallyLoaded=false
      drawAs=Image
      imageType=NoImage
      mirroring=NoMirror
      resourceName=None
      resourceObject=null
      tiling=NoTile
      imageSize=<0 0>
      margin=<ltrb 0 0 0 0>
      outlineSettings {
       bUseBrushTransparency=false
       roundingType=HalfHeightRadius
       width=0
       color {colorUseRule=UseColor_Specified, specifiedColor=<rgba 0 0 0 0>}
       cornerRadii=<0 0 0 0>
      }
      tintColor {colorUseRule=UseColor_Specified, specifiedColor=<rgba 1 1 1 1>}
      uVRegion {bIsValid=false, max=<0 0>, min=<0 0>}
     }
     colorAndOpacity {colorUseRule=UseColor_Foreground, specifiedColor=<rgba 1 0 1 1>}
     font {
      bForceMonospaced=false
      bMaterialIsStencil=false
      fontMaterial=null
      fontObject=null
      letterSpacing=0
      monospacedWidth=1
      size=14
      skewAmount=0
      typefaceFontName=Regular
      outlineSettings {
       bApplyOutlineToDropShadows=false
       bMiteredCorners=false
       bSeparateFillAlpha=false
       outlineMaterial=null
       outlineSize=0
       outlineColor=<rgba 0 0 0 1>
      }
     }
     hintColorAndOpacity {colorUseRule=UseColor_Foreground, specifiedColor=<rgba 1 0 1 1>}
     selectedBackgroundColor {colorUseRule=UseColor_Foreground, specifiedColor=<rgba 1 0 1 1>}
    }
   }
   slot {class=BorderSlot}
  }
  DebugBox {
   widgetClassPath=&VerticalBox
   parent=HorizontalBox_4
   widget=DebugBox
   slot {class=HorizontalBoxSlot, properties {size {sizeRule=Fill, value=1}}}
  }
  Border {
   widgetClassPath=&Border
   parent=DebugBox
   widget=Border
   properties {brushColor=<rgba 0.151 0.151 0.151 0.453>}
   slot {class=VerticalBoxSlot, properties {size {sizeRule=Fill, value=1}}}
  }
  HorizontalBox_125 {widgetClassPath=&HorizontalBox, parent=Border, widget=HorizontalBox_125, slot {class=BorderSlot}}
  VerticalBox_3 {
   widgetClassPath=&VerticalBox
   parent=HorizontalBox_125
   widget=VerticalBox_3
   slot {class=HorizontalBoxSlot, properties {horizontalAlignment=HAlign_Left, verticalAlignment=VAlign_Top}}
  }
  Debug {
   widgetClassPath=&Button
   parent=VerticalBox_3
   widget=Debug
   properties {widgetStyle {
     clickedSlateSound {resourceObject=null}
     disabled {
      bIsDynamicallyLoaded=false
      drawAs=RoundedBox
      imageType=NoImage
      mirroring=NoMirror
      resourceName=None
      resourceObject=null
      tiling=NoTile
      imageSize=<0 0>
      margin=<ltrb 0 0 0 0>
      outlineSettings {
       bUseBrushTransparency=false
       roundingType=FixedRadius
       width=1
       color {colorUseRule=UseColor_ColorTable, specifiedColor=<rgba 0 0 0 0>}
       cornerRadii=<4 4 4 4>
      }
      tintColor {colorUseRule=UseColor_Specified, specifiedColor=<rgba 0.0395 0.0395 0.0395 1>}
      uVRegion {bIsValid=false, max=<0 0>, min=<0 0>}
     }
     disabledForeground {colorUseRule=UseColor_ColorTable, specifiedColor=<rgba 0 0 0 0>}
     hovered {
      bIsDynamicallyLoaded=false
      drawAs=RoundedBox
      imageType=NoImage
      mirroring=NoMirror
      resourceName=None
      resourceObject=null
      tiling=NoTile
      imageSize=<32 32>
      margin=<ltrb 0 0 0 0>
      outlineSettings {
       bUseBrushTransparency=false
       roundingType=FixedRadius
       width=1
       color {colorUseRule=UseColor_ColorTable, specifiedColor=<rgba 0 0 0 0>}
       cornerRadii=<4 4 4 4>
      }
      tintColor {colorUseRule=UseColor_Specified, specifiedColor=<rgba 0.0953 0.0953 0.0953 1>}
      uVRegion {bIsValid=false, max=<0 0>, min=<0 0>}
     }
     hoveredForeground {colorUseRule=UseColor_ColorTable, specifiedColor=<rgba 0 0 0 0>}
     hoveredSlateSound {resourceObject=null}
     normal {
      bIsDynamicallyLoaded=false
      drawAs=RoundedBox
      imageType=NoImage
      mirroring=NoMirror
      resourceName=None
      resourceObject=null
      tiling=NoTile
      imageSize=<32 32>
      margin=<ltrb 0 0 0 0>
      outlineSettings {
       bUseBrushTransparency=false
       roundingType=FixedRadius
       width=1
       color {colorUseRule=UseColor_ColorTable, specifiedColor=<rgba 0 0 0 0>}
       cornerRadii=<4 4 4 4>
      }
      tintColor {colorUseRule=UseColor_Specified, specifiedColor=<rgba 0.0395 0.0395 0.0395 1>}
      uVRegion {bIsValid=false, max=<0 0>, min=<0 0>}
     }
     normalForeground {colorUseRule=UseColor_ColorTable, specifiedColor=<rgba 0 0 0 0>}
     normalPadding=<ltrb 12 1.5 12 1.5>
     pressed {
      bIsDynamicallyLoaded=false
      drawAs=RoundedBox
      imageType=NoImage
      mirroring=NoMirror
      resourceName=None
      resourceObject=null
      tiling=NoTile
      imageSize=<32 32>
      margin=<ltrb 0 0 0 0>
      outlineSettings {
       bUseBrushTransparency=false
       roundingType=FixedRadius
       width=1
       color {colorUseRule=UseColor_ColorTable, specifiedColor=<rgba 0 0 0 0>}
       cornerRadii=<4 4 4 4>
      }
      tintColor {colorUseRule=UseColor_Specified, specifiedColor=<rgba 0.0284 0.0284 0.0284 1>}
      uVRegion {bIsValid=false, max=<0 0>, min=<0 0>}
     }
     pressedForeground {colorUseRule=UseColor_ColorTable, specifiedColor=<rgba 0 0 0 0>}
     pressedPadding=<ltrb 12 2.5 12 0.5>
     pressedSlateSound {resourceObject=null}
    }}
   slot {class=VerticalBoxSlot}
  }
  TextBlock_2 {widgetClassPath=&TextBlock, parent=Debug, widget=TextBlock_2, properties {text=Generate Debug Truths}, slot {class=ButtonSlot}}
  Documentation {
   widgetClassPath=&Button
   parent=VerticalBox_3
   widget=Documentation
   properties {widgetStyle {
     clickedSlateSound {resourceObject=null}
     disabled {
      bIsDynamicallyLoaded=false
      drawAs=RoundedBox
      imageType=NoImage
      mirroring=NoMirror
      resourceName=None
      resourceObject=null
      tiling=NoTile
      imageSize=<0 0>
      margin=<ltrb 0 0 0 0>
      outlineSettings {
       bUseBrushTransparency=false
       roundingType=FixedRadius
       width=1
       color {colorUseRule=UseColor_ColorTable, specifiedColor=<rgba 0 0 0 0>}
       cornerRadii=<4 4 4 4>
      }
      tintColor {colorUseRule=UseColor_Specified, specifiedColor=<rgba 0.0395 0.0395 0.0395 1>}
      uVRegion {bIsValid=false, max=<0 0>, min=<0 0>}
     }
     disabledForeground {colorUseRule=UseColor_ColorTable, specifiedColor=<rgba 0 0 0 0>}
     hovered {
      bIsDynamicallyLoaded=false
      drawAs=RoundedBox
      imageType=NoImage
      mirroring=NoMirror
      resourceName=None
      resourceObject=null
      tiling=NoTile
      imageSize=<32 32>
      margin=<ltrb 0 0 0 0>
      outlineSettings {
       bUseBrushTransparency=false
       roundingType=FixedRadius
       width=1
       color {colorUseRule=UseColor_ColorTable, specifiedColor=<rgba 0 0 0 0>}
       cornerRadii=<4 4 4 4>
      }
      tintColor {colorUseRule=UseColor_Specified, specifiedColor=<rgba 0.0953 0.0953 0.0953 1>}
      uVRegion {bIsValid=false, max=<0 0>, min=<0 0>}
     }
     hoveredForeground {colorUseRule=UseColor_ColorTable, specifiedColor=<rgba 0 0 0 0>}
     hoveredSlateSound {resourceObject=null}
     normal {
      bIsDynamicallyLoaded=false
      drawAs=RoundedBox
      imageType=NoImage
      mirroring=NoMirror
      resourceName=None
      resourceObject=null
      tiling=NoTile
      imageSize=<32 32>
      margin=<ltrb 0 0 0 0>
      outlineSettings {
       bUseBrushTransparency=false
       roundingType=FixedRadius
       width=1
       color {colorUseRule=UseColor_ColorTable, specifiedColor=<rgba 0 0 0 0>}
       cornerRadii=<4 4 4 4>
      }
      tintColor {colorUseRule=UseColor_Specified, specifiedColor=<rgba 0.0395 0.0395 0.0395 1>}
      uVRegion {bIsValid=false, max=<0 0>, min=<0 0>}
     }
     normalForeground {colorUseRule=UseColor_ColorTable, specifiedColor=<rgba 0 0 0 0>}
     normalPadding=<ltrb 12 1.5 12 1.5>
     pressed {
      bIsDynamicallyLoaded=false
      drawAs=RoundedBox
      imageType=NoImage
      mirroring=NoMirror
      resourceName=None
      resourceObject=null
      tiling=NoTile
      imageSize=<32 32>
      margin=<ltrb 0 0 0 0>
      outlineSettings {
       bUseBrushTransparency=false
       roundingType=FixedRadius
       width=1
       color {colorUseRule=UseColor_ColorTable, specifiedColor=<rgba 0 0 0 0>}
       cornerRadii=<4 4 4 4>
      }
      tintColor {colorUseRule=UseColor_Specified, specifiedColor=<rgba 0.0284 0.0284 0.0284 1>}
      uVRegion {bIsValid=false, max=<0 0>, min=<0 0>}
     }
     pressedForeground {colorUseRule=UseColor_ColorTable, specifiedColor=<rgba 0 0 0 0>}
     pressedPadding=<ltrb 12 2.5 12 0.5>
     pressedSlateSound {resourceObject=null}
    }}
   slot {class=VerticalBoxSlot, properties {horizontalAlignment=HAlign_Left, verticalAlignment=VAlign_Top}}
  }
  TextBlock_4 {
   widgetClassPath=&TextBlock
   parent=Documentation
   widget=TextBlock_4
   properties {
    text=Documentation
    font {
     bForceMonospaced=false
     bMaterialIsStencil=false
     fontMaterial=null
     fontObject=&Roboto
     letterSpacing=0
     monospacedWidth=1
     size=18
     skewAmount=0
     typefaceFontName=Bold
     outlineSettings {
      bApplyOutlineToDropShadows=false
      bMiteredCorners=false
      bSeparateFillAlpha=false
      outlineMaterial=null
      outlineSize=0
      outlineColor=<rgba 0 0 0 1>
     }
    }
   }
   slot {class=ButtonSlot}
  }
  MRQSettingLocator {
   widgetClassPath=&DetailsView
   parent=HorizontalBox_125
   widget=MRQSettingLocator
   properties {categoriesToShow [MRQRenderSettings]}
   slot {class=HorizontalBoxSlot, properties {size {sizeRule=Fill, value=1}}}
  }
  MeshOverride_DV {
   widgetClassPath=&DetailsView
   parent=VerticalBox_4
   widget=MeshOverride_DV
   properties {categoriesToShow [OverrideMesh]}
   slot {class=VerticalBoxSlot}
  }
  MultiCIDCapture_DV {
   widgetClassPath=&DetailsView
   parent=VerticalBox_4
   widget=MultiCIDCapture_DV
   properties {categoriesToShow [MultiCID]}
   slot {class=VerticalBoxSlot}
  }
 ]
}
